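# Niagara body effect for the VoyagerCuboid player cosmetic, layering flames, glows, cubes and lightning on the character's skeletal mesh.
# Emitters pick bones or sockets by partition or spawn group: flame sprites, glows with lights, drifting cubes, spline-mesh arcs and longer bolts along multi-bone curves.
# User flags zero out hand, foot and ADS-sensitive elements and adjust glow color for front end.
# Timeline fades flames and long bolts; time of day scales glows.
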NiagaraSystem /BRCosmetics/Characters/Player/Male/Medium/Bodies/M_MED_VoyagerCuboid/FX/NS_Character_VoyagerCuboid_Body
dependencies {
 usedDataInterfaces [
  &NiagaraDataInterfaceMeshRendererInfo
  &NiagaraDataInterfaceCurve
  &NiagaraDataInterfaceVector2DCurve
  &NiagaraDataInterfaceArrayFloat3
  &NiagaraDataInterfaceArrayFloat
  &NiagaraDataInterfaceSpriteRendererInfo
  &NiagaraDataInterfaceArrayInt32
  &NiagaraDataInterfaceVectorCurve
  &NiagaraDataInterfaceSkeletalMesh
  &NiagaraDataInterfaceColorCurve
 ]
 usedDynamicInputs [
  &SelectFloatFromArray
  &MaskFloat
  &ReturnNormalizedExecIndex
 ]
 usedModules [
  &self:SystemSpawnScript.SystemScriptSource.SystemScriptGraph.NiagaraNodeAssignment_0.SetVariables_#1
  &self:SystemSpawnScript.SystemScriptSource.SystemScriptGraph.NiagaraNodeAssignment_1.SetVariables_#2
  &self:SystemSpawnScript.SystemScriptSource.SystemScriptGraph.NiagaraNodeAssignment_2.SetVariables_#3
  &self:SystemSpawnScript.SystemScriptSource.SystemScriptGraph.NiagaraNodeAssignment_3.SetVariables_#4
  &SystemState
  &Timeline
 ]
}
emitters {
 CubicFlame {
  emitterUpdateScript [
   &EmitterState()
   &SpawnRate(Spawn Probability=0.8, SpawnRate=50, Use Spawn Probability=true)
   &SpawnPerFrame(
    Spawn:NiagaraBool=&Boolean_AndOperation(
     Bool1=&Boolean_AndOperation(
      Bool1=&Boolean_NotOperation(Bool=$System.UpdateValueStateChange.Changed)
      Bool2=&Boolean_NotOperation(Bool=$Transient.FirstFrame)
     )
     Bool2=$System.Reactive
    )
    Spawn Count=200
    Use Spawn Probability=false
   ) {enabled=false}
  ]
  particleSpawnScript [
   &PartitionParticles(Number Of Partitions=3)
   &InitializeParticle(
    Color Mode=::NewEnumerator0
    Lifetime=&RandomRangeFloat(Minimum=0.5, Maximum=0.75)
    Sprite Size Max=<15 20>
    Sprite Size Min=<10 10>
    Sprite Size Mode=::NewEnumerator2
    UsePositionOffset=false
    Write Material Random=true
   )
   &AddVelocity(Module.Use Falloff Distance=false, Origin Offset=<0 0 -10>, Override Seed=false, Velocity Mode=::NewEnumerator1, Velocity Speed=10) {enabled=false}
   &ApplyInitialForces() {enabled=false}
   &DynamicMaterialParameters(
    Index 0 Param 1=&RandomRangeFloat(Minimum=0, Maximum=1)
    Index 0 Param 2=&RandomRangeFloat(Minimum=0, Maximum=1)
    Module.Index0Float0Write=true
    Module.Index0Float1Write=true
    Module.Index0Float2Write=true
    Module.Index0Float3Write=true
   )
  ]
  particleUpdateScript [
   &ParticleState()
   &ScaleSpriteSize(
    Non-Uniform Curve Scale=&RandomRangeVector2D(Minimum=<0.8 0.8>, Maximum=<1 1>)
    Non-Uniform Curve Sprite Scale=DataInterface() {propertyValues {
      XCurve {keys [{interpMode=RCIM_Cubic, tangentMode=RCTM_User, time=0, value=1}, {arriveTangent=-2.4882, tangentMode=RCTM_User, time=1, value=0}]}
      YCurve {keys [
        {arriveTangent=2.1941, interpMode=RCIM_Cubic, leaveTangent=2.1941, tangentMode=RCTM_User, time=0, value=0.1}
        {arriveTangent=0.6251, interpMode=RCIM_Cubic, leaveTangent=0.6251, tangentMode=RCTM_User, time=1, value=2}
       ]}
     }}
    Scale Sprite Size Mode=::NewEnumerator3
   )
   &CameraOffset(Camera Offset Amount=&RandomRangeFloat(Minimum=-2, Maximum=2))
   &SkeletalMeshLocation(
    Bone / Socket Index=&MergeHelper_IntSwitch(0=4, 1=5, 2=3, 3=?, 4=?, 5=?, Index=$Particles.PartitionParticles.Partition, Num=::NewEnumerator1)
    Bone Sampling Mode=::NewEnumerator4
    MaskSpawnGroup=false
    Orientation Sampling=::NewEnumerator1
    Sampled Position Offset=&SelectVectorFromArray(
     Direct Array Index=$Particles.PartitionParticles.Partition
     Vector Selection Array=DataInterface() {propertyValues {FloatData [<-3 -1 0>, <3 1 0>, <8 0 0>]}}
    )
    Skeletal Mesh=$System.Skeletal Mesh Reader
   )
   &ShapeLocation(
    Additional Quaternion Rotation=false
    Box Midpoint=<0.5 0.5 0.75>
    Box Size=&SelectVectorFromArray(
     Direct Array Index=$Particles.PartitionParticles.Partition
     Vector Selection Array=DataInterface() {propertyValues {FloatData [<5 7 2>, <5 7 2>, <12 10 5>]}}
    )
    MaskSpawnGroup=false
    Override Seed=false
    Rotation Mode=::NewEnumerator3
    Rotation Quaternion=&MultiplyQuaternion(
     Quaternion A=$Particles.SkeletalMeshLocation.SampledRotation
     Quaternion B=/Niagara/DynamicInputs/Rotation/InvertQuaternion(Quat:Quat4f=<0 1 0 1>)
    )
    Shape Origin=$Particles.Position
    Shape Primitive=::NewEnumerator2
    Surface Only=false
    UseOwnerScale=true
   )
   &GravityForce(Gravity=&RandomRangeVector(Minimum=<0 0 10>, Maximum=<0 0 25>)) {enabled=false}
   &SolveForcesAndVelocity(Clamp Velocity=false, Limit Acceleration=false)
   &ScaleColor(
    Scale Alpha=&FloatFromCurve(
     FloatCurve=DataInterface() {propertyValues {Curve {keys [<at 0 0>,<at 0.1 1>,<at 1 0>]}}}
     Scale Curve=&FloatFromCurve(
      FloatCurve=DataInterface() {propertyValues {Curve {
         defaultValue=0
         keys [{interpMode=RCIM_Cubic, tangentMode=RCTM_User, time=0, value=1}, {interpMode=RCIM_Cubic, tangentMode=RCTM_User, time=1, value=0}]
        }}}
      CurveIndex=$System.Timeline.NormalizedTime
     )
    )
    Scale RGB=&Multiply_VectorByFloat(
     Vector=<1 1 1>
     Float=&FloatFromCurve(FloatCurve=DataInterface() {propertyValues {Curve {keys [<at 0 1>,<at 0.084 3>,<at 1 0>]}}})
    )
    ScaleA=true
    ScaleRGB=true
   )
  ]
  rendererClasses [&NiagaraSpriteRendererProperties]
  renderers [{rendererClass=&NiagaraSpriteRendererProperties}]
 }
 Glows {
  emitterUpdateScript [
   &EmitterState(Life Cycle Mode=::NewEnumerator1, Loop Behavior=::NewEnumerator1, Loop Duration Mode=::NewEnumerator1, UseLoopDelay=false)
   &SpawnBurst_Instantaneous(Spawn Group=3, Use Spawn Probability=false, UseLoopCountLimit=false) {module=SpawnBurst_Instantaneous003}
   &SpawnBurst_Instantaneous(Spawn Count=2, Spawn Group=2, Use Spawn Probability=false, UseLoopCountLimit=false) {module=SpawnBurst_Instantaneous002}
   &SpawnBurst_Instantaneous(Spawn Count=5, Use Spawn Probability=false, UseLoopCountLimit=true) {enabled=false}
   &SpawnBurst_Instantaneous(Spawn Count=7, Spawn Group=1, Use Spawn Probability=false, UseLoopCountLimit=false) {module=SpawnBurst_Instantaneous001}
  ]
  particleSpawnScript [
   /SetVariables(
    Particles.GlowIndex:NiagaraInt32=&MergeHelper_IntSwitch(0=0, 1=1, 2=2, 3=3, 4=4, 5=5, 6=6, Index=&ReturnExecIndex(), Num=::NewEnumerator5)
    Particles.VisibilityTag=&SpawnGroup()
   ) {isSetParameters=true}
   &PartitionParticles(
    Number Of Partitions=&MergeHelper_IntSwitch(0=5, 1=7, 2=2, 3=1, 4=?, 5=?, Index=&SpawnGroup(), Num=::NewEnumerator2)
    Particles Per Partition=1
    Partition Mode:ENiagara_PartitionMode_v2=::NewEnumerator1
   )
   &InitializeParticle(
    Color=&MakeCustomLinearColorFromBool(
     Bool=&Boolean_OrOperation(Bool1=$User.bIsFrontEndPreview, Bool2=$User.bIsFrontEnd)
     True Color=<rgba 0.8 0.3828 0 1>
     False Color=<rgba 1.2 0.5742 0 1>
    )
    Mesh Scale Mode=::NewEnumerator3
    Mesh Uniform Scale=0.5
    Sprite Size=&MaskVector2dBySpawnGroup(
     Masked Vector2D=&MergeHelper_Vector2Switch(
      0=<18 30>
      1=<18 30>
      2=<22 40>
      3=<22 40>
      4=<18 40>
      5=<18 40>
      6=<30 50>
      Index=$Particles.PartitionParticles.Partition
      Num=::NewEnumerator5
     )
     Passthrough Vector2D=<25 25>
     Spawn Group Mask=1
    )
    Sprite Size Mode=::NewEnumerator4
    UsePositionOffset=false
    Write Material Random=true
   )
   /SetVariables(Particles.GlowCameraOffset=5) {isSetParameters=true}
  ]
  particleUpdateScript [
   &ParticleState(Kill Particles When Lifetime Has Elapsed=false, Let Infinitely Lived Particles Die When Emitter Deactivates=true)
   &SkeletalMeshLocation(
    Bone / Socket Index=&MaskIntBySpawnGroup(
     Masked Int=&MergeHelper_IntSwitch(0=16, 1=13, 2=14, 3=15, 4=12, 5=?, Index=$Particles.PartitionParticles.Partition, Num=::NewEnumerator3)
     Passthrough Int=/Niagara/DynamicInputs/Helpers/MaskInt(
      Mask Source:NiagaraInt32=0
      Mask Parameter:NiagaraInt32=?
      Mask:NiagaraInt32=1
      Masked Int:NiagaraInt32=&MergeHelper_IntSwitch(0=4, 1=5, 2=8, 3=9, 4=10, 5=11, 6=20, Index=$Particles.PartitionParticles.Partition, Num=::NewEnumerator5)
      Passthrough Int:NiagaraInt32=&MaskIntBySpawnGroup(
       Masked Int=&MergeHelper_IntSwitch(0=4, 1=5, 2=?, 3=?, 4=?, 5=?, Index=$Particles.PartitionParticles.Partition, Num=::NewEnumerator0)
       Passthrough Int=12
       Spawn Group Mask=2
      )
     )
     Spawn Group Mask=0
    )
    Bone Sampling Mode=::NewEnumerator4
    MaskSpawnGroup=false
    Sampled Position Offset=&MaskVectorBySpawnGroup(
     Masked Vector=<0 0 0>
     Passthrough Vector=&MaskVector(
      Mask Source=0
      Mask Parameter=?
      Mask=1
      Masked Vector=&MergeHelper_VectorSwitch(
       0=<17 0 0>
       1=<-17 0 0>
       2=<-20 0 0>
       3=<20 0 0>
       4=<-20 0 0>
       5=<20 0 0>
       6=<5 0 0>
       Index=$Particles.PartitionParticles.Partition
       Num=::NewEnumerator5
      )
      Passthrough Vector=&MaskVector(
       Mask Source=0
       Mask Parameter=?
       Mask=2
       Masked Vector=&MaskVector(Mask Source=1, Mask Parameter=$Particles.PartitionParticles.Partition, Mask=0, Masked Vector=<0 0 0>, Passthrough Vector=<1 1 1>)
       Passthrough Vector=<0 0 0>
      )
     )
    )
    Skeletal Mesh=$System.Skeletal Mesh Reader
   )
   &SpriteFacingAndAlignment(
    Alignment Coordinate Space=::Local
    Sprite Alignment=&ReturnMeshOrientationAxis(Mesh Orientation Axis=::XAxis, Mesh Orientation Quat=$Particles.MeshOrientation)
    WriteAlignment=true
    WriteFacing=false
   )
   &ScaleSpriteSize(
    Scale Factor=&Vector2DFromFloat(Value=&FloatFromCurve(FloatCurve=DataInterface() {propertyValues {Curve {keys [{interpMode=RCIM_Cubic, tangentMode=RCTM_User, time=0, value=1.2}, {interpMode=RCIM_Cubic, tangentMode=RCTM_User, time=0.5, value=1.7}]}}}))
    Scale Sprite Size Mode=::NewEnumerator2
   )
   &ScaleColor(
    Scale Alpha=&FloatFromCurve(
     FloatCurve=DataInterface() {propertyValues {Curve {keys [<at 0 0 RCIM_Cubic>,<at 0.2321 1 RCIM_Cubic>]}}}
     Scale Curve=&Multiply_Float(
      A=&FNTimeOfDayScalar(Night Scale=0.7, Morning Scale=1, Evening Scale=1)
      B=&MergeHelper_FloatSwitch(
       0=1
       1=&MergeHelper_FloatSwitch(
        0=1
        1=1
        2=1
        3=1
        4=&MakeCustomFloatFromBool(Bool=$User.bisFootHidden, True Float=0, False Float=1)
        5=&MakeCustomFloatFromBool(Bool=$User.bisFootHidden, True Float=0, False Float=1)
        6=1
        Index=$Particles.PartitionParticles.Partition
        Num=::NewEnumerator5
       )
       2=1
       3=1
       Index=&SpawnGroup()
       Num=::NewEnumerator2
      )
     )
    )
    Scale RGB=&VectorFromFloat(Value=&FloatFromCurve(FloatCurve=DataInterface() {propertyValues {Curve {keys [
          {interpMode=RCIM_Cubic, tangentMode=RCTM_User, time=0, value=0.5}
          <at 0.5 1 RCIM_Cubic>
          {arriveTangent=-0.1284, interpMode=RCIM_Cubic, leaveTangent=-0.1284, tangentMode=RCTM_User, time=1, value=0.5}
         ]}}}))
    ScaleA=true
    ScaleRGB=false
   )
   &Light_Attributes(
    Color=1
    Falloff Exponent=1
    Light Exponent=2
    Light Radius=80
    Position=1
    Radius=1
    Scale RGB:Vector3f=<4 4 4>
    ScaleA=true
    ScaleRGB=true
    UsePositionOffset=false
   )
  ]
  rendererClasses [
   &NiagaraSpriteRendererProperties
   &NiagaraMeshRendererProperties
   &NiagaraLightRendererProperties
  ]
  renderers [
   {rendererClass=&NiagaraSpriteRendererProperties}
   {rendererClass=&NiagaraMeshRendererProperties, rendererIndex=1}
   {rendererClass=&NiagaraSpriteRendererProperties, rendererIndex=2}
   {rendererClass=&NiagaraLightRendererProperties, rendererIndex=3}
  ]
 }
 IdleCubes {
  emitterUpdateScript [
   &EmitterState()
   &SpawnRate(Spawn Probability=0.8, SpawnRate=8, Use Spawn Probability=true)
  ]
  particleSpawnScript [
   &PartitionParticles(Number Of Partitions=9)
   &InitializeParticle(
    Color Maximum=<rgba 1 0.8541 0.8156 1>
    Color Minimum=<rgba 0.6094 0.5522 0.4975 1>
    Color Mode=::NewEnumerator2
    Lifetime=&RandomRangeFloat(Minimum=5, Maximum=7)
    Mesh Scale Mode=::NewEnumerator3
    Mesh Uniform Scale=&RandomRangeFloat(Minimum=0.4, Maximum=0.5)
    UsePositionOffset=false
    Write Material Random=true
   )
   /SetVariables() {enabled=false, isSetParameters=true}
   &SubUVAnimation(
    Mesh Renderer=DataInterface() {propertyValues {MeshRenderer=&self:IdleCubes_0.NiagaraMeshRendererProperties_0}}
    Output Alternate Renderer Binding=true
    Random Start Frame=true
    Renderer Type:ENiagaraMeshOrSprite=::NewEnumerator1
    SubUV Lookup Index Scale=0
    UseEndFrame=false
    UseStartFrame=false
   )
   &InitialMeshOrientation(Rotation Coordinate Space=::NewEnumerator1, Use Rotation Vector=true)
  ]
  particleUpdateScript [
   &ParticleState()
   &SkeletalMeshLocation(
    Bone / Socket Index=&SelectIntFromArray(
     Direct Array Index=$Particles.PartitionParticles.Partition
     Direct Array Mode=::NewEnumerator1
     Int Selection Array=DataInterface() {propertyValues {IntData [0,1,8,9,10,11,20]}}
    )
    Bone Sampling Mode=::NewEnumerator4
    MaskSpawnGroup=false
    Orientation Sampling=::NewEnumerator1
    Skeletal Mesh=$System.Skeletal Mesh Reader
   ) {enabled=false}
   &SkeletalMeshLocation(
    Bone / Socket Index=&SelectIntFromArray(
     Direct Array Index=$Particles.PartitionParticles.Partition
     Direct Array Mode=::NewEnumerator1
     Int Selection Array=DataInterface() {propertyValues {IntData [0,1,8,9,10,11,4,20,3]}}
    )
    Bone Sampling Mode=::NewEnumerator4
    MaskSpawnGroup=false
    Orientation Sampling=::NewEnumerator1
    Sampled Position Offset=&MergeHelper_VectorSwitch(
     0=<5 -3 0>
     1=<-10 0 0>
     2=<-20 0 0>
     3=<15 0 -3>
     4=<-20 0 3>
     5=<20 0 0>
     6=<20 0 0>
     7=<10 0 0>
     Index=$Particles.PartitionParticles.Partition
     Num=::NewEnumerator6
    )
    Skeletal Mesh=$System.Skeletal Mesh Reader
   ) {module=SkeletalMeshLocation001}
   &ShapeLocation(
    Additional Quaternion Rotation=true
    Additional Yaw / Pitch / Roll=<0 0.25 0>
    Cylinder Height=&MergeHelper_FloatSwitch(0=10, 1=10, 2=25, 3=30, 4=30, 5=30, 6=15, 7=30, 8=5, Index=$Particles.PartitionParticles.Partition, Num=::NewEnumerator7)
    Cylinder Radius=&MergeHelper_FloatSwitch(0=2, 1=4, 2=5, 3=5, 4=1, 5=1, 6=5, 7=10, 8=10, Index=$Particles.PartitionParticles.Partition, Num=::NewEnumerator7)
    MaskSpawnGroup=false
    Override Seed=false
    Rotation Mode=::NewEnumerator3
    Rotation Quaternion=$Particles.SkeletalMeshLocation001.SampledRotation
    Shape Origin=$StackContext.Position
    Shape Primitive=::NewEnumerator1
    Surface Only=true
    Use Endcaps In Surface Only Mode=false
    UseOwnerScale=true
   )
   &ScaleMeshSize(Scale Factor=&VectorFromCurve(
     VectorCurve=DataInterface() {propertyValues {
       XCurve {keys [<at 0 1>,<at 0.0821 0.9489>,<at 0.1617 1.1754>,<at 0.8653 1.1811>,<at 0.9758 1.3967>]}
       YCurve {keys [<at 0 1>,<at 0.1192 1.3385>,<at 0.1592 1.1826>,<at 0.9654 1.1803>]}
       ZCurve {keys [<at 0 0.577>,<at 0.0503 1.3231>,<at 0.1202 1.0349>,<at 0.1592 1.1826>,<at 0.9862 1.2393>]}
      }}
     Scale Curve=&Multiply_VectorByFloat(
      Vector=&RandomRangeVector(Minimum=<0.75 0.75 0.75>, Maximum=<1.25 1.25 1.25>)
      Float=&FloatFromCurve(
       FloatCurve=DataInterface() {propertyValues {Curve {keys [
           {interpMode=RCIM_Cubic, tangentMode=RCTM_User, time=0, value=0.25}
           {arriveTangentWeight=0.0895, interpMode=RCIM_Cubic, leaveTangentWeight=0.2109, tangentMode=RCTM_User, time=0.1037, value=1}
           {interpMode=RCIM_Cubic, tangentMode=RCTM_User, time=0.8499, value=1}
           {arriveTangent=-4.1587, interpMode=RCIM_Cubic, leaveTangent=-4.1587, tangentMode=RCTM_User, time=1, value=-0.0118}
          ]}}}
       Scale Curve=&MergeHelper_FloatSwitch(
        0=&MakeCustomFloatFromBool(Bool=$User.bisHandHidden, True Float=0, False Float=1)
        1=&MakeCustomFloatFromBool(Bool=$User.bisHandHidden, True Float=0, False Float=1)
        2=1
        3=1
        4=&MakeCustomFloatFromBool(Bool=$User.bisFootHidden, True Float=0, False Float=1)
        5=&MakeCustomFloatFromBool(Bool=$User.bisFootHidden, True Float=0, False Float=1)
        6=1
        7=1
        8=&MakeCustomFloatFromBool(Bool=$User.bIsADS, True Float=0, False Float=1)
        Index=$Particles.PartitionParticles.Partition
        Num=::NewEnumerator7
       )
      )
     )
    ))
   &DynamicMaterialParameters(
    Index 0 Param 1=&FloatFromCurve(FloatCurve=DataInterface() {propertyValues {Curve {keys [<at 0 0>,<at 1 4>]}}}, Scale Curve=2)
    Index 0 Param 2=1
    Index 0 Param 3=&FloatFromCurve(
     FloatCurve=DataInterface() {propertyValues {Curve {keys [{interpMode=RCIM_Cubic, tangentMode=RCTM_User, time=0, value=2}, {arriveTangent=-2.5233, tangentMode=RCTM_User, time=1.0059, value=1}]}}}
     Scale Curve=&Waveform(
      "Waveform [1]"=::NewEnumerator9
      "[1] Input Position"=&RandomRangeFloat(Minimum=0, Maximum=1)
      "[1] Frequency"=3
      "[1] Amplitude Min/Max"=<0.75 1>
     )
    )
    Module.Index0Float0Write=true
    Module.Index0Float1Write=true
    Module.Index0Float2Write=true
    Module.Index0Float3Write=true
   )
   /SetVariables(Particles.Position=&AddVectorToPosition(
     Position=$Particles.Position
     Vector=&VectorFromCurve(
      VectorCurve=DataInterface() {propertyValues {
        XCurve {keys [
          {arriveTangent=3.0078, interpMode=RCIM_Cubic, leaveTangent=3.0078, tangentMode=RCTM_User, time=0, value=0}
          {interpMode=RCIM_Cubic, tangentMode=RCTM_User, time=0.1471, value=0.7929}
          {arriveTangent=1.6833, interpMode=RCIM_Cubic, leaveTangent=1.6833, time=0.3939, value=0.5348}
          {arriveTangent=1.5617, interpMode=RCIM_Cubic, leaveTangent=1.5617, time=0.6089, value=1.5703}
          {arriveTangent=1.0987, interpMode=RCIM_Cubic, leaveTangent=1.0987, time=0.8159, value=1.1937}
          {arriveTangent=-3.9312, interpMode=RCIM_Cubic, leaveTangent=-3.9312, tangentMode=RCTM_User, time=1, value=2}
         ]}
        YCurve {keys [
          {arriveTangent=3.0078, interpMode=RCIM_Cubic, leaveTangent=3.0078, tangentMode=RCTM_User, time=0, value=0}
          {interpMode=RCIM_Cubic, tangentMode=RCTM_User, time=0.1471, value=0.7929}
          {arriveTangent=1.693, interpMode=RCIM_Cubic, leaveTangent=1.693, time=0.3966, value=0.5536}
          {arriveTangent=1.5716, interpMode=RCIM_Cubic, leaveTangent=1.5716, time=0.6062, value=1.5703}
          {arriveTangent=1.0913, interpMode=RCIM_Cubic, leaveTangent=1.0913, time=0.8158, value=1.2126}
          {arriveTangent=-3.9312, interpMode=RCIM_Cubic, leaveTangent=-3.9312, tangentMode=RCTM_User, time=1, value=2}
         ]}
        ZCurve {keys [
          {arriveTangent=3.0078, interpMode=RCIM_Cubic, leaveTangent=3.0078, tangentMode=RCTM_User, time=0, value=0}
          {interpMode=RCIM_Cubic, tangentMode=RCTM_User, time=0.1471, value=0.7929}
          {arriveTangent=1.693, interpMode=RCIM_Cubic, leaveTangent=1.693, time=0.3966, value=0.5348}
          {arriveTangent=1.6165, interpMode=RCIM_Cubic, leaveTangent=1.6165, time=0.6062, value=1.5703}
          {arriveTangent=1.0913, interpMode=RCIM_Cubic, leaveTangent=1.0913, time=0.8158, value=1.2126}
          {arriveTangent=-3.9312, interpMode=RCIM_Cubic, leaveTangent=-3.9312, tangentMode=RCTM_User, time=1, value=2}
         ]}
       }}
      Scale Curve=&Multiply_VectorByFloat(Vector=&RandomRangeVector(Minimum=<-1 -1 -1>, Maximum=<1 1 1>), Float=3)
     )
    )) {isSetParameters=true}
   &CameraOffset(Camera Offset Amount=10) {enabled=false}
  ]
  rendererClasses [&NiagaraSpriteRendererProperties, &NiagaraMeshRendererProperties]
  renderers [
   {rendererClass=&NiagaraSpriteRendererProperties}
   {rendererClass=&NiagaraMeshRendererProperties, rendererIndex=1}
  ]
 }
 Long_Lightning001 {
  emitterUpdateScript [
   &EmitterState(Life Cycle Mode=::NewEnumerator1, Loop Duration=1.2, UseLoopDelay=false)
   /SetVariables(Emitter.Random:NiagaraInt32=&SelectIntFromArray(
     Array Sampling Mode=::NewEnumerator1
     Direct Array Index=?
     Recalculate Random Each Loop=true
     Int Selection Array=DataInterface() {propertyValues {IntData [0,1]}}
     Random Seed=0
    )) {isSetParameters=true}
   &SpawnBurst_Instantaneous(Spawn Count=4, Spawn Probability=0.6, Use Spawn Probability=true, UseLoopCountLimit=false)
  ]
  particleSpawnScript [
   &PartitionParticles(Number Of Partitions=4, Particles Per Partition=1, Partition Mode:ENiagara_PartitionMode_v2=::NewEnumerator1)
   &InitializeParticle(
    Color=<rgba 20 2.811 0 1>
    Lifetime=0.6
    Mesh Scale Mode=::NewEnumerator3
    Mesh Uniform Scale=2
    Ribbon Width=10
    Ribbon Width Mode=::NewEnumerator1
    UsePositionOffset=false
    Write Material Random=true
   )
  ]
  particleUpdateScript [
   &ParticleState()
   &DynamicMaterialParameters(
    Index 0 Param 1=&FloatFromCurve(FloatCurve=DataInterface() {propertyValues {Curve {
        defaultValue=1
        keys [{interpMode=RCIM_Cubic, tangentMode=RCTM_User, time=0, value=0}, {interpMode=RCIM_Cubic, tangentMode=RCTM_User, time=1, value=1}]
       }}})
    Index 0 Param 3=&FloatFromCurve(FloatCurve=DataInterface() {propertyValues {Curve {
        defaultValue=1
        keys [{interpMode=RCIM_Cubic, tangentMode=RCTM_User, time=0.4057, value=0}, {interpMode=RCIM_Cubic, tangentMode=RCTM_User, time=1, value=1}]
       }}})
    Index 0 Param 4=1
    Index 1 Param 1=1
    Index 1 Param 2=&FloatFromCurve(
     FloatCurve=DataInterface() {propertyValues {Curve {defaultValue=1, keys [{tangentMode=RCTM_User, time=0, value=0}, {tangentMode=RCTM_User, time=1, value=1}]}}}
     Scale Curve=0.8
    )
    Index 1 Param 3=1
    Module.Index0Float0Write=true
    Module.Index0Float1Write=true
    Module.Index0Float2Write=true
    Module.Index0Float3Write=true
    Module.Index1Float0Write=true
    Module.Index1Float1Write=true
    Module.Index1Float2Write=true
    Module.Index1Float3Write=false
    Write Parameter Index 1=::NewEnumerator1
   )
   /Game/Effects/Niagara/Modules/Curves/FN_ListCurve_SkeletalMeshSampler(
    Apply Position=true
    Curve Index:NiagaraFloat=&RemapRange(
     Input Value To Remap=&MakeFloatFromInt(Int=$Particles.PartitionParticles.Partition)
     Input Min=0
     Input Max=5
     Output Min=0
     Output Max=1
    )
    Index_01:NiagaraInt32=&MergeHelper_IntSwitch(0=4, 1=1, 2=?, 3=?, 4=?, 5=?, Index=$Emitter.Random, Num=::NewEnumerator0)
    Index_02:NiagaraInt32=&MergeHelper_IntSwitch(0=3, 1=5, 2=?, 3=?, 4=?, 5=?, Index=$Emitter.Random, Num=::NewEnumerator0)
    Index_03:NiagaraInt32=&MergeHelper_IntSwitch(0=2, 1=4, 2=?, 3=?, 4=?, 5=?, Index=$Emitter.Random, Num=::NewEnumerator0)
    Index_04:NiagaraInt32=&MergeHelper_IntSwitch(0=17, 1=3, 2=?, 3=?, 4=?, 5=?, Index=$Emitter.Random, Num=::NewEnumerator0)
    Index_05:NiagaraInt32=&MergeHelper_IntSwitch(0=11, 1=10, 2=?, 3=?, 4=?, 5=?, Index=$Emitter.Random, Num=::NewEnumerator0)
    Number of Inputs=3
    Offset Coordinate Space=2
    Offset_01:Vector3f=&MergeHelper_VectorSwitch(0=<0 0 12>, 1=<10 10 0>, 2=?, 3=?, Index=$Emitter.Random, Num=::NewEnumerator0)
    Offset_02:Vector3f=&MergeHelper_VectorSwitch(0=<0 -20 12>, 1=<0 15 10>, 2=?, 3=?, Index=$Emitter.Random, Num=::NewEnumerator0)
    Offset_03:Vector3f=&MergeHelper_VectorSwitch(0=<0 -10 -20>, 1=<0 15 -20>, 2=?, 3=?, Index=$Emitter.Random, Num=::NewEnumerator0)
    Offset_04:Vector3f=&MergeHelper_VectorSwitch(0=<-35 0 0>, 1=<0 -10 -40>, 2=?, 3=?, Index=$Emitter.Random, Num=::NewEnumerator0)
    Offset_05:Vector3f=&MergeHelper_VectorSwitch(0=<80 50 -40>, 1=<50 150 -100>, 2=?, 3=?, Index=$Emitter.Random, Num=::NewEnumerator0)
    Skeletal Mesh:NiagaraDataInterfaceSkeletalMesh=$System.Skeletal Mesh Reader
   ) {enabled=false, module=FN_ListCurve_SkeletalMeshSampler002}
   &FN_ListCurve_SkeletalMeshSampler(
    Apply Position=true
    Curve Index:NiagaraFloat=&RemapRange(
     Input Value To Remap=&MakeFloatFromInt(Int=$Particles.PartitionParticles.Partition)
     Input Min=0
     Input Max=4
     Output Min=0
     Output Max=1
    )
    Index_01:NiagaraInt32=&MergeHelper_IntSwitch(0=1, 1=4, 2=?, 3=?, 4=?, 5=?, Index=$Emitter.Random, Num=::NewEnumerator0)
    Index_02:NiagaraInt32=&MergeHelper_IntSwitch(0=5, 1=3, 2=?, 3=?, 4=?, 5=?, Index=$Emitter.Random, Num=::NewEnumerator0)
    Index_03=2
    Index_04:NiagaraInt32=&MergeHelper_IntSwitch(0=4, 1=20, 2=?, 3=?, 4=?, 5=?, Index=$Emitter.Random, Num=::NewEnumerator0)
    Offset Coordinate Space=2
    Offset_01:Vector3f=&MergeHelper_VectorSwitch(0=<2 0 0>, 1=<0 -5 20>, 2=?, 3=?, Index=$Emitter.Random, Num=::NewEnumerator0)
    Offset_02:Vector3f=&MergeHelper_VectorSwitch(0=<15 10 20>, 1=<10 15 0>, 2=?, 3=?, Index=$Emitter.Random, Num=::NewEnumerator0)
    Offset_03:Vector3f=&MergeHelper_VectorSwitch(0=<-30 10 -30>, 1=<0 15 -10>, 2=?, 3=?, Index=$Emitter.Random, Num=::NewEnumerator0)
    Offset_04:Vector3f=&MergeHelper_VectorSwitch(0=<80 -50 20>, 1=<-50 -10 -10>, 2=?, 3=?, Index=$Emitter.Random, Num=::NewEnumerator0)
    Skeletal Mesh:NiagaraDataInterfaceSkeletalMesh=$System.Skeletal Mesh Reader
   )
   &FN_ListCurve_SkeletalMeshSampler(
    Apply Position=true
    Curve Index:NiagaraFloat=&RemapRange(
     Input Value To Remap=&MakeFloatFromInt(Int=$Particles.PartitionParticles.Partition)
     Input Min=0
     Input Max=4
     Output Min=0
     Output Max=1
    )
    Index_01=4
    Index_02=3
    Index_03=2
    Index_04=20
    Offset Coordinate Space=2
    Offset_01:Vector3f=<0 -5 20>
    Offset_02:Vector3f=<10 15 0>
    Offset_03:Vector3f=<0 15 -10>
    Offset_04:Vector3f=<-50 -10 -10>
    Skeletal Mesh:NiagaraDataInterfaceSkeletalMesh=$System.Skeletal Mesh Reader
   ) {enabled=false, module=FN_ListCurve_SkeletalMeshSampler001}
   &ScaleColor(
    Scale Alpha=&FloatFromCurve(
     FloatCurve=DataInterface() {propertyValues {Curve {
        defaultValue=0
        keys [{interpMode=RCIM_Cubic, tangentMode=RCTM_User, time=0, value=1}, {interpMode=RCIM_Cubic, tangentMode=RCTM_User, time=1, value=0}]
       }}}
     CurveIndex=$System.Timeline.NormalizedTime
    )
    ScaleA=true
    ScaleRGB=true
   )
  ]
  rendererClasses [
   &NiagaraSpriteRendererProperties
   &NiagaraMeshRendererProperties
   &NiagaraRibbonRendererProperties
  ]
  renderers [
   {rendererClass=&NiagaraSpriteRendererProperties}
   {rendererClass=&NiagaraMeshRendererProperties, rendererIndex=1}
   {rendererClass=&NiagaraRibbonRendererProperties, rendererIndex=2}
  ]
 }
 MergedSplineLightning {
  emitterUpdateScript [
   &EmitterState(
    Inactive Response=::NewEnumerator1
    Life Cycle Mode=::NewEnumerator1
    Loop Behavior=::NewEnumerator1
    Loop Duration Mode=::NewEnumerator1
    UseLoopDelay=false
   )
   &SpawnRate(Spawn Probability=0.3, SpawnRate=10, Use Spawn Probability=true)
   &SpawnRate(Spawn Probability=0.3, SpawnGroup=1, SpawnRate=10, Use Spawn Probability=true) {module=SpawnRate001}
   &SpawnRate(Spawn Probability=0.3, SpawnGroup=2, SpawnRate=20, Use Spawn Probability=true) {module=SpawnRate002}
   &SpawnRate(Spawn Probability=0.3, SpawnGroup=3, SpawnRate=10, Use Spawn Probability=true) {module=SpawnRate003}
  ]
  particleSpawnScript [
   /SetVariables(Particles.VisibilityTag=&SpawnGroup()) {isSetParameters=true}
   &PartitionParticles()
   /SetVariables(Particles.Random:NiagaraInt32=&RandomRangeInt(Minimum=0, Maximum=1)) {isSetParameters=true}
   &InitializeParticle(
    Lifetime=&MergeHelper_FloatSwitch(
     0=&RandomRangeFloat(Minimum=0.3, Maximum=0.5)
     1=&RandomRangeFloat(Minimum=0.3, Maximum=0.4)
     2=&RandomRangeFloat(Minimum=0.2, Maximum=0.4)
     3=&RandomRangeFloat(Minimum=0.3, Maximum=0.5)
     Index=&SpawnGroup()
     Num=::NewEnumerator2
    )
    UsePositionOffset=false
    Write Material Random=true
   ) {module=InitializeParticle001}
   /SetVariables(
    Particles.InvertLightning:NiagaraBool=&RandomBool()
    Particles.LightningIndex=1
    Particles.MirrorLightning:NiagaraBool=&RandomBool()
   ) {isSetParameters=true}
   /SetVariables(Particles.LightningIndex:NiagaraInt32=&MergeHelper_IntSwitch(
     0=&SelectIntFromArray(
      Direct Array Index=&ReturnParticleIDAttributes()
      Direct Array Mode=::NewEnumerator1
      Int Selection Array=DataInterface() {propertyValues {IntData [0,1,2,3,2,3,1,0,0,2]}}
     )
     1=&SelectIntFromArray(
      Direct Array Index=&ReturnParticleIDAttributes()
      Direct Array Mode=::NewEnumerator1
      Int Selection Array=DataInterface() {propertyValues {IntData [0,1,2,3,2,3,1,0,0,2]}}
     )
     2=&SelectIntFromArray(
      Direct Array Index=&ReturnParticleIDAttributes()
      Direct Array Mode=::NewEnumerator1
      Int Selection Array=DataInterface() {propertyValues {IntData [0,1,2,3,2,3,1,0,0,2]}}
     )
     3=&SelectIntFromArray(
      Direct Array Index=$Particles.Random
      Direct Array Mode=::NewEnumerator1
      Int Selection Array=DataInterface() {propertyValues {IntData [0,1,2,3]}}
     )
     4=?
     5=?
     Index=&SpawnGroup()
     Num=::NewEnumerator2
    )) {isSetParameters=true}
   /SetVariables(Particles.Scale=&MergeHelper_VectorSwitch(
     0=&Multiply_Vector(
      A=&MergeHelper_VectorSwitch(0=<1 1 1>, 1=<-1 1 1>, 2=?, 3=?, Index=$Particles.Random, Num=::NewEnumerator0)
      B=&RandomRangeVector(Minimum=<1 1 1>, Maximum=<0.8 1.3 1.5>)
     )
     1=&Multiply_Vector(
      A=&MergeHelper_VectorSwitch(0=<1 1 1>, 1=<-1 1 1>, 2=?, 3=?, Index=$Particles.Random, Num=::NewEnumerator0)
      B=&RandomRangeVector(Minimum=<1 1 1>, Maximum=<0.8 1.3 1.5>)
     )
     2=&MergeHelper_VectorSwitch(0=<1 1 1>, 1=<1 1 -1>, 2=?, 3=?, Index=&RandomRangeInt(Minimum=0, Maximum=1), Num=::NewEnumerator0)
     3=&MergeHelper_VectorSwitch(
      0=&MergeHelper_VectorSwitch(0=<1 1 1>, 1=<1 1 -1>, 2=?, 3=?, Index=&RandomRangeInt(Minimum=0, Maximum=1), Num=::NewEnumerator0)
      1=&MergeHelper_VectorSwitch(0=<1 1 1>, 1=<1 1 -1>, 2=?, 3=?, Index=&RandomRangeInt(Minimum=0, Maximum=1), Num=::NewEnumerator0)
      2=&MergeHelper_VectorSwitch(0=<1 1 1>, 1=<1 1 -1>, 2=?, 3=?, Index=&RandomRangeInt(Minimum=0, Maximum=1), Num=::NewEnumerator0)
      3=&MergeHelper_VectorSwitch(0=<1 1 1>, 1=<1 1 -1>, 2=?, 3=?, Index=&RandomRangeInt(Minimum=0, Maximum=1), Num=::NewEnumerator0)
      Index=$Particles.Random
      Num=::NewEnumerator2
     )
     Index=&SpawnGroup()
     Num=::NewEnumerator2
    )) {isSetParameters=true}
   &InitialMeshOrientation(
    Mesh Orientation Mode=::NewEnumerator7
    Rotation=&MergeHelper_VectorSwitch(0=<0 0 0>, 1=<0.5 0 0>, 2=?, 3=?, Index=$Particles.Random, Num=::NewEnumerator0)
    Use Rotation Vector=true
   )
   &InitialMeshOrientation(
    Mesh Orientation Mode=::NewEnumerator7
    Rotation=&RandomRangeVector(Minimum=<0 0 0>, Maximum=<1 0 0>)
    Use Rotation Vector=true
   ) {module=InitialMeshOrientation001}
   &CameraOffset(Camera Offset Amount=&MergeHelper_FloatSwitch(0=0, 1=0, 2=0, 3=5, Index=&SpawnGroup(), Num=::NewEnumerator2))
   &ShapeLocation(
    MaskSpawnGroup=false
    Override Seed=false
    Sphere Radius=&MergeHelper_FloatSwitch(0=3, 1=3, 2=5, 3=0, Index=&SpawnGroup(), Num=::NewEnumerator2)
    UseOwnerScale=true
   )
  ]
  particleUpdateScript [
   &ParticleState()
   &SkeletalMeshLocation(
    Bone / Socket Index=&MergeHelper_IntSwitch(
     0=&MergeHelper_IntSwitch(0=0, 1=1, 2=?, 3=?, 4=?, 5=?, Index=$Particles.Random, Num=::NewEnumerator0)
     1=&MergeHelper_IntSwitch(0=4, 1=5, 2=?, 3=?, 4=?, 5=?, Index=$Particles.Random, Num=::NewEnumerator0)
     2=2
     3=&MergeHelper_IntSwitch(0=8, 1=11, 2=9, 3=10, 4=?, 5=?, Index=$Particles.Random, Num=::NewEnumerator2)
     4=?
     5=?
     Index=&SpawnGroup()
     Num=::NewEnumerator2
    )
    Bone Sampling Mode=::NewEnumerator4
    MaskSpawnGroup=false
    Sampled Position Offset=&MergeHelper_VectorSwitch(
     0=Linked() {linkedVariable {name=Particles.Initial.ShapeLocation.ShapePosition, type=NiagaraPosition}}
     1=Linked() {linkedVariable {name=Particles.Initial.ShapeLocation.ShapePosition, type=NiagaraPosition}}
     2=Linked() {linkedVariable {name=Particles.Initial.ShapeLocation.ShapePosition, type=NiagaraPosition}}
     3=&AddVectorToPosition(
      Position=$Particles.Initial.ShapeLocation.ShapePosition
      Vector=&RandomRangeVector(Minimum=<0 0 0>, Maximum=<10 0 3>)
     )
     Index=&SpawnGroup()
     Num=::NewEnumerator2
    )
    Skeletal Mesh=$System.Skeletal Mesh Reader
   )
   &DynamicMaterialParameters(
    Index 0 Param 1=&Multiply_Float(
     A=&MergeHelper_FloatSwitch(0=5, 1=5, 2=5, 3=10, Index=&SpawnGroup(), Num=::NewEnumerator2)
     B=&FloatFromCurve(FloatCurve=DataInterface() {propertyValues {Curve {keys [{interpMode=RCIM_Cubic, leaveTangent=8.1085, tangentMode=RCTM_Break, time=0.8568, value=1}, {tangentMode=RCTM_User, time=1, value=0}]}}})
    )
    Index 0 Param 2=&Divide_Float(A=&RandomRangeFloat(Minimum=1, Maximum=1.5), B=1.5)
    Index 0 Param 3=&Lerp_Float(
     A=&RandomRangeFloat(Minimum=0, Maximum=0.3)
     B=1
     Alpha=&FloatFromCurve(FloatCurve=DataInterface() {propertyValues {Curve {keys [<at 0 1>,<at 0.1 0.375>,<at 0.5 0.05>,<at 1 0>]}}})
    )
    Index 0 Param 4=&MergeHelper_FloatSwitch(
     0=0
     1=0
     2=&MergeHelper_FloatSwitch(0=1, 1=1, 2=1, 3=0, 4=1, Index=$Particles.LightningIndex, Num=::NewEnumerator3)
     3=&MergeHelper_FloatSwitch(0=1, 1=1, 2=1, 3=0, 4=1, Index=$Particles.LightningIndex, Num=::NewEnumerator3)
     Index=&SpawnGroup()
     Num=::NewEnumerator2
    )
    Index 1 Param 1=&MakeFloatFromVector(Vector=$Particles.Scale, Channel=::NewEnumerator0)
    Index 1 Param 2=&MakeFloatFromVector(Vector=$Particles.Scale, Channel=::NewEnumerator1)
    Index 1 Param 3=&MakeFloatFromVector(Vector=$Particles.Scale, Channel=::NewEnumerator2)
    Index 2 Param 1=&Multiply_Float(
     A=&FloatFromCurve(FloatCurve=DataInterface() {propertyValues {Curve {keys [<at 0 0.5>,<at 1 1>]}}})
     B=&RandomRangeFloat(Minimum=3, Maximum=4)
    )
    Index 2 Param 2=&FloatFromCurve(
     FloatCurve=DataInterface() {propertyValues {Curve {keys [{interpMode=RCIM_Cubic, tangentMode=RCTM_Break, time=0.1, value=1}, <at 1 0>]}}}
     Scale Curve=&RandomRangeFloat(Minimum=1, Maximum=1)
    )
    Module.Index0Param0Write=true
    Module.Index0Param1Write=true
    Module.Index0Param2Write=true
    Module.Index0Param3Write=true
    Module.Index1Param0Write=true
    Module.Index1Param1Write=true
    Module.Index1Param2Write=true
    Module.Index1Param3Write=true
    Module.Index2Param0Write=true
    Module.Index2Param1Write=true
    Module.Index2Param2Write=true
    Module.Index2Param3Write=true
    Write Parameter Index 1=true
    Write Parameter Index 2=true
   ) {module=DynamicMaterialParameters001}
   &ScaleColor(
    Scale Alpha=&FloatFromCurve(
     FloatCurve=DataInterface() {propertyValues {Curve {keys [
         {tangentMode=RCTM_User, time=0, value=1}
         {arriveTangent=-0.0405, interpMode=RCIM_Cubic, leaveTangent=-0.0405, tangentMode=RCTM_User, time=0.7009, value=1}
         {arriveTangent=-8.0089, interpMode=RCIM_Cubic, leaveTangent=-8.0089, tangentMode=RCTM_User, time=1, value=0}
        ]}}}
     Scale Curve=&MergeHelper_FloatSwitch(
      0=&MakeCustomFloatFromBool(Bool=$User.bisHandHidden, True Float=0, False Float=1)
      1=&MakeCustomFloatFromBool(Bool=$User.bIsADS, True Float=0, False Float=1)
      2=1
      3=&MakeCustomFloatFromBool(Bool=$User.bisFootHidden, True Float=0, False Float=1)
      Index=&SpawnGroup()
      Num=::NewEnumerator2
     )
    )
    Scale RGB=&Lerp_Vector(
     A=&VectorFromFloat(Value=10)
     B=&VectorFromFloat(Value=&RandomRangeFloat(Minimum=0.8, Maximum=1, Evaluation Type=::NewEnumerator1))
     Alpha=&FloatFromCurve(FloatCurve=DataInterface() {propertyValues {Curve {keys [
          {interpMode=RCIM_Cubic, leaveTangent=2.5294, tangentMode=RCTM_Break, time=0.1, value=0}
          {arriveTangent=-0.2166, interpMode=RCIM_Cubic, tangentMode=RCTM_Break, time=0.4, value=1}
         ]}}})
    )
    ScaleA=true
    ScaleRGB=true
   ) {module=ScaleColor001}
   /SetVariables(Particles.GlowColor=&MultiplyLinearColor(
     Linear Color=<rgba 1 1 1 1>
     Scale Factor=&MakeVector4FromVectorAndFloat(
      "Vector (RGB)"=<1 1 1>
      "Float (Alpha)"=&Multiply_Float(
       A=0.2
       B=&FloatFromCurve(
        FloatCurve=DataInterface() {propertyValues {Curve {keys [
            {tangentMode=RCTM_User, time=0, value=1}
            {arriveTangent=-0.0405, interpMode=RCIM_Cubic, leaveTangent=-0.0405, tangentMode=RCTM_User, time=0.7009, value=1}
            {arriveTangent=-8.0089, interpMode=RCIM_Cubic, leaveTangent=-8.0089, tangentMode=RCTM_User, time=1, value=0}
           ]}}}
        Scale Curve=&MergeHelper_FloatSwitch(
         0=&MakeCustomFloatFromBool(Bool=$User.bisHandHidden, True Float=0, False Float=1)
         1=&MakeCustomFloatFromBool(Bool=$User.bIsADS, True Float=0, False Float=1)
         2=1
         3=&MakeCustomFloatFromBool(Bool=$User.bisFootHidden, True Float=0, False Float=1)
         Index=&SpawnGroup()
         Num=::NewEnumerator2
        )
       )
      )
     )
    )) {isSetParameters=true}
  ]
  rendererClasses [&NiagaraMeshRendererProperties]
  renderers [
   {rendererClass=&NiagaraMeshRendererProperties, rendererIndex=0}
   {rendererClass=&NiagaraMeshRendererProperties, rendererIndex=1}
   {rendererClass=&NiagaraMeshRendererProperties, rendererIndex=2}
   {rendererClass=&NiagaraMeshRendererProperties, rendererIndex=3}
   {rendererClass=&NiagaraMeshRendererProperties, rendererIndex=4}
   {rendererClass=&NiagaraMeshRendererProperties, rendererIndex=5}
   {rendererClass=&NiagaraMeshRendererProperties, rendererIndex=6}
   {rendererClass=&NiagaraMeshRendererProperties, rendererIndex=7}
  ]
 }
}
summary {
 systemName=NS_Character_VoyagerCuboid_Body
 emitters [
  {
   emitterName=Glows
   rendererClasses [
    &NiagaraSpriteRendererProperties
    &NiagaraMeshRendererProperties
    &NiagaraLightRendererProperties
   ]
  }
  {emitterName=CubicFlame, rendererClasses [&NiagaraSpriteRendererProperties]}
  {emitterName=IdleCubes, rendererClasses [&NiagaraSpriteRendererProperties, &NiagaraMeshRendererProperties]}
  {
   emitterName=Long_Lightning001
   rendererClasses [
    &NiagaraSpriteRendererProperties
    &NiagaraMeshRendererProperties
    &NiagaraRibbonRendererProperties
   ]
  }
  {emitterName=MergedSplineLightning, rendererClasses [&NiagaraMeshRendererProperties]}
 ]
 userVariables [
  User.bIsADS {type=NiagaraBool}
  User.bisFootHidden {type=NiagaraBool}
  User.bIsFrontEnd {type=NiagaraBool}
  User.bIsFrontEndPreview {type=NiagaraBool}
  User.bisHandHidden {type=NiagaraBool}
  User.bisHeadHidden {type=NiagaraBool}
 ]
}
