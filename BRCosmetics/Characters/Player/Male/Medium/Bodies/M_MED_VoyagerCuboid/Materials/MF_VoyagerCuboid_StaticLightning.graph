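# Material function for the VoyagerCuboid character body that adds animated static lightning to incoming material attributes, setting Emissive, Opacity and World Position Offset.
# Two lightning layers and a hot-spot layer sample UVs baked into a texture, with a size-mapping texture blending tiling, and get three-color glow gradients.
# Panned noise distortion, random-mosaic noise flicker and a "choppy time" blend of quantized and irregular time animate it.
# WPO pushes vertices along normals.
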MaterialFunction /BRCosmetics/Characters/Player/Male/Medium/Bodies/M_MED_VoyagerCuboid/Materials/MF_VoyagerCuboid_StaticLightning
graph {
 properties {previewBlendMode=BLEND_Translucent}
 nodes [
  n1 = NamedRerouteUsage()
  n2 = NamedRerouteUsage()
  n3 = LinearInterpolate(A=NamedRerouteUsage(), B=Power(Base=Multiply(A=@n2) {constB=4}) {constExponent=4}, Alpha=Saturate(None=NamedRerouteUsage()))
  n4 = NamedRerouteUsage()
  n5 = TextureSampleParameter2D!(UVs, Bias, Apply View MipBias) {
   mipValueMode=TMVM_MipBias
   parameterName=MaskForLightning
   samplerType=SAMPLERTYPE_LinearColor
   texture=/BRCosmetics/Characters/Player/Male/Medium/Bodies/M_MED_VoyagerCuboid/Textures/T_VoyagerCuboid_StaticLightningMask
  }
  n6 = Multiply(A=Multiply(A=Add(A=@n3, B=@n4), B=@n5.R), B=Multiply(A=NamedRerouteUsage()))
  n7 = ScalarParameter() {defaultValue=4, parameterName=EmissiveLightningStrength}
  n8 = NamedRerouteUsage()
  n9 = Multiply(A=NamedRerouteUsage(), B=Power(Base=NamedRerouteUsage()) {constExponent=1}) {constB=2}
  n10 = Saturate(None=@n9)
  n11 = Multiply(A=Add(A=LinearInterpolate(A=@n8, B=@n9, Alpha=@n10), B=NamedRerouteUsage()), B=Multiply(A=NamedRerouteUsage()))
  FunctionOutput(None=SetMaterialAttributes!(
    MaterialAttributes=GetMaterialAttributes(None=@n1) {attributeGetTypes [#1]}.MaterialAttributes
    EmissiveColor=Multiply(A=@n6, B=@n7)
    Opacity=Saturate(None=@n11)
    WorldPositionOffset=NamedRerouteUsage()
   ) {attributeSetTypes [#2, #3, #4]}) {outputName=Out_MaterialAttributes}
  Multiply(A=NamedRerouteUsage(), B=Constant2Vector() {g=0.5, r=1})
  Multiply(A=NamedRerouteUsage(), B=Constant2Vector() {g=1, r=1})
  n12 = TextureCoordinate()
  n13 = VectorParameter() {parameterName=UV Noise Scale and Speed Texture, defaultValue=<rgba 0.005 0.05 0.1 0.3>}
  n14 = ComponentMask(None=@n13.RGB) {g=true, r=true}
  n15 = TextureSample(UVs=Panner(Coordinate=Multiply(A=@n12, B=@n14), Time=NamedRerouteUsage(), Speed=AppendVector(A=@n13.B, B=@n13.B)) {bFractionalPart=true, speedY=1}) {samplerType=SAMPLERTYPE_LinearColor, texture=&HighContrastNoise}
  NamedRerouteDeclaration(None=Multiply(
    A=ConstantBiasScale(None=@n15.RGB) {bias=-0.5, scale=2}
    B=ScalarParameter() {defaultValue=0.1, parameterName="UV  Distortion Intensity texture"}
   )) {name=UVDistortionTexture, nodeColor=<rgba 0 0.5547 1 1>}
  n16 = Multiply(A=NamedRerouteUsage(), B=Constant2Vector() {g=0.25, r=1})
  n17 = LinearInterpolate(A=@n16, B=@n16, Alpha=ComponentMask(None=NamedRerouteUsage()) {g=true})
  n18 = Multiply(A=NamedRerouteUsage(), B=Constant2Vector() {g=0.5, r=1})
  n19 = ComponentMask(None=NamedRerouteUsage()) {b=true}
  n20 = NamedRerouteUsage()
  n21 = Constant2Vector() {g=1, r=1}
  n22 = NamedRerouteUsage()
  n23 = TextureObjectParameter() {
   parameterName=TextureHotSpots
   samplerType=SAMPLERTYPE_LinearColor
   texture=/BRCosmetics/Characters/Player/Male/Medium/Bodies/M_MED_VoyagerCuboid/Textures/T_VoyagerCuboid_HotSpots
  }
  n24 = TextureSample!(
   UVs=LinearInterpolate(A=LinearInterpolate(A=@n17, B=@n18, Alpha=@n19), B=Multiply(A=@n20, B=@n21), Alpha=ComponentMask(None=@n22) {a=true})
   Tex=@n23
   Level
   Apply View MipBias
  ) {
   mipValueMode=TMVM_MipLevel
   samplerSource=SSM_Wrap_WorldGroupSettings
   samplerType=SAMPLERTYPE_LinearColor
   texture=/Game/Effects/Fort_Effects/GameplayPulginsAssets/ToonLightning/MasterMaterial/V1/T_LightningSplineTexture
  }
  NamedRerouteDeclaration(None=Power(Base=Multiply(A=@n24.R)) {constExponent=4}) {name=HotSpots, nodeColor=<rgba 0.2188 1 0 1>}
  n25 = NamedRerouteUsage()
  NamedRerouteDeclaration(None=Add(A=NamedRerouteUsage(), B=ComponentMask(None=@n25) {g=true, r=true})) {name=MappedUVsHotSpots, nodeColor=<rgba 1 0 0 1>}
  n26 = NamedRerouteUsage()
  NamedRerouteDeclaration(None=MaterialFunctionCall(
    A=VectorParameter() {parameterName=GlowColorOuterC, defaultValue=<rgba 0 0 0 1>}.RGB
    B=VectorParameter() {parameterName=GlowColorEdgeC, defaultValue=<rgba 0 0 0 1>}.RGB
    C=VectorParameter() {parameterName=GlowColorInnerC, defaultValue=<rgba 2 0.2306 0.1042 1>}.RGB
    Alpha=Saturate(None=@n26)
   ) {materialFunction=&Lerp_3Color}.Result) {name=GlowHotSpots, nodeColor=<rgba 1 0.7969 0 1>}
  n27 = TextureObjectParameter() {
   parameterName=SizeMapping
   samplerType=SAMPLERTYPE_Masks
   texture=/BRCosmetics/Characters/Player/Male/Medium/Bodies/M_MED_VoyagerCuboid/Textures/T_VoyagerCuboid_ElectricitySize
  }
  n28 = TextureSample(Tex=@n27) {
   mipValueMode=TMVM_MipLevel
   samplerSource=SSM_Wrap_WorldGroupSettings
   samplerType=SAMPLERTYPE_LinearColor
   texture=&T_LightningSplineTexture
  }
  NamedRerouteDeclaration(None=@n28.RGBA) {name=SizeMapping, nodeColor=<rgba 1 0 0.8672 1>}
  n29 = TextureCoordinate()
  n30 = VectorParameter() {parameterName="UV Noise Scale and Speed  Mesh", defaultValue=<rgba 0.02 0.02 0.1 0.1>}
  n31 = ComponentMask(None=@n30.RGB) {g=true, r=true}
  n32 = TextureSample!(
   UVs=Panner(Coordinate=Multiply(A=@n29, B=@n31), Time=NamedRerouteUsage(), Speed=AppendVector(A=@n30.B, B=@n30.B)) {bFractionalPart=true, speedY=1}
   Tex
   Apply View MipBias
  ) {samplerType=SAMPLERTYPE_LinearColor, texture=&HighContrastNoise}
  NamedRerouteDeclaration(None=Multiply(A=ConstantBiasScale(None=@n32.RGB) {bias=-0.5, scale=2}, B=ScalarParameter() {defaultValue=1, parameterName=UV Distortion Intensity})) {name=UVDistortion, nodeColor=<rgba 0 0.5547 1 1>}
  n33 = NamedRerouteUsage()
  n34 = VectorParameter() {parameterName=PushOutNormalsOnLightning, defaultValue=<rgba 4 4 4 1>}
  NamedRerouteDeclaration(None=Multiply(A=VertexNormalWS(), B=Multiply(A=@n33, B=@n34.RGB) {constB=12})) {name=Out_WPO, nodeColor=<rgba 1 0.375 0 1>}
  n35 = Multiply(A=NamedRerouteUsage(), B=Constant2Vector() {g=0.25, r=1})
  n36 = LinearInterpolate(A=@n35, B=@n35, Alpha=ComponentMask(None=NamedRerouteUsage()) {g=true})
  n37 = Multiply(A=NamedRerouteUsage(), B=Constant2Vector() {g=0.75, r=1})
  n38 = ComponentMask(None=NamedRerouteUsage()) {b=true}
  n39 = NamedRerouteUsage()
  n40 = TextureObjectParameter() {
   parameterName=LightningTexture
   samplerType=SAMPLERTYPE_LinearColor
   texture=/BRCosmetics/Characters/Player/Male/Medium/Bodies/M_MED_VoyagerCuboid/Textures/T_VoyagerCuboid_StaticLightningB
  }
  n41 = TextureSample(
   UVs=LinearInterpolate(A=LinearInterpolate(A=@n36, B=@n37, Alpha=@n38), B=NamedRerouteUsage(), Alpha=ComponentMask(None=@n39) {a=true})
   Tex=@n40
  ) {
   samplerType=SAMPLERTYPE_LinearColor
   texture=&T_LightningSplineTexture
  }
  NamedRerouteDeclaration(None=Power(Base=Multiply(A=@n41.R))) {name=OpacityShapeA, nodeColor=<rgba 1 0 0.0703 1>}
  n42 = VectorParameter() {parameterName=GlowColorOuter, defaultValue=<rgba 0.7913 0.1239 0 1>}
  n43 = VectorParameter() {parameterName=GlowColorEdge, defaultValue=<rgba 0.8348 0.0974 0 1>}
  n44 = VectorParameter() {parameterName=GlowColorInner, defaultValue=<rgba 0.4 0.1245 0.0557 1>}
  n45 = NamedRerouteUsage()
  n46 = Fresnel()
  NamedRerouteDeclaration(None=LinearInterpolate(
    A=MaterialFunctionCall(A=@n42.RGB, B=@n43.RGB, C=@n44.RGB, Alpha=@n45) {materialFunction=&Lerp_3Color}.Result
    B=MaterialFunctionCall(A=@n42.RGB, B=@n43.RGB, C=@n44.RGB, Alpha=@n46) {materialFunction=&Lerp_3Color}.Result
   )) {name=GlowLightningColor, nodeColor=<rgba 1 0.7969 0 1>}
  n47 = TextureObjectParameter() {
   parameterName=TextureWithUVsInRGChannels
   samplerType=SAMPLERTYPE_LinearColor
   texture=/BRCosmetics/Characters/Player/Male/Medium/Bodies/M_MED_VoyagerCuboid/Textures/T_VoyagerCuboid_StaticLightningUVGradient
  }
  n48 = TextureSample!(UVs, Tex=@n47, Level, Apply View MipBias) {
   mipValueMode=TMVM_MipLevel
   samplerSource=SSM_Wrap_WorldGroupSettings
   samplerType=SAMPLERTYPE_LinearColor
   texture=&T_LightningSplineTexture
  }
  NamedRerouteDeclaration(None=@n48.G) {name=GreyscaleVerticalGradients, nodeColor=<rgba 0.2187 1 0 1>}
  n49 = Multiply(A=Time(), B=ScalarParameter() {defaultValue=2, parameterName=TimeScale, sortPriority=35}) {constB=0.1}
  n50 = TextureSampleParameter2D!(UVs, Apply View MipBias) {
   parameterName=RandomSquares
   samplerType=SAMPLERTYPE_LinearColor
   texture=&T_VoyagerCuboid_StaticLightningUVGradient
  }
  n51 = NamedRerouteDeclaration(None=@n50.B) {name=RandomMosaic, nodeColor=<rgba 0.7109 1 0 1>}
  n52 = Multiply(A=Time(), B=ScalarParameter() {defaultValue=2, parameterName=TimeScaleB, sortPriority=35}) {constB=0.1}
  n53 = TextureSampleParameter2D(UVs=Add(A=Add(A=@n49, B=@n51), B=Add(A=@n52, B=@n51))) {
   group=BoldTouch_RandomSquares
   parameterName=NoiseTexture
   samplerType=SAMPLERTYPE_Grayscale
   sortPriority=33
   texture=/Game/Global/Textures/Default/Noise/LowResBlurredNoiseNormalized
  }
  NamedRerouteDeclaration(None=Power(Base=@n53.RGB)) {name=RandomLightning, nodeColor=<rgba 0.1953 1 0 1>}
  NamedRerouteDeclaration(None=FunctionInput() {inputName=In_MaterialAttributes, inputType=FunctionInput_MaterialAttributes}) {name=In_MaterialAttributes, nodeColor=<rgba 0 0.4844 1 1>}
  n54 = TextureObjectParameter() {
   parameterName=TextureWithUVsInRGChannels
   samplerType=SAMPLERTYPE_LinearColor
   texture=&T_VoyagerCuboid_StaticLightningUVGradient
  }
  n55 = TextureSample(UVs=TextureCoordinate(), Tex=@n54) {
   mipValueMode=TMVM_MipLevel
   samplerSource=SSM_Wrap_WorldGroupSettings
   samplerType=SAMPLERTYPE_LinearColor
   texture=&T_LightningSplineTexture
  }
  n56 = ComponentMask(None=@n55.RGB) {g=true, r=true}
  n57 = Time()
  n58 = Constant2Vector() {g=0.1}
  NamedRerouteDeclaration(None=Add(A=NamedRerouteDeclaration(None=@n56) {name=UVFromTexture, nodeColor=<rgba 0 1 0.0391 1>}, B=Multiply(A=@n57, B=@n58))) {name=MappedUVs, nodeColor=<rgba 1 0 0 1>}
  n59 = Multiply(A=NamedRerouteUsage(), B=Constant2Vector() {g=0.5, r=1})
  n60 = LinearInterpolate(A=@n59, B=@n59, Alpha=ComponentMask(None=NamedRerouteUsage()) {g=true})
  n61 = Multiply(A=NamedRerouteUsage(), B=Constant2Vector() {g=1, r=1})
  n62 = ComponentMask(None=NamedRerouteUsage()) {b=true}
  n63 = NamedRerouteUsage()
  n64 = Constant2Vector() {g=2, r=1}
  n65 = NamedRerouteUsage()
  n66 = TextureObjectParameter() {
   parameterName=LightningTextureB
   samplerType=SAMPLERTYPE_LinearColor
   texture=/BRCosmetics/Characters/Player/Male/Medium/Bodies/M_MED_VoyagerCuboid/Textures/T_VoyagerCuboid_StaticLightningC
  }
  n67 = TextureSample!(
   UVs=LinearInterpolate(A=LinearInterpolate(A=@n60, B=@n61, Alpha=@n62), B=Multiply(A=@n63, B=@n64), Alpha=ComponentMask(None=@n65) {a=true})
   Tex=@n66
   Apply View MipBias
  ) {
   samplerType=SAMPLERTYPE_LinearColor
   texture=&T_LightningSplineTexture
  }
  n68 = Multiply(A=@n67.R)
  NamedRerouteDeclaration(None=@n68) {name=OpacityShapeB, nodeColor=<rgba 1 0 0.0703 1>}
  n69 = NamedRerouteUsage()
  n70 = Multiply(A=Time(), B=Constant2Vector() {g=0.215})
  n71 = NamedRerouteUsage()
  NamedRerouteDeclaration(None=Add(A=Add(A=@n69, B=@n70), B=ComponentMask(None=@n71) {g=true, r=true})) {name=MappedUVsB, nodeColor=<rgba 1 0 0 1>}
  n72 = NamedRerouteUsage()
  NamedRerouteDeclaration(None=MaterialFunctionCall(
    A=VectorParameter() {parameterName=GlowColorOuterB, defaultValue=<rgba 0.5 0.0417 0 1>}.RGB
    B=VectorParameter() {parameterName=GlowColorEdgeB, defaultValue=<rgba 0.4609 0.1659 0.0922 1>}.RGB
    C=VectorParameter() {parameterName=GlowColorInnerB, defaultValue=<rgba 0.8348 0.2443 0.1348 1>}.RGB
    Alpha=Saturate(None=@n72)
   ) {materialFunction=&Lerp_3Color}.Result) {name=GlowLightningColorB, nodeColor=<rgba 1 0.7969 0 1>}
  n73 = ScalarParameter() {defaultValue=1, parameterName=UV Panner Flicker Frequency}
  n74 = Time()
  n75 = Ceil(None=Multiply(A=@n74, B=@n73) {constB=4})
  n76 = ScalarParameter() {defaultValue=0.2, parameterName=Time Choppiness Frequency}
  n77 = Multiply(A=@n74, B=@n76)
  n78 = ComponentMask(None=VectorNoise(World Position=@n77)) {r=true}
  n79 = Add(A=Multiply(A=Subtract(A=@n78) {constB=0.5}, B=ScalarParameter() {defaultValue=0.2, parameterName=Time Choppiness Randomness Range})) {constB=0.5}
  n80 = Floor(None=@n77)
  n81 = Divide(A=Subtract(A=Subtract(A=@n77, B=@n79), B=@n80), B=OneMinus(None=@n79))
  n82 = Divide(A=Max(A=Add(A=@n81, B=@n80), B=@n80), B=@n76)
  n83 = Divide(A=Subtract(A=@n74, B=@n82), B=@n79)
  NamedRerouteDeclaration(None=LinearInterpolate(A=Divide(A=@n75, B=@n73), B=@n82, Alpha=Saturate(None=@n83)) {desc=Blending between quantized time and choppy time}) {name=Choppy Time, nodeColor=<rgba 0 1 0.9297 1>}
  Power(Base=@n68) {constExponent=4}
 ]
}
