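# Masked, two-sided character material for the M_MED_VoyagerCuboid player body, built on the shared 2023 CharacterParent function chain.
# The PreFX result passes through the cube effect function borrowed from the SpacePlunge backpack, with a constant mask of 1.
# Emissive is then scaled separately for morning, day, evening and night by a vector parameter and selected through Color_Time_of_Day.
# The modified attributes feed the PostFX function for final output.
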Material /BRCosmetics/Characters/Player/Male/Medium/Bodies/M_MED_VoyagerCuboid/Materials/M_Cube_VoyagerCuboid
graph {
 properties {
  bUseMaterialAttributes=true
  bUsedWithInstancedStaticMeshes=true
  bUsedWithMeshDeformer=true
  bUsedWithSkeletalMesh=true
  blendMode=BLEND_Masked
  pixelDepthOffsetMode=PDOM_Legacy
  twoSided=true
  usedShadingModels=Default Lit
 }
 nodes [
  n1 = MaterialFunctionCall() {materialFunction=&MF_CharacterParent_2023_PreFX}
  n2 = GetMaterialAttributes(None=MaterialFunctionCall!(In=@n1.Result, Mask=Constant() {r=1}) {materialFunction=/BRCosmetics/Accessories/FORT_Backpacks/Backpack_M_MED_SpacePlunge/Material/MF_SpacePlungeCube}.Result) {attributeGetTypes [#1]}
  n3 = VectorParameter() {desc="R=morning, G=day, B=evening, A=night", parameterName=MultiplyTimesOfDay, defaultValue=<rgba 1.3 1 1.5 2>}
  n4 = Multiply(A=@n2.Emissive Color, B=@n3.R)
  n5 = Multiply(A=@n2.Emissive Color, B=@n3.G)
  n6 = Multiply(A=@n2.Emissive Color, B=@n3.B)
  n7 = Multiply(A=@n2.Emissive Color, B=@n3.A)
  MaterialFunctionCall(
   Material=SetMaterialAttributes(
    MaterialAttributes=@n2.MaterialAttributes
    Emissive Color=MaterialFunctionCall(Color_Morning=@n4, Color_Day=@n5, Color_Evening=@n6, Color_Night=@n7) {materialFunction=&Color_Time_of_Day}.Result
   ) {attributeSetTypes [#1]}
   M=@n1.M
   Pre Skinned Local Position=@n1.M
   Distance Blend=@n1.M
   Fresnel=@n1.M
  ) {materialFunction=&MF_CharacterParent_2023_PostFX}
 ]
}
