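# Translucent, unlit, two-sided material that draws glowing static-electricity lightning for the VoyagerCuboid character cosmetic, usable on skeletal meshes and Niagara particles.
# A gradient texture supplies remapped UVs and random offsets that stagger curve-atlas animation and noise flicker.
# Two lightning textures and hot spots, with size-map UV scaling, three-color ramps and a mask, form emissive and opacity.
# Choppy quantized time, noise UV distortion and world-position offsets add jitter.
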Material /BRCosmetics/Characters/Player/Male/Medium/Bodies/M_MED_VoyagerCuboid/Materials/M_ParticleStyleLightning
graph {
 properties {
  bAutomaticallySetUsageInEditor=false
  bUsedWithMorphTargets=true
  bUsedWithNiagaraMeshParticles=true
  bUsedWithNiagaraSprites=true
  bUsedWithSkeletalMesh=true
  blendMode=BLEND_Translucent
  opacityMaskClipValue=0.1
  pixelDepthOffsetMode=PDOM_Legacy
  shadingModel=MSM_Unlit
  twoSided=true
  usedShadingModels=Unlit
 }
 nodes [
  n1 = TextureObjectParameter() {
   parameterName=TextureWithUVsInRGChannels
   samplerType=SAMPLERTYPE_LinearColor
   texture=/BRCosmetics/Characters/Player/Male/Medium/Bodies/M_MED_VoyagerCuboid/Textures/T_VoyagerCuboid_StaticLightningUVGradient
  }
  n2 = TextureSample!(UVs=TextureCoordinate(), Tex=@n1, Level, Apply View MipBias) {
   mipValueMode=TMVM_MipLevel
   samplerSource=SSM_Wrap_WorldGroupSettings
   samplerType=SAMPLERTYPE_LinearColor
   texture=/Game/Effects/Fort_Effects/GameplayPulginsAssets/ToonLightning/MasterMaterial/V1/T_LightningSplineTexture
  }
  n3 = NamedRerouteDeclaration(None=ComponentMask(None=@n2.RGB) {g=true, r=true}) {name=UVFromTexture, nodeColor=<rgba 0 1 0.0391 1>}
  Add(A=@n3, B=Multiply(A=Time(), B=Constant2Vector() {g=1}))
  n4 = NamedRerouteUsage()
  n5 = VectorParameter() {group=em, parameterName=Glow_EdgeMask_Control, channelNames {a="", b=Power, g=Multiply, r=Offset}, defaultValue=<rgba 0 2 1 1>}
  n6 = VectorParameter() {parameterName=Glow_Refine, defaultValue=<rgba 6 2 0.1 0.5>}
  n7 = MaterialFunctionCall(
   Texture=MaterialFunctionCall!(
    UorV=MaterialFunctionCall(Float2=@n4) {materialFunction=&BreakOutFloat2Components}.R
    EdgeControl=@n5
   ) {materialFunction=/Game/Effects/Niagara/MasterMaterials/Functions/Common/FXCommon_EdgeMask_Simple}.Mask
   RGB=@n6.RGB
   A=@n6.RGB
  ) {materialFunction=&MF_FX_Refine}
  n8 = Add(A=@n7.Result)
  n9 = MaterialFunctionCall(Float2=NamedRerouteUsage()) {materialFunction=&BreakOutFloat2Components}
  n10 = LinearInterpolate(
   A=VectorParameter() {parameterName=Glow_ColorA, defaultValue=<rgba 1 0.6801 0.1274 1>}.RGB
   B=VectorParameter() {parameterName=Glow_ColorB, defaultValue=<rgba 0.4 0.196 0.04 1>}.RGB
   Alpha=Power(Base=@n9.G, Exp=ScalarParameter() {defaultValue=3, parameterName=Glow_ColorFromSpike_Power})
  )
  n11 = Multiply(
   A=Power(Base=@n9.R, Exp=ScalarParameter() {defaultValue=2, parameterName=Glow_BrogjtmessFromSpike_Power})
   B=ScalarParameter() {defaultValue=0.1, parameterName=Glow_BrogjtmessFromSpike_Str}
  )
  n12 = TextureObjectParameter(UVs=NamedRerouteUsage()) {
   group=LightningT
   parameterName=LightningT_Texture
   samplerType=SAMPLERTYPE_Masks
   sortPriority=0
   texture=/BRCosmetics/Characters/Player/Male/Medium/Bodies/M_MED_VoyagerCuboid/Textures/T_VoyagerCuboid_StaticLightning
  }
  n13 = VectorParameter() {group=LightningT, parameterName=LightningT_UVControl, channelNames {a=VPan, b=UPan, g=VScale, r=UScale}, defaultValue=<rgba 1 3 0 0.67>}
  n14 = NamedRerouteUsage()
  n15 = VectorParameter() {parameterName=UV Noise Scale and Seped, defaultValue=<rgba 0.4 0.4 0.4 1.3>}
  n16 = ComponentMask(None=@n15.RGB) {g=true, r=true}
  n17 = TextureSample!(
   UVs=Panner(Coordinate=Multiply(A=@n14, B=@n16), Time=NamedRerouteUsage(), Speed=AppendVector(A=@n15.B, B=@n15.B)) {bFractionalPart=true, speedY=1}
   Tex
   Apply View MipBias
  ) {samplerType=SAMPLERTYPE_LinearColor, texture=&HighContrastNoise}
  n18 = Multiply(
   A=Multiply(A=ConstantBiasScale(None=@n17.R) {bias=-0.5, scale=2}, B=NamedRerouteUsage())
   B=ScalarParameter() {defaultValue=0.02, parameterName=UV Distortion Intensity}
  )
  n19 = Multiply(A=ScalarParameter() {defaultValue=1, parameterName=Lightning UV Width Scale}, B=Multiply(A=NamedRerouteUsage(), B=NamedRerouteUsage()))
  n20 = Constant2Vector() {g=1, r=1}
  n21 = MaterialFunctionCall(Float2=MaterialFunctionCall(UVs=NamedRerouteUsage(), Texture Scale=AppendVector(A=@n19, B=@n20)) {materialFunction=&ScaleUVsByCenter}.UVs) {materialFunction=&BreakOutFloat2Components}
  n22 = AppendVector(A=Constant(), B=Divide(A=ScalarParameter() {defaultValue=0.4, parameterName=Lightning Texture Pan Speed}, B=@n20))
  n23 = Panner(
   Coordinate=MaterialFunctionCall(X=Saturate(None=@n21.R), Y=@n21.G) {materialFunction=&MakeFloat2}.Result
   Time=Time()
   Speed=@n22
  ) {bFractionalPart=true, speedY=1.1}
  n24 = Constant()
  n25 = ParticleRandom()
  n26 = MaterialFunctionCall!(
   Texture=@n12
   AdditionalUVOffset
   UVControl_RGB=@n13
   UVControl_A=@n13
   "[Override]TimeInput"
   "[Override]ScaleUVByCenter"
   "[Override]UVInput"=Add(A=Add(A=@n18, B=@n23), B=AppendVector(A=@n24, B=@n25))
  ) {materialFunction=/Game/Effects/Niagara/MasterMaterials/Functions/Common/FXCommon_TextureSample_V2}
  n27 = VectorParameter() {parameterName=InnerGlowRefine, defaultValue=<rgba 5 2 1 0.5>}
  n28 = Multiply(
   A=MaterialFunctionCall(Texture=OneMinus(None=@n26.G), RGB=@n27.RGB, A=@n27.RGB) {materialFunction=&MF_FX_Refine}.Result
   B=@n26.B
  )
  n29 = LinearInterpolate(
   A=VectorParameter() {parameterName=GlowColor, defaultValue=<rgba 0.5 0.1667 0 1>}.RGB
   B=LinearInterpolate(
    A=VectorParameter() {parameterName=BoltColor, defaultValue=<rgba 1.25 0.3517 0.025 1>}.RGB
    B=VectorParameter() {parameterName=InnerGlowColor, defaultValue=<rgba 12 2.4 0 5>}.RGB
    Alpha=Saturate(None=@n28)
   )
   Alpha=@n26.B
  )
  n30 = Multiply(A=Multiply(A=Multiply(A=@n8, B=@n10), B=Add(A=@n11)), B=@n29)
  Add(A=@n29, B=@n30)
  Constant() {r=1}
  DynamicParameter() {defaultValue=<rgba 10 0 0 0>, paramNames [Lightning Width, Lightning Tiling, Lightning Pan, Lightning Invert]}
  DynamicParameter() {parameterIndex=2, defaultValue=<rgba -1 -1 1 1>, paramNames [WPO Distort, Dissolve, Scale Z, Param4]}
  DynamicParameter() {parameterIndex=2, defaultValue=<rgba -1 -1 1 1>, paramNames [WPO Distort, Dissolve, Scale Z, Param4]}
  Multiply(A=NamedRerouteUsage(), B=Constant2Vector() {g=0.5, r=1})
  Multiply(A=NamedRerouteUsage(), B=Constant2Vector() {g=1, r=1})
  n31 = ComponentMask(None=TextureCoordinate() {coordinateIndex=1}) {r=true}
  Multiply(A=VertexNormalWS(), B=Sign(None=Subtract(A=@n31) {constB=0.5}))
  n32 = ScalarParameter() {defaultValue=1, parameterName=DissolveMultiplier}
  Multiply(A=NamedRerouteUsage(), B=@n32)
  n33 = Time()
  n34 = TextureSampleParameter2D() {
   parameterName=RandomSquares
   samplerType=SAMPLERTYPE_LinearColor
   texture=&T_VoyagerCuboid_StaticLightningUVGradient
  }
  n35 = Frac(None=Multiply(A=Add(A=@n33, B=@n34.B), B=ScalarParameter() {defaultValue=1.5, parameterName=SpeedOfStaticLightning}))
  n36 = CurveAtlasRowParameter(CurveTime=@n35) {
   atlas=/BRCosmetics/Characters/Player/Male/Medium/Bodies/M_MED_VoyagerCuboid/Curves/CA_VoyagerCuboidStaticLightning
   curve=/BRCosmetics/Characters/Player/Male/Medium/Bodies/M_MED_VoyagerCuboid/Curves/CUR_VoyagerCuboidStaticLightningA
   parameterName=StaticElectricityCurvesA
  }
  NamedRerouteDeclaration(None=@n36.R) {name=WPODistort_FromCurve, nodeColor=<rgba 1 0 0 1>}
  n37 = ComponentMask(None=NamedRerouteUsage()) {g=true}
  n38 = Subtract(A=@n37, B=NamedRerouteUsage())
  n39 = ScalarParameter() {defaultValue=0.1, parameterName=Opacity - Fade Size Tip}
  NamedRerouteDeclaration(None=Saturate(None=Divide(A=@n38, B=@n39))) {name=Length Mask, nodeColor=<rgba 1 0 0.8437 1>}
  n40 = NamedRerouteUsage()
  n41 = MaterialFunctionCall(Float2=@n40) {materialFunction=&BreakOutFloat2Components}
  n42 = OneMinus(None=@n41.G)
  n43 = DynamicParameter() {defaultValue=<rgba 10 0 0 0>, paramNames [Lightning Width, Lightning Tiling, Lightning Pan, Lightning Invert]}
  n44 = Saturate(None=@n43.Lightning Invert)
  NamedRerouteDeclaration(None=LinearInterpolate(
    A=@n40
    B=MaterialFunctionCall(X=@n41.R, Y=@n42) {materialFunction=&MakeFloat2}.Result
    Alpha=Round(None=@n44)
   ) {constAlpha=1}) {name=UVs, nodeColor=<rgba 1 0.5859 0 1>}
  n45 = TextureObjectParameter() {group=TextureWPONoise, parameterName=TextureWPONoise_Texture, sortPriority=0, texture=&T_NoisePack_01}
  n46 = ParticleRandom()
  n47 = VectorParameter() {
   group=TextureWPONoise
   parameterName=TextureWPONoise_UVControl
   channelNames {a=VPan, b=UPan, g=VScale, r=UScale}
   defaultValue=<rgba 0 0.75 0 0.02>
  }
  n48 = Add(A=Time() {bOverride_Period=true, period=6666}, B=@n46)
  n49 = NamedRerouteUsage()
  n50 = Multiply(
   A=Add(
    A=MaterialFunctionCall(Texture=@n45, AdditionalUVOffset=@n46, UVControl_RGB=@n47, UVControl_A=@n47, "[Override]TimeInput"=@n48, "[Override]UVInput"=@n49) {materialFunction=&FXCommon_TextureSample}.RGB
    B=ScalarParameter() {defaultValue=-0.5, parameterName=TextureWPONoise_Offset}
   )
   B=Add(A=NamedRerouteUsage(), B=ScalarParameter() {parameterName=TextureWPONoise_BaseNoiseIntensity})
  )
  n51 = MaterialFunctionCall(Float2=NamedRerouteUsage()) {materialFunction=&BreakOutFloat2Components}
  n52 = VectorParameter() {
   group=TextureWPONoise
   parameterName=TextureWPONoise_EdgeMask_InvertControl
   channelNames {a="", b=Power, g=Multiply, r=Offset}
   defaultValue=<rgba 0 5 1 1>
  }
  n53 = VectorParameter() {
   group=TextureWPONoise
   parameterName=TextureWPONoise_EdgeMask_Control
   channelNames {a="", b=Power, g=Multiply, r=Offset}
   defaultValue=<rgba 0 5 1 1>
  }
  n54 = Transform(None=Multiply(
    A=Multiply(A=@n50)
    B=MaterialFunctionCall!(UorV=@n51.G, Invert_EdgeControl=@n52, NoInvert_EdgeControl=@n53) {materialFunction=/Game/Effects/Niagara/MasterMaterials/Functions/Common/FXCommon_EdgeMask_Ex}.Mask
   )) {transformSourceType=TRANSFORMSOURCE_Local}
  n55 = Constant3Vector()
  NamedRerouteDeclaration(None=MaterialFunctionCall!("Xb1,PS4"=@n54, "PC-High&&Epic,XSX,PS5"=@n54, "PC-Perf,Android"=@n55, Switch=@n55, "PC-Low(NotPerf)&&Medium"=@n54) {materialFunction=/Game/Effects/Niagara/MasterMaterials/Functions/Common/PlatformSwitch_3D}.Result) {name=DistortWPO, nodeColor=<rgba 1 0.7031 0 1>}
  n56 = MaterialFunctionCall(Float2=NamedRerouteUsage()) {materialFunction=&BreakOutFloat2Components}
  n57 = Constant() {r=0.5}
  n58 = TextureObjectParameter() {
   group=DissolveTexture
   parameterName=DissolveTexture_Texture
   samplerType=SAMPLERTYPE_LinearColor
   sortPriority=0
   texture=/Engine/Functions/Engine_MaterialFunctions02/ExampleContent/Textures/OffsetNoiseDistanceFields_ForNormals
  }
  n59 = VectorParameter() {
   group=DissolveTexture
   parameterName=DissolveTexture_UVControl
   channelNames {a=VPan, b=UPan, g=VScale, r=UScale}
   defaultValue=<rgba 0 0.5 0 0.5>
  }
  n60 = MaterialFunctionCall(Texture=@n58, AdditionalUVOffset=ParticleRandom(), UVControl_RGB=@n59, UVControl_A=@n59, "[Override]UVInput"=NamedRerouteUsage()) {materialFunction=&FXCommon_TextureSample}
  n61 = NamedRerouteUsage()
  n62 = Constant() {r=0.6}
  NamedRerouteDeclaration(None=MaterialFunctionCall!(
    Texture=SphereMask(A=@n56.R, B=@n57, Radius=@n57) {hardnessPercent=0}
    Erosion=MaterialFunctionCall(Texture=@n60.R, Erosion=@n61, Softness=@n62) {materialFunction=/Game/Effects/Niagara/MasterMaterials/Functions/MF_Erosion_Smoothstep_01}.Result
    Softness=Constant() {r=0.3}
   ) {materialFunction=&MF_Erosion_Smoothstep_01}.Result) {name=Dissolve, nodeColor=<rgba 0.2422 0 1 1>}
  n63 = Multiply(A=NamedRerouteUsage(), B=NamedRerouteUsage())
  n64 = NamedRerouteUsage()
  n65 = NamedRerouteUsage()
  n66 = NamedRerouteUsage()
  NamedRerouteDeclaration(None=Multiply(A=StaticSwitchParameter(True=@n65, False=@n66) {parameterName=Glow}, B=Multiply(A=@n63, B=@n64))) {name=BaseOpacity, nodeColor=<rgba 0.7578 1 0 1>}
  NamedRerouteDeclaration(None=@n30) {name=TexColor, nodeColor=<rgba 0 1 0.0156 1>}
  NamedRerouteDeclaration(None=Constant()) {name=LightningPan_FromCurve, nodeColor=<rgba 0 0 1 1>}
  n67 = VectorParameter() {parameterName=OutGlowRefine, defaultValue=<rgba 1 1 0 0>}
  NamedRerouteDeclaration(None=MaterialFunctionCall(Texture=@n26.R, RGB=@n67.RGB, A=@n67.RGB) {materialFunction=&MF_FX_Refine}.Result) {name=OpacityShape, nodeColor=<rgba 1 0 0.0703 1>}
  NamedRerouteDeclaration(None=@n7.Result) {name=GlowShape, nodeColor=<rgba 0 0.3906 1 1>}
  NamedRerouteDeclaration(None=@n36.A) {name=LightningWidth_FromCurve, nodeColor=<rgba 1 1 1 1>}
  n68 = CurveAtlasRowParameter(CurveTime=@n35) {
   atlas=&CA_VoyagerCuboidStaticLightning
   curve=&CUR_VoyagerCuboidStaticLightningA
   parameterName=StaticElectricityCurvesB
  }
  NamedRerouteDeclaration(None=@n68.R) {name=ScaleColor_FromCurve, nodeColor=<rgba 1 0 0 1>}
  n69 = ComponentMask(None=NamedRerouteUsage()) {r=true}
  n70 = OneMinus(None=Abs(None=ConstantBiasScale(None=@n69) {bias=-0.5, scale=2}))
  n71 = ScalarParameter() {defaultValue=0.5, parameterName=Opacity - Fade Size Width}
  NamedRerouteDeclaration(None=Saturate(None=Divide(A=@n70, B=@n71))) {name=Width Mask, nodeColor=<rgba 0 0.2969 1 1>}
  NamedRerouteDeclaration(None=@n68.G) {name=ScaleAlpha_FromCurve, nodeColor=<rgba 0.3828 1 0 1>}
  NamedRerouteDeclaration(None=@n68.B) {name=GlowColor_FromCurve, nodeColor=<rgba 0 0 1 1>}
  NamedRerouteDeclaration() {name=In_MaterialAttributes, nodeColor=<rgba 0 0.4844 1 1>}
  n72 = ScalarParameter() {defaultValue=19, parameterName=UV Panner Flicker Frequency}
  n73 = Time()
  n74 = Ceil(None=Multiply(A=@n73, B=@n72) {constB=4})
  n75 = ScalarParameter() {defaultValue=2.62, parameterName=Time Choppiness Frequency}
  n76 = Multiply(A=@n73, B=@n75)
  n77 = ComponentMask(None=VectorNoise(World Position=@n76)) {r=true}
  n78 = Add(A=Multiply(A=Subtract(A=@n77) {constB=0.5}, B=ScalarParameter() {defaultValue=0.9, parameterName=Time Choppiness Randomness Range})) {constB=0.5}
  n79 = Floor(None=@n76)
  n80 = Divide(A=Subtract(A=Subtract(A=@n76, B=@n78), B=@n79), B=OneMinus(None=@n78))
  n81 = Divide(A=Max(A=Add(A=@n80, B=@n79), B=@n79), B=@n75)
  n82 = Divide(A=Subtract(A=@n73, B=@n81), B=@n78)
  NamedRerouteDeclaration(None=LinearInterpolate(A=Divide(A=@n74, B=@n72), B=@n81, Alpha=Saturate(None=@n82)) {desc=Blending between quantized time and choppy time}) {name=Choppy Time_1, nodeColor=<rgba 0 1 0.9297 1>}
  n83 = TextureCoordinate()
  n84 = VectorParameter() {parameterName="UV Noise Scale and Speed  Mesh", defaultValue=<rgba 4 4 0.5 0.5>}
  n85 = ComponentMask(None=@n84.RGB) {g=true, r=true}
  n86 = TextureSample(UVs=Panner(Coordinate=Multiply(A=@n83, B=@n85), Time=NamedRerouteUsage(), Speed=AppendVector(A=@n84.B, B=@n84.B)) {bFractionalPart=true, speedY=1}) {samplerType=SAMPLERTYPE_LinearColor, texture=&HighContrastNoise}
  NamedRerouteDeclaration(None=Multiply(A=ConstantBiasScale(None=@n86.RGB) {bias=-0.5, scale=2}, B=ScalarParameter() {defaultValue=0.02, parameterName=UV Distortion Intensity})) {name=UVDistortion, nodeColor=<rgba 0 0.5547 1 1>}
  n87 = NamedRerouteUsage()
  n88 = VectorParameter() {parameterName=PushOutNormalsOnLightning, defaultValue=<rgba 1 1 1 1>}
  NamedRerouteDeclaration(None=Multiply(A=VertexNormalWS(), B=Multiply(A=@n87, B=@n88.RGB) {constB=12})) {name=Out_WPO, nodeColor=<rgba 1 0.375 0 1>}
  n89 = Multiply(A=NamedRerouteUsage(), B=Constant2Vector() {g=0.25, r=1})
  n90 = LinearInterpolate(A=@n89, B=@n89, Alpha=ComponentMask(None=NamedRerouteUsage()) {g=true})
  n91 = Multiply(A=NamedRerouteUsage(), B=Constant2Vector() {g=0.75, r=1})
  n92 = ComponentMask(None=NamedRerouteUsage()) {b=true}
  n93 = NamedRerouteUsage()
  n94 = TextureObjectParameter() {
   parameterName=LightningTexture
   samplerType=SAMPLERTYPE_LinearColor
   texture=/BRCosmetics/Characters/Player/Male/Medium/Bodies/M_MED_VoyagerCuboid/Textures/T_VoyagerCuboid_StaticLightningB
  }
  n95 = TextureSample(
   UVs=LinearInterpolate(A=LinearInterpolate(A=@n90, B=@n91, Alpha=@n92), B=NamedRerouteUsage(), Alpha=ComponentMask(None=@n93) {a=true})
   Tex=@n94
  ) {
   samplerType=SAMPLERTYPE_LinearColor
   texture=&T_LightningSplineTexture
  }
  NamedRerouteDeclaration(None=Power(Base=Multiply(A=@n95.R))) {name=OpacityShapeA, nodeColor=<rgba 1 0 0.0703 1>}
  n96 = VectorParameter() {parameterName=GlowColorOuter, defaultValue=<rgba 0.7913 0.1239 0 1>}
  n97 = VectorParameter() {parameterName=GlowColorEdge, defaultValue=<rgba 0.8348 0.0974 0 1>}
  n98 = VectorParameter() {parameterName=GlowColorInner, defaultValue=<rgba 0.4 0.1245 0.0557 1>}
  n99 = NamedRerouteUsage()
  n100 = Fresnel()
  NamedRerouteDeclaration(None=LinearInterpolate(
    A=MaterialFunctionCall(A=@n96.RGB, B=@n97.RGB, C=@n98.RGB, Alpha=@n99) {materialFunction=&Lerp_3Color}.Result
    B=MaterialFunctionCall(A=@n96.RGB, B=@n97.RGB, C=@n98.RGB, Alpha=@n100) {materialFunction=&Lerp_3Color}.Result
   )) {name=GlowLightningColor, nodeColor=<rgba 1 0.7969 0 1>}
  n101 = TextureObjectParameter() {
   parameterName=TextureWithUVsInRGChannels
   samplerType=SAMPLERTYPE_LinearColor
   texture=&T_VoyagerCuboid_StaticLightningUVGradient
  }
  n102 = TextureSample!(UVs, Tex=@n101, Level, Apply View MipBias) {
   mipValueMode=TMVM_MipLevel
   samplerSource=SSM_Wrap_WorldGroupSettings
   samplerType=SAMPLERTYPE_LinearColor
   texture=&T_LightningSplineTexture
  }
  NamedRerouteDeclaration(None=@n102.G) {name=GreyscaleVerticalGradients, nodeColor=<rgba 0.2188 1 0 1>}
  n103 = CameraPositionWS()
  n104 = Add(A=WorldPosition().XYZ, B=NamedRerouteUsage())
  n105 = Multiply(A=Normalize(VectorInput=Subtract(A=@n103, B=@n104)), B=ScalarParameter() {defaultValue=22, parameterName=Camera Offset})
  n106 = NamedRerouteUsage()
  NamedRerouteDeclaration(None=Add(A=Add(A=@n105, B=@n106), B=NamedRerouteUsage()) {constB=0}) {name=WPO, nodeColor=<rgba 1 0 0.5391 1>}
  n107 = Multiply(A=Time(), B=ScalarParameter() {defaultValue=2, parameterName=TimeScale, sortPriority=35}) {constB=0.1}
  n108 = TextureSampleParameter2D() {
   parameterName=RandomSquares
   samplerType=SAMPLERTYPE_LinearColor
   texture=&T_VoyagerCuboid_StaticLightningUVGradient
  }
  n109 = NamedRerouteDeclaration(None=@n108.B) {name=RandomMosaic, nodeColor=<rgba 0.7109 1 0 1>}
  n110 = Multiply(A=Time(), B=ScalarParameter() {defaultValue=2, parameterName=TimeScaleB, sortPriority=35}) {constB=0.1}
  n111 = TextureSampleParameter2D(UVs=Add(A=Add(A=@n107, B=@n109), B=Add(A=@n110, B=@n109))) {
   group=BoldTouch_RandomSquares
   parameterName=NoiseTexture
   samplerType=SAMPLERTYPE_Grayscale
   sortPriority=33
   texture=/Game/Global/Textures/Default/Noise/LowResBlurredNoiseNormalized
  }
  NamedRerouteDeclaration(None=Power(Base=@n111.RGB)) {name=RandomLightning, nodeColor=<rgba 0.1953 1 0 1>}
  n112 = TextureObjectParameter() {
   parameterName=SizeMapping
   samplerType=SAMPLERTYPE_Masks
   texture=/BRCosmetics/Characters/Player/Male/Medium/Bodies/M_MED_VoyagerCuboid/Textures/T_VoyagerCuboid_ElectricitySize
  }
  n113 = TextureSample(Tex=@n112) {
   mipValueMode=TMVM_MipLevel
   samplerSource=SSM_Wrap_WorldGroupSettings
   samplerType=SAMPLERTYPE_LinearColor
   texture=&T_LightningSplineTexture
  }
  NamedRerouteDeclaration(None=@n113.RGBA) {name=SizeMapping, nodeColor=<rgba 1 0 0.8672 1>}
  NamedRerouteDeclaration(None=@n3) {name=MappedUVs, nodeColor=<rgba 1 0 0 1>}
  n114 = Multiply(A=NamedRerouteUsage(), B=Constant2Vector() {g=0.5, r=1})
  n115 = LinearInterpolate(A=@n114, B=@n114, Alpha=ComponentMask(None=NamedRerouteUsage()) {g=true})
  n116 = Multiply(A=NamedRerouteUsage(), B=Constant2Vector() {g=1, r=1})
  n117 = ComponentMask(None=NamedRerouteUsage()) {b=true}
  n118 = NamedRerouteUsage()
  n119 = Constant2Vector() {g=2, r=1}
  n120 = NamedRerouteUsage()
  n121 = TextureObjectParameter() {
   parameterName=LightningTextureB
   samplerType=SAMPLERTYPE_LinearColor
   texture=/BRCosmetics/Characters/Player/Male/Medium/Bodies/M_MED_VoyagerCuboid/Textures/T_VoyagerCuboid_StaticLightningC
  }
  n122 = TextureSample!(
   UVs=LinearInterpolate(A=LinearInterpolate(A=@n115, B=@n116, Alpha=@n117), B=Multiply(A=@n118, B=@n119), Alpha=ComponentMask(None=@n120) {a=true})
   Tex=@n121
   Apply View MipBias
  ) {
   samplerType=SAMPLERTYPE_LinearColor
   texture=&T_LightningSplineTexture
  }
  n123 = Multiply(A=@n122.R)
  NamedRerouteDeclaration(None=@n123) {name=OpacityShapeB, nodeColor=<rgba 1 0 0.0703 1>}
  n124 = NamedRerouteUsage()
  n125 = Multiply(A=Time(), B=Constant2Vector() {g=0.215})
  n126 = NamedRerouteUsage()
  NamedRerouteDeclaration(None=Add(A=Add(A=@n124, B=@n125), B=ComponentMask(None=@n126) {g=true, r=true})) {name=MappedUVsB, nodeColor=<rgba 1 0 0 1>}
  n127 = NamedRerouteUsage()
  NamedRerouteDeclaration(None=MaterialFunctionCall(
    A=VectorParameter() {parameterName=GlowColorOuterB, defaultValue=<rgba 0.5 0.0417 0 1>}.RGB
    B=VectorParameter() {parameterName=GlowColorEdgeB, defaultValue=<rgba 0.4609 0.1659 0.0922 1>}.RGB
    C=VectorParameter() {parameterName=GlowColorInnerB, defaultValue=<rgba 0.8348 0.2443 0.1348 1>}.RGB
    Alpha=Saturate(None=@n127)
   ) {materialFunction=&Lerp_3Color}.Result) {name=GlowLightningColorB, nodeColor=<rgba 1 0.7969 0 1>}
  n128 = TextureCoordinate()
  n129 = VectorParameter() {parameterName=UV Noise Scale and Speed Texture, defaultValue=<rgba 0.005 0.05 0.1 0.3>}
  n130 = ComponentMask(None=@n129.RGB) {g=true, r=true}
  n131 = TextureSample(UVs=Panner(Coordinate=Multiply(A=@n128, B=@n130), Time=NamedRerouteUsage(), Speed=AppendVector(A=@n129.B, B=@n129.B)) {bFractionalPart=true, speedY=1}) {samplerType=SAMPLERTYPE_LinearColor, texture=&HighContrastNoise}
  NamedRerouteDeclaration(None=Multiply(
    A=ConstantBiasScale(None=@n131.RGB) {bias=-0.5, scale=2}
    B=ScalarParameter() {defaultValue=0.1, parameterName="UV  Distortion Intensity texture"}
   )) {name=UVDistortionTexture, nodeColor=<rgba 0 0.5547 1 1>}
  n132 = Multiply(A=NamedRerouteUsage(), B=Constant2Vector() {g=0.25, r=1})
  n133 = LinearInterpolate(A=@n132, B=@n132, Alpha=ComponentMask(None=NamedRerouteUsage()) {g=true})
  n134 = Multiply(A=NamedRerouteUsage(), B=Constant2Vector() {g=0.5, r=1})
  n135 = ComponentMask(None=NamedRerouteUsage()) {b=true}
  n136 = NamedRerouteUsage()
  n137 = Constant2Vector() {g=1, r=1}
  n138 = NamedRerouteUsage()
  n139 = TextureObjectParameter() {
   parameterName=TextureHotSpots
   samplerType=SAMPLERTYPE_LinearColor
   texture=/BRCosmetics/Characters/Player/Male/Medium/Bodies/M_MED_VoyagerCuboid/Textures/T_VoyagerCuboid_HotSpots
  }
  n140 = TextureSample!(
   UVs=LinearInterpolate(A=LinearInterpolate(A=@n133, B=@n134, Alpha=@n135), B=Multiply(A=@n136, B=@n137), Alpha=ComponentMask(None=@n138) {a=true})
   Tex=@n139
   Level
   Apply View MipBias
  ) {
   mipValueMode=TMVM_MipLevel
   samplerSource=SSM_Wrap_WorldGroupSettings
   samplerType=SAMPLERTYPE_LinearColor
   texture=&T_LightningSplineTexture
  }
  NamedRerouteDeclaration(None=Power(Base=Multiply(A=@n140.R)) {constExponent=4}) {name=HotSpots, nodeColor=<rgba 0.2188 1 0 1>}
  n141 = NamedRerouteUsage()
  NamedRerouteDeclaration(None=Add(A=NamedRerouteUsage(), B=ComponentMask(None=@n141) {g=true, r=true})) {name=MappedUVsHotSpots, nodeColor=<rgba 1 0 0 1>}
  n142 = AppendVector(A=@n22, B=Constant())
  n143 = DeltaTime()
  NamedRerouteDeclaration(None=Transform(None=Multiply(A=@n142, B=@n143))) {name=Panner Vector, nodeColor=<rgba 1 0 0.7969 1>}
  n144 = NamedRerouteUsage()
  NamedRerouteDeclaration(None=MaterialFunctionCall(
    A=VectorParameter() {parameterName=GlowColorOuterC, defaultValue=<rgba 0 0 0 1>}.RGB
    B=VectorParameter() {parameterName=GlowColorEdgeC, defaultValue=<rgba 0 0 0 1>}.RGB
    C=VectorParameter() {parameterName=GlowColorInnerC, defaultValue=<rgba 2 0.2306 0.1042 1>}.RGB
    Alpha=Saturate(None=@n144)
   ) {materialFunction=&Lerp_3Color}.Result) {name=GlowHotSpots, nodeColor=<rgba 1 0.7969 0 1>}
  n145 = NamedRerouteUsage()
  n146 = VectorParameter() {parameterName=PushOutNormalsOnLightning, defaultValue=<rgba 1 1 1 1>}
  NamedRerouteDeclaration(None=Multiply(A=VertexNormalWS(), B=Multiply(A=@n145, B=@n146.RGB) {constB=12})) {name=Out_WPO_1, nodeColor=<rgba 1 0.375 0 1>}
  n147 = TextureCoordinate()
  n148 = VectorParameter() {parameterName="UV Noise Scale and Speed  Mesh", defaultValue=<rgba 4 4 0.5 0.5>}
  n149 = ComponentMask(None=@n148.RGB) {g=true, r=true}
  n150 = TextureSampleParameter2D() {
   parameterName=RandomSquares
   samplerType=SAMPLERTYPE_LinearColor
   texture=&T_VoyagerCuboid_StaticLightningUVGradient
  }
  n151 = Add(A=Time(), B=@n150.B)
  n152 = ScalarParameter() {defaultValue=0.5, parameterName=SmoothDistortionTimeMultiplier}
  n153 = TextureSample!(
   UVs=Panner(Coordinate=Multiply(A=@n147, B=@n149), Time=Multiply(A=@n151, B=@n152), Speed=AppendVector(A=@n148.B, B=@n148.B)) {bFractionalPart=true, speedY=1}
   Tex
   Apply View MipBias
  ) {texture=/Game/Global/Textures/Default/Noise/T_SoftNoise}
  NamedRerouteDeclaration(None=Multiply(A=ConstantBiasScale(None=@n153.RGB) {bias=-0.5, scale=2}, B=ScalarParameter() {defaultValue=0.02, parameterName=UV Distortion Intensity})) {name=SmoothDistortion, nodeColor=<rgba 0 0.5547 1 1>}
  NamedRerouteDeclaration(None=@n36.G) {name=Dissolve_FromCurve, nodeColor=<rgba 0 1 0 1>}
  n154 = ScalarParameter() {defaultValue=19, parameterName=UV Panner Flicker Frequency}
  n155 = ParticleRandom()
  n156 = Add(A=Time(), B=Frac(None=Multiply(A=@n155) {constB=3.618})) {constB=0}
  n157 = Ceil(None=Multiply(A=@n156, B=@n154) {constB=4})
  n158 = ScalarParameter() {defaultValue=2.62, parameterName=Time Choppiness Frequency}
  n159 = Multiply(A=@n156, B=@n158)
  n160 = ComponentMask(None=VectorNoise(World Position=@n159)) {r=true}
  n161 = Add(A=Multiply(A=Subtract(A=@n160) {constB=0.5}, B=ScalarParameter() {defaultValue=0.9, parameterName=Time Choppiness Randomness Range})) {constB=0.5}
  n162 = Floor(None=@n159)
  n163 = Divide(A=Subtract(A=Subtract(A=@n159, B=@n161), B=@n162), B=OneMinus(None=@n161))
  n164 = Divide(A=Max(A=Add(A=@n163, B=@n162), B=@n162), B=@n158)
  n165 = Divide(A=Subtract(A=@n156, B=@n164), B=@n161)
  NamedRerouteDeclaration(None=LinearInterpolate(A=Divide(A=@n157, B=@n154), B=@n164, Alpha=Saturate(None=@n165)) {desc=Blending between quantized time and choppy time}) {name=Choppy Time, nodeColor=<rgba 0 1 0.9297 1>}
  n166 = Normalize(VectorInput=Subtract(A=CameraPositionWS(), B=WorldPosition().XYZ))
  n167 = Transform(None=Constant3Vector() {constant=<rgba 0 1 0 0>})
  n168 = DynamicParameter() {
   desc="This only exists to fix the splinethicken meshes on android,\r\napparently the tangents do not properly respond to negative scale there"
   parameterIndex=1
   defaultValue=<rgba 1 1 1 0>
   paramNames [Scale X, Scale Y, Scale Z, ""]
  }
  n169 = ShadingPathSwitch(Default=@n167, Deferred=@n167, Mobile=Multiply(A=@n167, B=@n168.RGB)) {desc=Migrated from FeatureLevelSwitch}
  n170 = DotProduct(A=@n166, B=@n169)
  NamedRerouteDeclaration(None=OneMinus(None=Abs(None=@n170))) {name=Fresnel Fade, nodeColor=<rgba 0 0.9297 1 1>}
  n171 = OneMinus(None=@n37)
  n172 = ScalarParameter() {defaultValue=0.2, parameterName=Opacity - Fade Size Base}
  NamedRerouteDeclaration(None=Saturate(None=Divide(A=@n171, B=@n172))) {name=Base Mask, nodeColor=<rgba 1 0.9375 0 1>}
  n173 = NamedRerouteUsage()
  n174 = Multiply(A=Subtract(A=ComponentMask(None=@n173) {r=true}) {constB=0.5}, B=NamedRerouteUsage())
  n175 = ScalarParameter() {defaultValue=1, parameterName=Width}
  n176 = Multiply(A=Normalize(VectorInput=CrossProduct(A=@n166, B=@n169)), B=Multiply(A=Multiply(A=@n174, B=@n175)))
  n177 = Multiply(A=NamedRerouteUsage(), B=NamedRerouteUsage())
  NamedRerouteDeclaration(None=Multiply(A=Multiply(A=@n176, B=@n177))) {name=SplineThicken WPO, nodeColor=<rgba 0.2891 0 1 1>}
  Power(Base=@n123) {constExponent=4}
  n178 = NamedRerouteUsage()
  PreviousFrameSwitch(Current Frame=@n178, Previous Frame=Add(A=@n178, B=NamedRerouteUsage()))
  n179 = NamedRerouteUsage()
  n180 = VectorParameter() {parameterName=ColourMultiplier, defaultValue=<rgba 1 1 1 0>}
  Reroute(None=MaterialFunctionCall(Emissive=Multiply(A=@n179, B=@n180.RGB)) {materialFunction=&MF_FX_EmissiveAdjustments}.Result)
  n181 = DepthFade(FadeDistance=ScalarParameter() {defaultValue=2, parameterName=Depth Fade Distance})
  n182 = Constant() {r=1}
  n183 = Multiply(A=Multiply(A=QualitySwitch(Default=@n181, Low=@n182, Medium=@n182), B=NamedRerouteUsage()), B=NamedRerouteUsage())
  Saturate(None=Multiply(A=Multiply(A=@n32, B=@n183) {constA=1}, B=ScalarParameter() {defaultValue=0.9, parameterName=Opacity}))
  n184 = NamedRerouteUsage()
  n185 = Power(Base=Multiply(A=NamedRerouteUsage()) {constB=4}) {constExponent=4}
  n186 = Saturate(None=NamedRerouteUsage())
  n187 = TextureSampleParameter2D!(UVs, Bias, Apply View MipBias) {
   mipValueMode=TMVM_MipBias
   parameterName=MaskForLightning
   samplerType=SAMPLERTYPE_LinearColor
   texture=/BRCosmetics/Characters/Player/Male/Medium/Bodies/M_MED_VoyagerCuboid/Textures/T_VoyagerCuboid_StaticLightningMask
  }
  n188 = Multiply(A=Add(A=LinearInterpolate(A=@n184, B=@n185, Alpha=@n186), B=NamedRerouteUsage()), B=@n187.R)
  n189 = Multiply(A=NamedRerouteUsage())
  n190 = Multiply(A=NamedRerouteUsage(), B=Power(Base=NamedRerouteUsage()) {constExponent=1}) {constB=2}
  n191 = Add(A=LinearInterpolate(A=NamedRerouteUsage(), B=@n190, Alpha=Saturate(None=@n190)), B=NamedRerouteUsage())
  n192 = Multiply(A=NamedRerouteUsage())
  SetMaterialAttributes!(
   MaterialAttributes=GetMaterialAttributes(None=NamedRerouteUsage()) {attributeGetTypes [#1]}.MaterialAttributes
   Emissive Color=Multiply(A=Multiply(A=@n188, B=@n189), B=ScalarParameter() {defaultValue=4, parameterName=EmissiveLightningStrength})
   Opacity=Saturate(None=Multiply(A=@n191, B=@n192))
   World Position Offset=NamedRerouteUsage()
  ) {attributeSetTypes [#2, #3, #4]}
  StaticSwitchParameter() {parameterName=Glow}
  TextureCoordinate()
  TextureCoordinate()
  TextureCoordinate()
  TextureCoordinate()
  TextureCoordinate()
  TextureCoordinate() {coordinateIndex=1}
  TextureCoordinate()
  TextureCoordinate()
  TextureCoordinate()
 ]
}
