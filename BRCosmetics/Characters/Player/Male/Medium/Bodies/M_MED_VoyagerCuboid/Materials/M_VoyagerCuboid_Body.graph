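# Skinned body material for the VoyagerCuboid player cosmetic, adding animated, reactive emissive effects between the CharacterParent PreFX and PostFX functions.
# HuntReactivity masks shape panning cloud-noise fire, dot noise and a flickering curve-atlas-animated center glow, positioned using pre-skinned local position.
# ReactiveTime and phase parameters build a burn-up and fade envelope, overridable by IsInFrontEnd.
# A scrolling Signal texture, clipped to a remapped UV rectangle, also feeds emissive.
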Material /BRCosmetics/Characters/Player/Male/Medium/Bodies/M_MED_VoyagerCuboid/Materials/M_VoyagerCuboid_Body
graph {
 properties {
  bUseMaterialAttributes=true
  bUsedWithMeshDeformer=true
  bUsedWithMorphTargets=true
  bUsedWithSkeletalMesh=true
  pixelDepthOffsetMode=PDOM_Legacy
  shadingModel=MSM_FromMaterialExpression
  usedShadingModels=Default Lit | Subsurface | Clear Coat | Subsurface Profile | Cloth
 }
 nodes [
  n1 = Time()
  n2 = ScalarParameter() {defaultValue=7, group=HuntReactivity, parameterName=FlashTimeFrequence}
  ConstantBiasScale(None=Sine(None=Multiply(A=@n1, B=@n2) {constB=3}))
  n3 = MaterialFunctionCall() {materialFunction=&MF_CharacterParent_2023_PreFX}
  n4 = GetMaterialAttributes(None=@n3.Result) {attributeGetTypes [#1]}
  n5 = Multiply(A=@n4.Emissive Color, B=VectorParameter() {group=HuntReactivity, parameterName=EmissiveTint, defaultValue=<rgba 0.9281 0.7757 0.7705 1>}.RGB)
  n6 = NamedRerouteUsage()
  n7 = TextureSampleParameter2D() {
   group=HuntReactivity
   parameterName=AreaControl
   samplerType=SAMPLERTYPE_Masks
   texture=/BRCosmetics/Characters/Player/Male/Medium/Bodies/M_MED_SpacePlunge_Hunt/Textures/T_spaceplungeHunt_New_Body_NewArea
  }
  n8 = VectorParameter() {group=HuntReactivity, parameterName=Subtle, defaultValue=<rgba 1 0.1627 0 1.2>}
  n9 = VectorParameter() {group=HuntReactivity, parameterName=SubtleOutline, defaultValue=<rgba 0.2448 0.0021 0 1>}
  n10 = OneMinus(None=NamedRerouteUsage())
  n11 = Multiply(
   A=Multiply(A=@n6, B=@n7.R)
   B=LinearInterpolate(A=Multiply(A=@n8.RGB, B=@n8.RGB), B=Multiply(A=@n9.RGB, B=@n9.RGB), Alpha=Power(Base=@n10) {constExponent=1})
  )
  n12 = LinearInterpolate(B=ScalarParameter() {defaultValue=0.7, group=HuntReactivity, parameterName=DotIntensity_OG}, Alpha=NamedRerouteUsage()) {constA=1}
  n13 = LinearInterpolate(Alpha=NamedRerouteUsage()) {constA=1, constB=4}
  n14 = VectorParameter() {group=Signal, parameterName=CenterUV, channelNames {a=ScaleY, b=ScaleX, g=CenterY, r=CenterX}, defaultValue=<rgba 0.5 0.5 1 1>}
  MaterialExpressionAppendVector_5 = AppendVector(A=@n14.CenterX, B=@n14.CenterX)
  MaterialExpressionAppendVector_6 = AppendVector(A=@n14.ScaleX, B=@n14.ScaleX)
  MaterialExpressionTextureCoordinate_1 = TextureCoordinate()
  n15 = Custom!(UVPoint=@MaterialExpressionAppendVector_5, UVScale=@MaterialExpressionAppendVector_6, OriginalUV=@MaterialExpressionTextureCoordinate_1) {
   code="float2 DistUV =OriginalUV-UVPoint;\r\nDistUV = DistUV/UVScale + float2(0.5,0.5);\r\n\r\nreturn DistUV;"
   description=CopyToPointUV
   outputType=CMOT_Float2
   inputs [
    {
     inputName=UVPoint
     input {expression=&self:MaterialExpressionAppendVector_5}
    }
    {
     inputName=UVScale
     input {expression=&self:MaterialExpressionAppendVector_6}
    }
    {
     inputName=OriginalUV
     input {expression=&self:MaterialExpressionTextureCoordinate_1}
    }
   ]
  }
  n16 = TextureSampleParameter2D(UVs=Panner(Coordinate=MaterialFunctionCall(XY=@n15) {materialFunction=&Swizzle}.YX) {speedX=-0.1}) {
   group=Signal
   parameterName=SignalTexture
   samplerType=SAMPLERTYPE_Masks
   texture=/BRCosmetics/Characters/Player/Male/Medium/Bodies/M_MED_SpacePlunge/Textures/T_Signal
  }
  n17 = ComponentMask(None=@n15) {r=true}
  n18 = ComponentMask(None=@n15) {g=true}
  n19 = Multiply(
   A=Multiply(A=Step(X=@n17) {constX=0.99, constY=0.01}, B=Step(Y=@n17) {constX=0.99})
   B=Multiply(A=Step(X=@n18) {constX=0.99, constY=0.01}, B=Step(Y=@n18) {constX=0.99})
  )
  n20 = VectorParameter() {group=Signal, parameterName=SignalColor, defaultValue=<rgba 1 1 1 1>}
  n21 = Add(A=Multiply(A=Add(A=@n5, B=@n11), B=Multiply(A=@n12, B=@n13)), B=Multiply(A=Multiply(A=@n16.R, B=@n19), B=Multiply(A=@n20.RGB, B=@n20.RGB)))
  n22 = NamedRerouteUsage()
  MaterialFunctionCall(
   Material=SetMaterialAttributes(MaterialAttributes=@n4.MaterialAttributes, Emissive Color=Add(A=@n21, B=@n22)) {attributeSetTypes [#1]}
   M=@n3.M
   Pre Skinned Local Position=@n3.M
   Distance Blend=@n3.M
   Fresnel=@n3.M
  ) {materialFunction=&MF_CharacterParent_2023_PostFX}
  MaterialFunctionCall() {materialFunction=/Game/MF__FloatFlickerRandom_Generic}
  n23 = NamedRerouteUsage()
  Multiply(A=@n10, B=@n23)
  n24 = TextureSampleParameter2D() {
   group=HuntReactivity
   parameterName=HuntMask01
   samplerType=SAMPLERTYPE_LinearColor
   texture=/BRCosmetics/Characters/Player/Male/Medium/Bodies/M_MED_SpacePlunge_Hunt/Textures/T_spaceplungeHunt_Body_TechArt
  }
  NamedRerouteDeclaration(None=@n24.R) {name=AreaMask, nodeColor=<rgba 1 0.5625 0 1>}
  NamedRerouteDeclaration(None=@n24.G) {name=EdgeMask, nodeColor=<rgba 0.2422 0 1 1>}
  n25 = ComponentMask(None=NamedRerouteUsage()) {b=true, r=true}
  n26 = Divide(A=@n25, B=ScalarParameter() {defaultValue=10, group=HuntReactivity, parameterName=UnitTiling})
  n27 = TextureObject() {texture=/Game/Global/Textures/Default/Noise/T_Noise_Cloud}
  NamedRerouteDeclaration(None=TextureSample(UVs=Panner(Coordinate=@n26) {speedY=-1}, Tex=@n27) {samplerSource=SSM_Wrap_WorldGroupSettings, texture=&T_Noise_Cloud}.R) {name=PanningFire01, nodeColor=<rgba 0.5469 1 0 1>}
  n28 = Panner(Coordinate=Divide(A=@n25, B=ScalarParameter() {defaultValue=40, group=HuntReactivity, parameterName=UnitTiling02})) {speedY=-0.3}
  NamedRerouteDeclaration(None=Multiply(
    A=ScalarParameter() {defaultValue=0.3, group=HuntReactivity, parameterName=FireIntensity02}
    B=TextureSample(UVs=@n28, Tex=@n27) {samplerSource=SSM_Wrap_WorldGroupSettings, texture=&T_Noise_Cloud}.G
   )) {name=PanningFire02, nodeColor=<rgba 0.5469 1 0 1>}
  n29 = Sine(None=Multiply(A=Time(), B=ScalarParameter() {defaultValue=0.3, group=HuntReactivity, parameterName=IdleTimeSpeed}))
  NamedRerouteDeclaration(None=SmoothStep(Value=ConstantBiasScale(None=@n29))) {name=GradientChangeAlpha, nodeColor=<rgba 1 0 0.1406 1>}
  NamedRerouteDeclaration(None=@n24.B) {name=GradientMask, nodeColor=<rgba 0.0781 1 0 1>}
  n30 = Saturate(None=LinearInterpolate(
    A=NamedRerouteUsage()
    B=ScalarParameter() {defaultValue=1, group=HuntReactivity, parameterName=ReactiveAlpha}
    Alpha=ScalarParameter() {group=HuntReactivity, parameterName=IsInFrontEnd}
   ))
  n31 = Multiply(
   A=SphereMask(A=NamedRerouteUsage(), B=Constant() {r=1}, Radius=Multiply(A=@n30) {constB=2}) {hardnessPercent=0}
   B=Power(Base=@n30) {constExponent=0.5}
  )
  n32 = NamedRerouteUsage()
  n33 = LinearInterpolate(Alpha=NamedRerouteUsage()) {constA=0.4, constB=0}
  n34 = NamedRerouteUsage()
  n35 = NamedRerouteUsage()
  n36 = NamedRerouteUsage()
  n37 = Add(A=Multiply(A=NamedRerouteUsage(), B=SmoothStep(Min=@n33, Value=@n34)), B=Multiply(A=NamedRerouteUsage(), B=Multiply(A=@n35, B=@n36)))
  n38 = ScalarParameter() {defaultValue=1, group=HuntReactivity, parameterName=CubeEM_Intensity}
  n39 = VectorParameter() {group=HuntReactivity, parameterName=CentralEM_High, defaultValue=<rgba 1 0.0713 0 10>}
  n40 = Multiply(A=@n39.RGB, B=@n39.RGB)
  n41 = VectorParameter() {group=HuntReactivity, parameterName=CentralEM_Low, defaultValue=<rgba 1 0 0 2>}
  n42 = Multiply(A=@n41.RGB, B=@n41.RGB)
  n43 = OneMinus(None=Power(Base=@n34) {constExponent=0.5})
  n44 = SmoothStep(Max=LinearInterpolate(Alpha=NamedRerouteUsage()) {constA=0.5}, Value=Saturate(None=Add(A=@n32, B=@n43)))
  n45 = Multiply(
   A=LinearInterpolate(Alpha=OneMinus(None=@n32)) {constA=0.3}
   B=Multiply(A=Multiply(A=@n37, B=@n38), B=LinearInterpolate(A=@n40, B=@n42, Alpha=@n44))
  )
  n46 = LinearInterpolate(B=ScalarParameter() {defaultValue=0.3, parameterName=DotIntensity}, Alpha=NamedRerouteUsage()) {constA=1}
  n47 = ScalarParameter() {defaultValue=2, group=HuntReactivity, parameterName=ShoulderIntensity}
  n48 = NamedRerouteUsage()
  n49 = Add(
   A=Multiply(A=Multiply(A=@n45, B=@n46), B=LinearInterpolate(B=@n47, Alpha=@n48) {constA=1, constB=0.3})
   B=Multiply(A=@n40, B=NamedRerouteUsage())
  )
  NamedRerouteDeclaration(None=Multiply(A=Multiply(A=@n31, B=@n49), B=NamedRerouteUsage())) {name=EM_OUT, nodeColor=<rgba 0.6406 0 1 1>}
  n50 = TextureCoordinate()
  n51 = TextureObjectParameter() {
   group=HuntReactivity
   parameterName=HuntNoiseTexture
   samplerType=SAMPLERTYPE_Masks
   texture=/BRCosmetics/Characters/Player/Male/Medium/Bodies/M_MED_SpacePlunge_Hunt/Textures/T_SpaceHuntNoise
  }
  n52 = TextureSample(UVs=@n50, Tex=@n51) {
   samplerSource=SSM_Wrap_WorldGroupSettings
   samplerType=SAMPLERTYPE_Masks
   texture=&T_SpaceHuntNoise
  }
  NamedRerouteDeclaration(None=@n52.B) {name=AdjustmentArea, nodeColor=<rgba 0.4062 0 1 1>}
  n53 = TextureSample(UVs=Multiply(A=@n50, B=ScalarParameter() {defaultValue=5, group=HuntReactivity, parameterName=DotTiling}), Tex=@n51) {
   samplerSource=SSM_Wrap_WorldGroupSettings
   samplerType=SAMPLERTYPE_Masks
   texture=&T_SpaceHuntNoise
  }
  n54 = Multiply(A=@n53.G) {constB=3}
  NamedRerouteDeclaration(None=LinearInterpolate(A=Saturate(None=@n54), Alpha=NamedRerouteUsage()) {constB=0}) {name=DotNoise, nodeColor=<rgba 0 0.7187 1 1>}
  NamedRerouteDeclaration(None=@n52.R) {name=CubeThickness, nodeColor=<rgba 0.3359 1 0 1>}
  NamedRerouteDeclaration(None=@n3.Pre skinned Local Position) {name=PreskinnedPOS, nodeColor=<rgba 0 0.5312 1 1>}
  n55 = NamedRerouteUsage()
  n56 = NamedRerouteUsage()
  n57 = ComponentMask(None=Subtract(A=NamedRerouteUsage(), B=VectorParameter() {group=HuntReactivity, parameterName=GlowingCenter, defaultValue=<rgba 0 0 120 1>}.RGB)) {b=true, r=true}
  n58 = ScalarParameter() {defaultValue=0.06, group=HuntReactivity, parameterName=FlashTimeSpeed}
  n59 = CurveAtlasRowParameter(CurveTime=Frac(None=Multiply(A=@n1, B=@n58))) {
   atlas=/BRCosmetics/Characters/Player/Male/Medium/Bodies/M_MED_SpacePlunge_Hunt/Textures/CA_AnimateLight
   curve=/BRCosmetics/Characters/Player/Male/Medium/Bodies/M_MED_SpacePlunge_Hunt/Textures/C_AnimateLight
   group=HuntReactivity
   parameterName=FlashTime
  }
  n60 = Multiply(
   A=ScalarParameter() {defaultValue=30, group=HuntReactivity, parameterName=OffsetMax}
   B=Subtract(A=ComponentMask(None=@n59) {b=true, g=true}, B=Constant2Vector() {g=0.5, r=0.5})
  )
  n61 = NamedRerouteUsage()
  n62 = Divide(
   A=Length(None=Add(A=@n57, B=@n60)) {is=class MaterialExpressionLength}
   B=LinearInterpolate(
    A=ScalarParameter() {defaultValue=40, group=HuntReactivity, parameterName=CenterGlowRadiusMin}
    B=ScalarParameter() {defaultValue=20, group=HuntReactivity, parameterName=CenterGlowRadiusLarge}
    Alpha=@n61
   )
  )
  n63 = Power(Base=OneMinus(None=Saturate(None=@n62))) {constExponent=4}
  n64 = Multiply(A=SmoothStep(Value=@n61) {constMax=0.5}, B=OneMinus(None=NamedRerouteUsage()))
  n65 = Saturate(None=@n59.A)
  n66 = MaterialFunctionCall() {materialFunction=&MF__FloatFlickerRandom_Generic}
  n67 = Multiply(
   A=Multiply(A=LinearInterpolate(Alpha=@n56) {constA=2, constB=0.2}, B=Multiply(A=@n63, B=@n64))
   B=Multiply(
    A=NamedRerouteDeclaration(None=@n65) {name=ShinningAlpha, nodeColor=<rgba 1 0.2578 0 1>}
    B=LinearInterpolate(Alpha=@n66.RandomFlickerFloat) {constA=0.2}
   )
  )
  NamedRerouteDeclaration(None=Multiply(A=Multiply(A=@n55, B=@n67), B=ScalarParameter() {defaultValue=5, parameterName=CenterGlowMulti}) {constB=2}) {name=CenterGlow, nodeColor=<rgba 1 0.0234 0 1>}
  n68 = Time()
  n69 = ScalarParameter() {group=HuntReactivity, parameterName=ReactiveTime}
  n70 = ScalarParameter() {defaultValue=8, group=HuntReactivity, parameterName=ReactivePhase}
  n71 = Saturate(None=Divide(A=Subtract(A=@n68, B=@n69), B=@n70))
  n72 = Multiply(A=@n71, B=ScalarParameter() {defaultValue=16, parameterName=BurnUpSpeed})
  n73 = ScalarParameter() {defaultValue=2, group=HuntReactivity, parameterName=EndPhase}
  n74 = Divide(A=Subtract(A=@n70, B=@n73), B=@n70)
  n75 = Multiply(A=Subtract(A=@n71, B=@n74), B=Divide(A=@n70, B=@n73))
  NamedRerouteDeclaration(None=LinearInterpolate(A=Saturate(None=@n72), B=OneMinus(None=@n75), Alpha=Step(Y=@n74, X=@n71))) {name=ReactiveAlpha, nodeColor=<rgba 1 0 0.0469 1>}
  NamedRerouteUsage()
  NamedRerouteUsage()
  OneMinus(None=@n23)
  Power(Base=@n6)
 ]
}
