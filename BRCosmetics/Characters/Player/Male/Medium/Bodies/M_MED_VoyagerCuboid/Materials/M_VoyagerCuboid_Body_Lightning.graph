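# Translucent, two-sided lightning bolt material for Niagara effects on the VoyagerCuboid player body, with dynamic parameters for width, tiling, dissolve and distortion.
# Camera-facing spline thickening, camera offset and platform-switched noise distortion build world position offset.
# A panned bolt texture supplies layered color and opacity, or a static Glow switch substitutes an edge-mask glow.
# UV noise, blended quantized and choppy time, erosion dissolve, edge fades and depth fade add flicker and shaping.
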Material /BRCosmetics/Characters/Player/Male/Medium/Bodies/M_MED_VoyagerCuboid/Materials/M_VoyagerCuboid_Body_Lightning
graph {
 properties {
  bAutomaticallySetUsageInEditor=false
  bUsedWithNiagaraMeshParticles=true
  bUsedWithNiagaraSprites=true
  blendMode=BLEND_Translucent
  opacityMaskClipValue=0.1
  pixelDepthOffsetMode=PDOM_Legacy
  shadingModel=MSM_FromMaterialExpression
  twoSided=true
  usedShadingModels=Default Lit
 }
 nodes [
  n1 = ComponentMask(None=NamedRerouteUsage()) {g=true}
  n2 = Subtract(A=@n1, B=NamedRerouteUsage())
  n3 = ScalarParameter() {defaultValue=0.1, parameterName=Opacity - Fade Size Tip}
  NamedRerouteDeclaration(None=Saturate(None=Divide(A=@n2, B=@n3))) {name=Length Mask, nodeColor=<rgba 1 0 0.8437 1>}
  n4 = TextureCoordinate()
  n5 = MaterialFunctionCall(Float2=@n4) {materialFunction=&BreakOutFloat2Components}
  n6 = OneMinus(None=@n5.G)
  n7 = DynamicParameter() {defaultValue=<rgba 10 0 0 0>, paramNames [Lightning Width, Lightning Tiling, Lightning Pan, Lightning Invert]}
  n8 = Saturate(None=@n7.Lightning Invert)
  NamedRerouteDeclaration(None=LinearInterpolate(
    A=@n4
    B=MaterialFunctionCall(X=@n5.R, Y=@n6) {materialFunction=&MakeFloat2}.Result
    Alpha=Round(None=@n8)
   ) {constAlpha=1}) {name=UVs, nodeColor=<rgba 1 0.5859 0 1>}
  n9 = TextureObjectParameter() {group=TextureWPONoise, parameterName=TextureWPONoise_Texture, sortPriority=0, texture=&T_NoisePack_01}
  n10 = ParticleRandom()
  n11 = VectorParameter() {
   group=TextureWPONoise
   parameterName=TextureWPONoise_UVControl
   channelNames {a=VPan, b=UPan, g=VScale, r=UScale}
   defaultValue=<rgba 0 0.75 0 0.02>
  }
  n12 = Add(A=Time() {bOverride_Period=true, period=6666}, B=@n10)
  n13 = NamedRerouteUsage()
  n14 = DynamicParameter() {parameterIndex=2, defaultValue=<rgba -1 -1 1 1>, paramNames [WPO Distort, Dissolve, Scale Z, Param4]}
  n15 = Multiply(
   A=Add(
    A=MaterialFunctionCall(Texture=@n9, AdditionalUVOffset=@n10, UVControl_RGB=@n11, UVControl_A=@n11, "[Override]TimeInput"=@n12, "[Override]UVInput"=@n13) {materialFunction=&FXCommon_TextureSample}.RGB
    B=ScalarParameter() {defaultValue=-0.5, parameterName=TextureWPONoise_Offset}
   )
   B=Add(A=@n14.WPO Distort, B=ScalarParameter() {parameterName=TextureWPONoise_BaseNoiseIntensity})
  )
  n16 = MaterialFunctionCall(Float2=TextureCoordinate()) {materialFunction=&BreakOutFloat2Components}
  n17 = VectorParameter() {
   group=TextureWPONoise
   parameterName=TextureWPONoise_EdgeMask_InvertControl
   channelNames {a="", b=Power, g=Multiply, r=Offset}
   defaultValue=<rgba 0 5 1 1>
  }
  n18 = VectorParameter() {
   group=TextureWPONoise
   parameterName=TextureWPONoise_EdgeMask_Control
   channelNames {a="", b=Power, g=Multiply, r=Offset}
   defaultValue=<rgba 0 5 1 1>
  }
  n19 = Transform(None=Multiply(
    A=Multiply(A=@n15)
    B=MaterialFunctionCall!(UorV=@n16.G, Invert_EdgeControl=@n17, NoInvert_EdgeControl=@n18) {materialFunction=/Game/Effects/Niagara/MasterMaterials/Functions/Common/FXCommon_EdgeMask_Ex}.Mask
   )) {transformSourceType=TRANSFORMSOURCE_Local}
  n20 = Constant3Vector()
  NamedRerouteDeclaration(None=MaterialFunctionCall!("Xb1,PS4"=@n19, "PC-High&&Epic,XSX,PS5"=@n19, "PC-Perf,Android"=@n20, Switch=@n20, "PC-Low(NotPerf)&&Medium"=@n19) {materialFunction=/Game/Effects/Niagara/MasterMaterials/Functions/Common/PlatformSwitch_3D}.Result) {name=DistortWPO, nodeColor=<rgba 1 0.7031 0 1>}
  n21 = MaterialFunctionCall(Float2=TextureCoordinate()) {materialFunction=&BreakOutFloat2Components}
  n22 = Constant() {r=0.5}
  n23 = TextureObjectParameter() {
   group=DissolveTexture
   parameterName=DissolveTexture_Texture
   samplerType=SAMPLERTYPE_LinearColor
   sortPriority=0
   texture=/Engine/Functions/Engine_MaterialFunctions02/ExampleContent/Textures/OffsetNoiseDistanceFields_ForNormals
  }
  n24 = VectorParameter() {
   group=DissolveTexture
   parameterName=DissolveTexture_UVControl
   channelNames {a=VPan, b=UPan, g=VScale, r=UScale}
   defaultValue=<rgba 0 0.5 0 0.5>
  }
  n25 = MaterialFunctionCall(Texture=@n23, AdditionalUVOffset=ParticleRandom(), UVControl_RGB=@n24, UVControl_A=@n24, "[Override]UVInput"=TextureCoordinate()) {materialFunction=&FXCommon_TextureSample}
  n26 = DynamicParameter() {parameterIndex=2, defaultValue=<rgba -1 -1 1 1>, paramNames [WPO Distort, Dissolve, Scale Z, Param4]}
  n27 = Constant() {r=0.6}
  NamedRerouteDeclaration(None=MaterialFunctionCall!(
    Texture=SphereMask(A=@n21.R, B=@n22, Radius=@n22) {hardnessPercent=0}
    Erosion=MaterialFunctionCall(Texture=@n25.R, Erosion=@n26.Dissolve, Softness=@n27) {materialFunction=/Game/Effects/Niagara/MasterMaterials/Functions/MF_Erosion_Smoothstep_01}.Result
    Softness=Constant() {r=0.3}
   ) {materialFunction=&MF_Erosion_Smoothstep_01}.Result) {name=Dissolve, nodeColor=<rgba 0.2422 0 1 1>}
  n28 = Multiply(A=NamedRerouteUsage(), B=NamedRerouteUsage())
  n29 = NamedRerouteUsage()
  n30 = NamedRerouteUsage()
  n31 = NamedRerouteUsage()
  NamedRerouteDeclaration(None=Multiply(A=StaticSwitchParameter(True=@n30, False=@n31) {parameterName=Glow}, B=Multiply(A=@n28, B=@n29))) {name=BaseOpacity, nodeColor=<rgba 0.7578 1 0 1>}
  n32 = TextureCoordinate()
  n33 = VectorParameter() {group=em, parameterName=Glow_EdgeMask_Control, channelNames {a="", b=Power, g=Multiply, r=Offset}, defaultValue=<rgba 0 2 2 1>}
  n34 = VectorParameter() {parameterName=Glow_Refine, defaultValue=<rgba 4 2 0.5 0.5>}
  n35 = MaterialFunctionCall(
   Texture=MaterialFunctionCall!(
    UorV=MaterialFunctionCall(Float2=@n32) {materialFunction=&BreakOutFloat2Components}.R
    EdgeControl=@n33
   ) {materialFunction=/Game/Effects/Niagara/MasterMaterials/Functions/Common/FXCommon_EdgeMask_Simple}.Mask
   RGB=@n34.RGB
   A=@n34.RGB
  ) {materialFunction=&MF_FX_Refine}
  n36 = MaterialFunctionCall(Float2=TextureCoordinate() {coordinateIndex=1}) {materialFunction=&BreakOutFloat2Components}
  n37 = ScalarParameter() {defaultValue=3, parameterName=Glow_ColorFromSpike_Power}
  n38 = Multiply(
   A=Add(A=@n35.Result)
   B=LinearInterpolate(
    A=VectorParameter() {parameterName=Glow_ColorA, defaultValue=<rgba 1 0.6801 0.1274 1>}.RGB
    B=VectorParameter() {parameterName=Glow_ColorB, defaultValue=<rgba 0.4 0.196 0.04 1>}.RGB
    Alpha=Power(Base=@n36.G, Exp=@n37)
   )
  )
  n39 = ScalarParameter() {defaultValue=2, parameterName=Glow_BrogjtmessFromSpike_Power}
  n40 = Add(A=Multiply(A=Power(Base=@n36.R, Exp=@n39), B=ScalarParameter() {defaultValue=0.1, parameterName=Glow_BrogjtmessFromSpike_Str}))
  n41 = VectorParameter() {parameterName=GlowColor, defaultValue=<rgba 0.5 0.1667 0 1>}
  n42 = TextureObjectParameter() {
   group=LightningT
   parameterName=LightningT_Texture
   sortPriority=0
   texture=/Game/Effects/Fort_Effects/GameplayPulginsAssets/ToonLightning/MasterMaterial/V1/T_LightningSplineTexture
  }
  n43 = VectorParameter() {group=LightningT, parameterName=LightningT_UVControl, channelNames {a=VPan, b=UPan, g=VScale, r=UScale}, defaultValue=<rgba 1 1.5 0 0.666>}
  n44 = NamedRerouteUsage()
  n45 = Multiply(A=ScalarParameter() {defaultValue=1, parameterName=Lightning UV Width Scale}, B=Multiply(A=NamedRerouteUsage(), B=NamedRerouteUsage()))
  n46 = DynamicParameter() {defaultValue=<rgba 10 0 0 0>, paramNames [Lightning Width, Lightning Tiling, Lightning Pan, Lightning Invert]}
  n47 = MaterialFunctionCall(Float2=MaterialFunctionCall(UVs=NamedRerouteUsage(), Texture Scale=AppendVector(A=@n45, B=@n46.Lightning Tiling)) {materialFunction=&ScaleUVsByCenter}.UVs) {materialFunction=&BreakOutFloat2Components}
  n48 = AppendVector(A=Constant(), B=Divide(A=ScalarParameter() {defaultValue=0.2, parameterName=Lightning Texture Pan Speed}, B=@n46.Lightning Tiling))
  n49 = Panner(
   Coordinate=MaterialFunctionCall(X=Saturate(None=@n47.R), Y=@n47.G) {materialFunction=&MakeFloat2}.Result
   Time=Time()
   Speed=@n48
  ) {bFractionalPart=true, speedY=1.1}
  n50 = Constant()
  n51 = ParticleRandom()
  n52 = MaterialFunctionCall!(
   Texture=@n42
   AdditionalUVOffset
   UVControl_RGB=@n43
   UVControl_A=@n43
   "[Override]TimeInput"
   "[Override]ScaleUVByCenter"
   "[Override]UVInput"=Add(A=Add(A=@n44, B=@n49), B=AppendVector(A=@n50, B=@n51))
  ) {materialFunction=/Game/Effects/Niagara/MasterMaterials/Functions/Common/FXCommon_TextureSample_V2}
  n53 = VectorParameter() {parameterName=InnerGlowRefine, defaultValue=<rgba 5 2 1 0>}
  n54 = MaterialFunctionCall(Texture=OneMinus(None=@n52.G), RGB=@n53.RGB, A=@n53.RGB) {materialFunction=&MF_FX_Refine}
  n55 = LinearInterpolate(
   A=VectorParameter() {parameterName=BoltColor, defaultValue=<rgba 1.25 0.3517 0.025 1>}.RGB
   B=VectorParameter() {parameterName=InnerGlowColor, defaultValue=<rgba 6 1.2 0 5>}.RGB
   Alpha=Saturate(None=Multiply(A=@n54.Result, B=@n52.B))
  )
  NamedRerouteDeclaration(None=StaticSwitchParameter(True=Multiply(A=@n38, B=@n40), False=LinearInterpolate(A=@n41.RGB, B=@n55, Alpha=@n52.B)) {parameterName=Glow}) {name=TexColor, nodeColor=<rgba 0 1 0.0156 1>}
  NamedRerouteDeclaration(None=@n46.Lightning Pan) {name=LightningPan, nodeColor=<rgba 1 0 0.2344 1>}
  n56 = VectorParameter() {parameterName=OutGlowRefine, defaultValue=<rgba 1 1 0 0>}
  NamedRerouteDeclaration(None=MaterialFunctionCall(Texture=@n52.R, RGB=@n56.RGB, A=@n56.RGB) {materialFunction=&MF_FX_Refine}.Result) {name=OpacityShape, nodeColor=<rgba 1 0 0.0703 1>}
  NamedRerouteDeclaration(None=@n35.Result) {name=GlowShape, nodeColor=<rgba 0 0.3906 1 1>}
  n57 = VectorParameter() {parameterName=UV Noise Scale and Seped, defaultValue=<rgba 0.4 0.4 0.4 1.3>}
  n58 = Multiply(A=TextureCoordinate(), B=ComponentMask(None=@n57.RGB) {g=true, r=true})
  n59 = NamedRerouteUsage()
  n60 = AppendVector(A=@n57.B, B=@n57.B)
  n61 = ConstantBiasScale(None=TextureSample(UVs=Panner(Coordinate=@n58, Time=@n59, Speed=@n60) {bFractionalPart=true, speedY=1}) {samplerType=SAMPLERTYPE_LinearColor, texture=&HighContrastNoise}.R) {bias=-0.5, scale=2}
  n62 = NamedRerouteUsage()
  NamedRerouteDeclaration(None=Multiply(A=Multiply(A=@n61, B=@n62), B=ScalarParameter() {defaultValue=0.03, parameterName=UV Distortion Intensity})) {name=UVDistortion, nodeColor=<rgba 0 0.5547 1 1>}
  n63 = ComponentMask(None=TextureCoordinate()) {r=true}
  n64 = OneMinus(None=Abs(None=ConstantBiasScale(None=@n63) {bias=-0.5, scale=2}))
  n65 = ScalarParameter() {defaultValue=0.2, parameterName=Opacity - Fade Size Width}
  NamedRerouteDeclaration(None=Saturate(None=Divide(A=@n64, B=@n65))) {name=Width Mask, nodeColor=<rgba 0 0.2969 1 1>}
  n66 = CameraPositionWS()
  n67 = Add(A=WorldPosition().XYZ, B=NamedRerouteUsage())
  n68 = Multiply(A=Normalize(VectorInput=Subtract(A=@n66, B=@n67)), B=ScalarParameter() {defaultValue=22, parameterName=Camera Offset})
  n69 = NamedRerouteUsage()
  NamedRerouteDeclaration(None=Add(A=Add(A=@n68, B=@n69), B=NamedRerouteUsage()) {constB=0}) {name=WPO, nodeColor=<rgba 1 0 0.5391 1>}
  n70 = AppendVector(A=@n48, B=Constant())
  n71 = DeltaTime()
  NamedRerouteDeclaration(None=Transform(None=Multiply(A=@n70, B=@n71))) {name=Panner Vector, nodeColor=<rgba 1 0 0.7969 1>}
  n72 = ScalarParameter() {defaultValue=19, parameterName=UV Panner Flicker Frequency}
  n73 = ParticleRandom()
  n74 = Add(A=Time(), B=Frac(None=Multiply(A=@n73) {constB=3.618})) {constB=0}
  n75 = Ceil(None=Multiply(A=@n74, B=@n72) {constB=4})
  n76 = ScalarParameter() {defaultValue=2.62, parameterName=Time Choppiness Frequency}
  n77 = Multiply(A=@n74, B=@n76)
  n78 = ComponentMask(None=VectorNoise(World Position=@n77)) {r=true}
  n79 = Add(A=Multiply(A=Subtract(A=@n78) {constB=0.5}, B=ScalarParameter() {defaultValue=0.9, parameterName=Time Choppiness Randomness Range})) {constB=0.5}
  n80 = Floor(None=@n77)
  n81 = Divide(A=Subtract(A=Subtract(A=@n77, B=@n79), B=@n80), B=OneMinus(None=@n79))
  n82 = Divide(A=Max(A=Add(A=@n81, B=@n80), B=@n80), B=@n76)
  n83 = Divide(A=Subtract(A=@n74, B=@n82), B=@n79)
  NamedRerouteDeclaration(None=LinearInterpolate(A=Divide(A=@n75, B=@n72), B=@n82, Alpha=Saturate(None=@n83)) {desc=Blending between quantized time and choppy time}) {name=Choppy Time, nodeColor=<rgba 0 1 0.9297 1>}
  n84 = Normalize(VectorInput=Subtract(A=CameraPositionWS(), B=WorldPosition().XYZ))
  n85 = Transform(None=Constant3Vector() {constant=<rgba 0 1 0 0>})
  n86 = DynamicParameter() {
   desc="This only exists to fix the splinethicken meshes on android,\r\napparently the tangents do not properly respond to negative scale there"
   parameterIndex=1
   defaultValue=<rgba 1 1 1 0>
   paramNames [Scale X, Scale Y, Scale Z, ""]
  }
  n87 = ShadingPathSwitch(Default=@n85, Deferred=@n85, Mobile=Multiply(A=@n85, B=@n86.RGB)) {desc=Migrated from FeatureLevelSwitch}
  n88 = DotProduct(A=@n84, B=@n87)
  NamedRerouteDeclaration(None=OneMinus(None=Abs(None=@n88))) {name=Fresnel Fade, nodeColor=<rgba 0 0.9297 1 1>}
  n89 = OneMinus(None=@n1)
  n90 = ScalarParameter() {defaultValue=0.1, parameterName=Opacity - Fade Size Base}
  NamedRerouteDeclaration(None=Saturate(None=Divide(A=@n89, B=@n90))) {name=Base Mask, nodeColor=<rgba 1 0.9375 0 1>}
  n91 = TextureCoordinate()
  n92 = DynamicParameter() {defaultValue=<rgba 10 0 0 0>, paramNames [Lightning Width, Lightning Tiling, Lightning Pan, Lightning Invert]}
  n93 = Multiply(A=Subtract(A=ComponentMask(None=@n91) {r=true}) {constB=0.5}, B=@n92.Lightning Width)
  n94 = ScalarParameter() {defaultValue=1, parameterName=Width}
  n95 = Multiply(A=Normalize(VectorInput=CrossProduct(A=@n84, B=@n87)), B=Multiply(A=Multiply(A=@n93, B=@n94)))
  n96 = Multiply(A=NamedRerouteUsage(), B=NamedRerouteUsage())
  NamedRerouteDeclaration(None=Multiply(A=Multiply(A=@n95, B=@n96))) {name=SplineThicken WPO, nodeColor=<rgba 0.2891 0 1 1>}
  n97 = NamedRerouteUsage()
  PreviousFrameSwitch(Current Frame=@n97, Previous Frame=Add(A=@n97, B=NamedRerouteUsage()))
  n98 = NamedRerouteUsage()
  n99 = DepthFade(FadeDistance=ScalarParameter() {defaultValue=3, parameterName=Depth Fade Distance})
  n100 = Constant() {r=1}
  n101 = Multiply(A=Multiply(A=QualitySwitch(Default=@n99, Low=@n100, Medium=@n100), B=NamedRerouteUsage()), B=NamedRerouteUsage())
  n102 = ParticleColor()
  Reroute(None=Multiply(A=Multiply(A=@n98, B=@n101) {constA=1}, B=@n102.A))
  n103 = NamedRerouteUsage()
  Reroute(None=MaterialFunctionCall(Emissive=Multiply(A=@n103, B=@n102.RGB)) {materialFunction=&MF_FX_EmissiveAdjustments}.Result)
 ]
}
