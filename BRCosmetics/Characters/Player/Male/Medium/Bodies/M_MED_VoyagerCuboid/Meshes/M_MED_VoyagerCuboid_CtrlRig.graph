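# Control Rig for the Voyager Cuboid male body cosmetic, animating its torso cubes and floating fragments on top of the standard Fortnite deform rig.
# Torso cubes are parent-constrained to weighted body bones, then rotated by Ambient Rotation through chained quaternion springs with varying damping.
# Fragments and limb cubes translate from Ambient Traslation via springs, some inverted or scaled, with ShardSlowAmount adjusting some damping.
# Swing/twist alpha and spherical pose readers drive shoulder and lat correctives.
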ControlRigBlueprint /BRCosmetics/Characters/Player/Male/Medium/Bodies/M_MED_VoyagerCuboid/Meshes/M_MED_VoyagerCuboid_CtrlRig
graphs [
 CollapseNode_1_ContainedGraph {nodes [
   RigVMFunctionEntryNode()
   > &RigUnit_ParentConstraint(
    AdvancedSettings=(RotationOrderForFilter=XZY)
    Child=(Type=Bone,Name="base_torso_cube_A")
    Parents="((Item=(Type=Bone,Name=\"clavicle_r\"),Weight=0.0004),(Item=(Type=Bone,Name=\"spine_01\"),Weight=0.006),(Item=(Type=Bone,Name=\"deform_pec_r\"),Weight=0.0064),(Item=(Type=Bone,Name=\"spine_02\"),Weight=0.0782),(Item=(Type=Bone,Name=\"spine_03\"),Weight=0.1909),(Item ...(truncated)"
    Weight=1
   )
   > &RigUnit_ParentConstraint(
    AdvancedSettings=(RotationOrderForFilter=XZY)
    Child=(Type=Bone,Name="base_torso_cube_B")
    Parents="((Item=(Type=Bone,Name=\"spine_01\"),Weight=0.0033),(Item=(Type=Bone,Name=\"clavicle_r\"),Weight=0.0046),(Item=(Type=Bone,Name=\"deform_upperarm_fr_r\"),Weight=0.0128),(Item=(Type=Bone,Name=\"spine_02\"),Weight=0.0587),(Item=(Type=Bone,Name=\"deform_pec_r\"),Weight= ...(truncated)"
    Weight=1
   )
   > &RigUnit_ParentConstraint(
    AdvancedSettings=(RotationOrderForFilter=XZY)
    Child=(Type=Bone,Name="base_torso_cube_C")
    Parents="((Item=(Type=Bone,Name=\"clavicle_r\"),Weight=0.0004),(Item=(Type=Bone,Name=\"spine_01\"),Weight=0.006),(Item=(Type=Bone,Name=\"deform_pec_r\"),Weight=0.0064),(Item=(Type=Bone,Name=\"spine_02\"),Weight=0.0782),(Item=(Type=Bone,Name=\"spine_03\"),Weight=0.1909),(Item ...(truncated)"
    Weight=1
   )
   > &RigUnit_ParentConstraint(
    AdvancedSettings=(RotationOrderForFilter=XZY)
    Child=(Type=Bone,Name="base_torso_cube_D")
    Parents="((Item=(Type=Bone,Name=\"spine_01\"),Weight=0.0033),(Item=(Type=Bone,Name=\"clavicle_r\"),Weight=0.0046),(Item=(Type=Bone,Name=\"deform_upperarm_fr_r\"),Weight=0.0128),(Item=(Type=Bone,Name=\"spine_02\"),Weight=0.0587),(Item=(Type=Bone,Name=\"deform_pec_r\"),Weight= ...(truncated)"
    Weight=1
   )
   > &RigUnit_ParentConstraint(
    AdvancedSettings=(RotationOrderForFilter=XZY)
    Child=(Type=Bone,Name="base_torso_cube_E")
    Parents="((Item=(Type=Bone,Name=\"deform_upperarm_fr_r\"),Weight=0.0004),(Item=(Type=Bone,Name=\"clavicle_r\"),Weight=0.0007),(Item=(Type=Bone,Name=\"deform_pec_r\"),Weight=0.0232),(Item=(Type=Bone,Name=\"spine_01\"),Weight=0.0313),(Item=(Type=Bone,Name=\"spine_02\"),Weight= ...(truncated)"
    Weight=1
   )
   > RigVMFunctionReturnNode()
   &RigUnit_ParentConstraint(
    AdvancedSettings=(RotationOrderForFilter=XZY)
    Child=(Type=Bone,Name="base_torso_cube_C")
    ExecutePin=()
    Parents="((Item=(Type=Bone,Name=\"clavicle_r\"),Weight=0.0004),(Item=(Type=Bone,Name=\"spine_01\"),Weight=0.006),(Item=(Type=Bone,Name=\"deform_pec_r\"),Weight=0.0064),(Item=(Type=Bone,Name=\"spine_02\"),Weight=0.0782),(Item=(Type=Bone,Name=\"spine_03\"),Weight=0.1909),(Item ...(truncated)"
    Weight=1
   )
   &RigUnit_ParentConstraint(
    AdvancedSettings=(RotationOrderForFilter=XZY)
    Child=(Type=Bone,Name="base_torso_cube_A")
    ExecutePin=()
    Parents="((Item=(Type=Bone,Name=\"clavicle_r\"),Weight=0.0004),(Item=(Type=Bone,Name=\"spine_01\"),Weight=0.006),(Item=(Type=Bone,Name=\"deform_pec_r\"),Weight=0.0064),(Item=(Type=Bone,Name=\"spine_02\"),Weight=0.0782),(Item=(Type=Bone,Name=\"spine_03\"),Weight=0.1909),(Item ...(truncated)"
    Weight=1
   )
   &RigUnit_ParentConstraint(
    AdvancedSettings=(RotationOrderForFilter=XZY)
    Child=(Type=Bone,Name="base_torso_cube_D")
    ExecutePin=()
    Parents="((Item=(Type=Bone,Name=\"spine_01\"),Weight=0.0033),(Item=(Type=Bone,Name=\"clavicle_r\"),Weight=0.0046),(Item=(Type=Bone,Name=\"deform_upperarm_fr_r\"),Weight=0.0128),(Item=(Type=Bone,Name=\"spine_02\"),Weight=0.0587),(Item=(Type=Bone,Name=\"deform_pec_r\"),Weight= ...(truncated)"
    Weight=1
   )
   &RigUnit_ParentConstraint(
    AdvancedSettings=(RotationOrderForFilter=XZY)
    Child=(Type=Bone,Name="base_torso_cube_E")
    ExecutePin=()
    Parents="((Item=(Type=Bone,Name=\"deform_upperarm_fr_r\"),Weight=0.0004),(Item=(Type=Bone,Name=\"clavicle_r\"),Weight=0.0007),(Item=(Type=Bone,Name=\"deform_pec_r\"),Weight=0.0232),(Item=(Type=Bone,Name=\"spine_01\"),Weight=0.0313),(Item=(Type=Bone,Name=\"spine_02\"),Weight= ...(truncated)"
    Weight=1
   )
   &RigUnit_ParentConstraint(
    AdvancedSettings=(RotationOrderForFilter=XZY)
    Child=(Type=Bone,Name="base_torso_cube_B")
    ExecutePin=()
    Parents="((Item=(Type=Bone,Name=\"spine_01\"),Weight=0.0033),(Item=(Type=Bone,Name=\"clavicle_r\"),Weight=0.0046),(Item=(Type=Bone,Name=\"deform_upperarm_fr_r\"),Weight=0.0128),(Item=(Type=Bone,Name=\"spine_02\"),Weight=0.0587),(Item=(Type=Bone,Name=\"deform_pec_r\"),Weight= ...(truncated)"
    Weight=1
   )
  ]}
 CollapseNode_ContainedGraph {nodes [
   n1 = RigVMFunctionEntryNode()
   RigVMFunction_MathQuaternionFromRotatorV2 = /Script/RigVM.RigVMFunction_MathQuaternionFromRotatorV2(Value=RigVMVariableNode() {variable=Ambient Rotation}.Value) {is=struct RigVMUnitNode, template="Cast::Execute(in Value,out Result)"}
   &RigUnit_ModifyTransforms(
    ItemToModify=((Item=(Type=Bone,Name="dyn_torso_cube_A"),Transform=(Rotation=(X=0,Y=0,Z=0,W=1),Translation=(X=0,Y=0,Z=0),Scale3D=(X=1,Y=1,Z=1))))
    Weight=1
    WeightMinimum=0
    ExecutePin=@n1.ExecutePin
    ItemToModify.0.Transform.Rotation=@RigVMFunction_MathQuaternionFromRotatorV2.Result
   )
   n2 = /Script/ControlRig.RigUnit_SpringInterpQuaternionV2(
    CriticalDamping=0.5
    Current=(X=0,Y=0,Z=0,W=1)
    Strength=1
    TargetVelocityAmount=0
    Torque=(X=0,Y=0,Z=0)
    bInitializeFromTarget=False
    bUseCurrentInput=False
    Target=@RigVMFunction_MathQuaternionFromRotatorV2.Result
   ) {
    is=struct RigVMUnitNode
    template="SpringInterp::Execute(in Target,in Strength,in CriticalDamping,in Torque,in bUseCurrentInput,in Current,in TargetVelocityAmount,in bInitializeFromTarget,out Result,out AngularVelocity)"
   }
   > &RigUnit_ModifyTransforms(
    ItemToModify=((Item=(Type=Bone,Name="dyn_torso_cube_B"),Transform=(Rotation=(X=0,Y=0,Z=0,W=1),Translation=(X=0,Y=0,Z=0),Scale3D=(X=1,Y=1,Z=1))))
    Weight=1
    WeightMinimum=0
    ItemToModify.0.Transform.Rotation=@n2.Result
   )
   n3 = &RigUnit_SpringInterpQuaternionV2(
    CriticalDamping=1
    Current=(X=0,Y=0,Z=0,W=1)
    Strength=1
    TargetVelocityAmount=0
    Torque=(X=0,Y=0,Z=0)
    bInitializeFromTarget=False
    bUseCurrentInput=False
    Target=@n2.Result
   ) {template="SpringInterp::Execute(in Target,in Strength,in CriticalDamping,in Torque,in bUseCurrentInput,in Current,in TargetVelocityAmount,in bInitializeFromTarget,out Result,out AngularVelocity)"}
   > &RigUnit_ModifyTransforms(
    ItemToModify=((Item=(Type=Bone,Name="dyn_torso_cube_C"),Transform=(Rotation=(X=0,Y=0,Z=0,W=1),Translation=(X=0,Y=0,Z=0),Scale3D=(X=1,Y=1,Z=1))))
    Weight=1
    WeightMinimum=0
    ItemToModify.0.Transform.Rotation=@n3.Result
   )
   n4 = &RigUnit_SpringInterpQuaternionV2(
    CriticalDamping=0.25
    Current=(X=0,Y=0,Z=0,W=1)
    Strength=1
    TargetVelocityAmount=0
    Torque=(X=0,Y=0,Z=0)
    bInitializeFromTarget=False
    bUseCurrentInput=False
    Target=@n3.Result
   ) {template="SpringInterp::Execute(in Target,in Strength,in CriticalDamping,in Torque,in bUseCurrentInput,in Current,in TargetVelocityAmount,in bInitializeFromTarget,out Result,out AngularVelocity)"}
   > &RigUnit_ModifyTransforms(
    ItemToModify=((Item=(Type=Bone,Name="dyn_torso_cube_D"),Transform=(Rotation=(X=0,Y=0,Z=0,W=1),Translation=(X=0,Y=0,Z=0),Scale3D=(X=1,Y=1,Z=1))))
    Weight=1
    WeightMinimum=0
    ItemToModify.0.Transform.Rotation=@n4.Result
   )
   n5 = &RigUnit_SpringInterpQuaternionV2(
    CriticalDamping=0.1
    Current=(X=0,Y=0,Z=0,W=1)
    Strength=1
    TargetVelocityAmount=0
    Torque=(X=0,Y=0,Z=0)
    bInitializeFromTarget=False
    bUseCurrentInput=False
    Target=@n4.Result
   ) {template="SpringInterp::Execute(in Target,in Strength,in CriticalDamping,in Torque,in bUseCurrentInput,in Current,in TargetVelocityAmount,in bInitializeFromTarget,out Result,out AngularVelocity)"}
   > n6 = &RigUnit_ModifyTransforms(
    ItemToModify=((Item=(Type=Bone,Name="base_torso_cube_E"),Transform=(Rotation=(X=0,Y=0,Z=0,W=1),Translation=(X=0,Y=0,Z=0),Scale3D=(X=1,Y=1,Z=1))))
    Weight=1
    WeightMinimum=0
    ItemToModify.0.Transform.Rotation=@n5.Result
   )
   RigVMFunctionReturnNode(ExecutePin=@n6.ExecutePin)
  ]}
 CollapseNode_ContainedGraph {nodes [
   n1 = RigVMFunctionEntryNode()
   RigVMFunction_Sequence = RigVMAggregateNode(ExecuteContext=@n1.ExecutePin)
   RigVMFunctionReturnNode(ExecutePin=@RigVMFunction_Sequence.C)
   n2 = RigVMVariableNode() {variable=ShardSlowAmount}
   Add = &RigVMFunction_MathDoubleAdd(B=0.2, A=@n2.Value)
   n3 = RigVMVariableNode() {variable=Ambient Traslation}
   n4 = /Script/ControlRig.RigUnit_SpringInterpVectorV2(
    Current=(X=0,Y=0,Z=0)
    Force=(X=0,Y=0,Z=0)
    Strength=1
    TargetVelocityAmount=0
    bInitializeFromTarget=False
    bUseCurrentInput=False
    CriticalDamping=@Add.Result
    Target=@n3.Value
   ) {
    is=struct RigVMUnitNode
    template="SpringInterp::Execute(in Target,in Strength,in CriticalDamping,in Force,in bUseCurrentInput,in Current,in TargetVelocityAmount,in bInitializeFromTarget,out Result,out Velocity)"
   }
   &RigUnit_ModifyTransforms(
    ItemToModify="((Item=(Type=Bone,Name=\"dyn_shoulder_l_floatingFragment_01\"),Transform=(Rotation=(X=0,Y=0,Z=0,W=1),Translation=(X=0,Y=0,Z=0),Scale3D=(X=1,Y=1,Z=1))),(Item=(Type=Bone,Name=\"dyn_arm_l_floatingFragmentf_01\"),Transform=(Rotation=(X=0,Y=0,Z=0,W=1),Translation=( ...(truncated)"
    Weight=1
    WeightMinimum=0
    ItemToModify.0.Transform.Translation=@n4.Result
    ItemToModify.1.Transform.Translation=@n4.Result
    ItemToModify.2.Transform.Translation=@n4.Result
    ItemToModify.3.Transform.Translation=@n4.Result
    ItemToModify.4.Transform.Translation=@n4.Result
    ItemToModify.5.Transform.Translation=@n4.Result
    ExecutePin=@RigVMFunction_Sequence.B
   )
   n5 = /Script/RigVM.RigVMFunction_MathVectorMul(B=(X=-1,Y=-1,Z=-1), A=@n4.Result) {is=struct RigVMUnitNode, template="Multiply::Execute(in A,in B,out Result)"}
   > &RigUnit_ModifyTransforms(
    ItemToModify="((Item=(Type=Bone,Name=\"dyb_shoulder_l_floatingFragment_02\"),Transform=(Rotation=(X=0,Y=0,Z=0,W=1),Translation=(X=0,Y=0,Z=0),Scale3D=(X=1,Y=1,Z=1))),(Item=(Type=Bone,Name=\"dyn_hand_l_floatingFragmentf_02\"),Transform=(Rotation=(X=0,Y=0,Z=0,W=1),Translation= ...(truncated)"
    Weight=1
    WeightMinimum=0
    ItemToModify.0.Transform.Translation=@n5.Result
    ItemToModify.1.Transform.Translation=@n5.Result
   )
   SpringInterp = &RigUnit_SpringInterpVectorV2(
    Current=(X=0,Y=0,Z=0)
    Force=(X=0,Y=0,Z=0)
    Strength=1
    TargetVelocityAmount=0
    bInitializeFromTarget=False
    bUseCurrentInput=False
    CriticalDamping=&RigVMFunction_MathDoubleAdd(B=0.7, A=@n2.Value).Result
    Target=@n3.Value
   ) {template="SpringInterp::Execute(in Target,in Strength,in CriticalDamping,in Force,in bUseCurrentInput,in Current,in TargetVelocityAmount,in bInitializeFromTarget,out Result,out Velocity)"}
   > &RigUnit_ModifyTransforms(
    ItemToModify="((Item=(Type=Bone,Name=\"base_shoulder_l_floatingFragmentf_03\"),Transform=(Rotation=(X=0,Y=0,Z=0,W=1),Translation=(X=0,Y=0,Z=0),Scale3D=(X=1,Y=1,Z=1))),(Item=(Type=Bone,Name=\"dyn_arm_l_floatingFragmentf_03\"),Transform=(Rotation=(X=0,Y=0,Z=0,W=1),Translation ...(truncated)"
    Weight=1
    WeightMinimum=0
    ItemToModify.0.Transform.Translation=@SpringInterp.Result
    ItemToModify.1.Transform.Translation=@SpringInterp.Result
    ItemToModify.2.Transform.Translation=@SpringInterp.Result
    ItemToModify.3.Transform.Translation=@SpringInterp.Result
    ItemToModify.4.Transform.Translation=@SpringInterp.Result
   )
   n6 = &RigVMFunction_MathVectorMul(B=(X=-2,Y=-2,Z=-2), A=@SpringInterp.Result) {template="Multiply::Execute(in A,in B,out Result)"}
   > &RigUnit_ModifyTransforms(
    ItemToModify="((Item=(Type=Bone,Name=\"dyn_shoulder_l_floatingFragment_04\"),Transform=(Rotation=(X=0,Y=0,Z=0,W=1),Translation=(X=0,Y=0,Z=0),Scale3D=(X=1,Y=1,Z=1))),(Item=(Type=Bone,Name=\"dyn_hand_l_floatingFragmentf_03\"),Transform=(Rotation=(X=0,Y=0,Z=0,W=1),Translation= ...(truncated)"
    Weight=1
    WeightMinimum=0
    ItemToModify.0.Transform.Translation=@n6.Result
    ItemToModify.1.Transform.Translation=@n6.Result
   )
   n7 = &RigUnit_SpringInterpVectorV2(
    Current=(X=0,Y=0,Z=0)
    Force=(X=0,Y=0,Z=0)
    Strength=1
    TargetVelocityAmount=0
    bInitializeFromTarget=False
    bUseCurrentInput=False
    CriticalDamping=&RigVMFunction_MathDoubleAdd(B=1.5, A=@n2.Value).Result
    Target=@SpringInterp.Result
   ) {template="SpringInterp::Execute(in Target,in Strength,in CriticalDamping,in Force,in bUseCurrentInput,in Current,in TargetVelocityAmount,in bInitializeFromTarget,out Result,out Velocity)"}
   > &RigUnit_ModifyTransforms(
    ItemToModify="((Item=(Type=Bone,Name=\"dyn_body_floatingCube_01\"),Transform=(Rotation=(X=0,Y=0,Z=0,W=1),Translation=(X=0,Y=0,Z=0),Scale3D=(X=1,Y=1,Z=1))),(Item=(Type=Bone,Name=\"dyn_arm_l_floatingFragmentf_02\"),Transform=(Rotation=(X=0,Y=0,Z=0,W=1),Translation=(X=0,Y=0,Z= ...(truncated)"
    Weight=1
    WeightMinimum=0
    ItemToModify.0.Transform.Translation=@n7.Result
    ItemToModify.1.Transform.Translation=@n7.Result
    ItemToModify.2.Transform.Translation=@n7.Result
    ItemToModify.3.Transform.Translation=@n7.Result
    ItemToModify.4.Transform.Translation=@n7.Result
    ItemToModify.5.Transform.Translation=@n7.Result
   )
   &RigUnit_ModifyTransforms(
    ItemToModify="((Item=(Type=Bone,Name=\"dyn_leg_l_floatingCube_02\"),Transform=(Rotation=(X=0,Y=0,Z=0,W=1),Translation=(X=0,Y=0,Z=0),Scale3D=(X=1,Y=1,Z=1))),(Item=(Type=Bone,Name=\"dyn_leg_r_floatingCube_01\"),Transform=(Rotation=(X=0,Y=0,Z=0,W=1),Translation=(X=0,Y=0,Z=0),S ...(truncated)"
    Weight=1
    WeightMinimum=0
    ItemToModify.0.Transform.Translation=@n3.Value
    ItemToModify.1.Transform.Translation=@n3.Value
    ItemToModify.2.Transform.Translation=@n3.Value
    ExecutePin=@RigVMFunction_Sequence.A
   )
   n8 = &RigUnit_SpringInterpVectorV2(
    CriticalDamping=0.7
    Current=(X=0,Y=0,Z=0)
    Force=(X=0,Y=0,Z=0)
    Strength=1
    TargetVelocityAmount=0
    bInitializeFromTarget=False
    bUseCurrentInput=False
    Target=@n3.Value
   ) {template="SpringInterp::Execute(in Target,in Strength,in CriticalDamping,in Force,in bUseCurrentInput,in Current,in TargetVelocityAmount,in bInitializeFromTarget,out Result,out Velocity)"}
   > &RigUnit_ModifyTransforms(
    ItemToModify=((Item=(Type=Bone,Name="dyn_leg_r_floatingCube_03"),Transform=(Rotation=(X=0,Y=0,Z=0,W=1),Translation=(X=0,Y=0,Z=0),Scale3D=(X=1,Y=1,Z=1))))
    Weight=1
    WeightMinimum=0
    ItemToModify.0.Transform.Translation=@n8.Result
   )
   > &RigUnit_ModifyTransforms(
    ItemToModify=((Item=(Type=Bone,Name="dyn_leg_r_floatingCube_05"),Transform=(Rotation=(X=0,Y=0,Z=0,W=1),Translation=(X=0,Y=0,Z=0),Scale3D=(X=1,Y=1,Z=1))))
    Weight=1
    WeightMinimum=0
    ItemToModify.0.Transform.Translation=&RigVMFunction_MathVectorMul(B=(X=-2,Y=-2,Z=-2), A=@n8.Result) {template="Multiply::Execute(in A,in B,out Result)"}.Result
   )
   n9 = &RigUnit_SpringInterpVectorV2(
    CriticalDamping=1.5
    Current=(X=0,Y=0,Z=0)
    Force=(X=0,Y=0,Z=0)
    Strength=1
    TargetVelocityAmount=0
    bInitializeFromTarget=False
    bUseCurrentInput=False
    Target=@n8.Result
   ) {template="SpringInterp::Execute(in Target,in Strength,in CriticalDamping,in Force,in bUseCurrentInput,in Current,in TargetVelocityAmount,in bInitializeFromTarget,out Result,out Velocity)"}
   > &RigUnit_ModifyTransforms(
    ItemToModify="((Item=(Type=Bone,Name=\"dyn_leg_l_floatingCube_01\"),Transform=(Rotation=(X=0,Y=0,Z=0,W=1),Translation=(X=0,Y=0,Z=0),Scale3D=(X=1,Y=1,Z=1))),(Item=(Type=Bone,Name=\"dyn_leg_l_floatingCube_04\"),Transform=(Rotation=(X=0,Y=0,Z=0,W=1),Translation=(X=0,Y=0,Z=0),S ...(truncated)"
    Weight=1
    WeightMinimum=0
    ItemToModify.0.Transform.Translation=@n9.Result
    ItemToModify.1.Transform.Translation=@n9.Result
   )
   n10 = &RigVMFunction_MathVectorMul(B=(X=-3,Y=-3,Z=-3), A=@n9.Result) {template="Multiply::Execute(in A,in B,out Result)"}
   > &RigUnit_ModifyTransforms(
    ItemToModify="((Item=(Type=Bone,Name=\"dyn_leg_l_floatingCube_03\"),Transform=(Rotation=(X=0,Y=0,Z=0,W=1),Translation=(X=0,Y=0,Z=0),Scale3D=(X=1,Y=1,Z=1))),(Item=(Type=Bone,Name=\"dyn_leg_r_floatingCube_02\"),Transform=(Rotation=(X=0,Y=0,Z=0,W=1),Translation=(X=0,Y=0,Z=0),S ...(truncated)"
    Weight=1
    WeightMinimum=0
    ItemToModify.0.Transform.Translation=@n10.Result
    ItemToModify.1.Transform.Translation=@n10.Result
   )
  ]}
 CollapseNode_ContainedGraph {nodes [
   RigVMFunctionEntryNode()
   > GetAlphaFromQuatDelta = RigVMFunctionReferenceNode(
    Rotation Order=ZYX
    child=(Type=Bone,Name="clavicle_l")
    debugRotation=false
    outputMinMax=(X=0,Y=1)
    parent=(Type=Null,Name="clavicle_l_null")
    premult=1
    rotationMinMax=(X=0,Y=-40)
    triggerAxis=Y
    twistAxis=(X=1,Y=0,Z=0)
   )
   > &RigUnit_ModifyTransforms(
    ItemToModify=((Item=(Type=Bone,Name="shoulder_cube_fragments_l"),Transform=(Rotation=(X=0.0444,Y=0.2133,Z=-0.3259,W=0.92),Translation=(X=0,Y=0,Z=0),Scale3D=(X=1,Y=1,Z=1))))
    WeightMinimum=0
    Weight=@GetAlphaFromQuatDelta.alpha
   )
   > n1 = &RigUnit_SphericalPoseReader(
    ActiveRegionScaleFactors=(PositiveWidth=1,NegativeWidth=1,PositiveHeight=1,NegativeHeight=1)
    ActiveRegionSize=0.1
    Debug=(bDrawDebug=false,bDraw2D=False,bDrawLocalAxes=False,DebugScale=25,DebugSegments=20,DebugThickness=0.25)
    DriverAxis=(X=1,Y=0,Z=0)
    DriverItem=(Type=Bone,Name="upperarm_l")
    FalloffRegionScaleFactors=(PositiveWidth=1,NegativeWidth=1,PositiveHeight=1,NegativeHeight=1)
    FalloffSize=0.2
    FlipHeightScaling=False
    FlipWidthScaling=False
    OptionalParentItem=(Type=Bone,Name="None")
    RotationOffset=(X=50,Y=0,Z=-80)
   )
   > &RigUnit_ModifyTransforms(
    ItemToModify=((Item=(Type=Bone,Name="dyn_lat_l"),Transform=(Rotation=(X=0,Y=0,Z=0,W=1),Translation=(X=-6,Y=0,Z=0),Scale3D=(X=1,Y=1,Z=1))))
    WeightMinimum=0
    Weight=@n1.OutputParam
   )
   > n2 = &RigUnit_SphericalPoseReader(
    ActiveRegionScaleFactors=(PositiveWidth=1,NegativeWidth=1,PositiveHeight=1,NegativeHeight=1)
    ActiveRegionSize=0.1
    Debug=(bDrawDebug=false,bDraw2D=False,bDrawLocalAxes=False,DebugScale=25,DebugSegments=20,DebugThickness=0.25)
    DriverAxis=(X=-1,Y=0,Z=0)
    DriverItem=(Type=Bone,Name="upperarm_r")
    FalloffRegionScaleFactors=(PositiveWidth=1,NegativeWidth=1,PositiveHeight=1,NegativeHeight=1)
    FalloffSize=0.2
    FlipHeightScaling=False
    FlipWidthScaling=False
    OptionalParentItem=(Type=Bone,Name="None")
    RotationOffset=(X=-135,Y=0,Z=-80)
   )
   > &RigUnit_ModifyTransforms(
    ItemToModify=((Item=(Type=Bone,Name="base_lat_r"),Transform=(Rotation=(X=0,Y=0,Z=0,W=1),Translation=(X=-5,Y=0,Z=0),Scale3D=(X=1,Y=1,Z=1))))
    WeightMinimum=0
    Weight=@n2.OutputParam
   )
   > RigVMFunctionReturnNode()
  ]}
 New Function_ContainedGraph {nodes [
   RigVMFunctionEntryNode()
   RigVMFunction_MathBoolToInteger = &RigVMFunction_MathBoolToInteger(Value=RigVMVariableNode() {variable=onlyTwist}.Value)
   RigVMFunction_MathQuaternionSwingTwist = &RigVMFunction_MathQuaternionSwingTwist(
    Input=&RigUnit_GetRelativeTransformForItem(
     bChildInitial=False
     bParentInitial=False
     Child=RigVMVariableNode() {variable=child}.Value
     Parent=RigVMVariableNode() {variable=parent}.Value
    ).RelativeTransform.Rotation
    TwistAxis=RigVMVariableNode() {variable=twistAxis}.Value
   )
   n1 = RigVMDispatchNode(
    Values=((),())
    Index=@RigVMFunction_MathBoolToInteger.Result
    Values.0=@RigVMFunction_MathQuaternionSwingTwist.Swing
    Values.1=@RigVMFunction_MathQuaternionSwingTwist.Twist
   ) {template="DISPATCH_RigVMDispatch_SelectInt32(in Index,in Values,out Result)"}
   > RigVMVariableNode(Value=@n1.Result) {variable=entry_outputQuat}
   RigVMFunction_MathQuaternionToEuler = &RigVMFunction_MathQuaternionToEuler(Value=@n1.Result, RotationOrder=RigVMVariableNode() {variable=Rotation Order}.Value)
   > DISPATCH_RigVMDispatch_Print = RigVMDispatchNode(
    ScreenColor=(R=1,G=1,B=1,A=1)
    ScreenDuration=1.5
    Value=@RigVMFunction_MathQuaternionToEuler.Result
    Enabled=RigVMVariableNode() {variable=debugRotation}.Value
   ) {template="DISPATCH_RigVMDispatch_Print(in Prefix,in Value,in Enabled,in ScreenDuration,in ScreenColor)"}
   Cast = RigVMDispatchNode(Value=RigVMVariableNode() {variable=triggerAxis}.Value) {template="DISPATCH_RigVMDispatch_CastEnumToInt(in Value,out Result)"}
   DISPATCH_RigVMDispatch_SelectInt32 = RigVMDispatchNode(
    Values=(0,0,0)
    Index=@Cast.Result
    Values.0=@RigVMFunction_MathQuaternionToEuler.Result.X
    Values.1=@RigVMFunction_MathQuaternionToEuler.Result.Y
    Values.2=@RigVMFunction_MathQuaternionToEuler.Result.Z
   ) {template="DISPATCH_RigVMDispatch_SelectInt32(in Index,in Values,out Result)"}
   n2 = RigVMVariableNode() {variable=rotationMinMax}
   n3 = RigVMVariableNode() {variable=outputMinMax}
   Remap = &RigVMFunction_MathDoubleRemap(
    bClamp=False
    Value=@DISPATCH_RigVMDispatch_SelectInt32.Result
    SourceMinimum=@n2.Value.X
    SourceMaximum=@n2.Value.Y
    TargetMinimum=@n3.Value.X
    TargetMaximum=@n3.Value.Y
   )
   Absolute = &RigVMFunction_MathDoubleAbs(Value=@Remap.Result)
   n4 = RigVMVariableNode() {variable=absoluteValue}
   Multiply = &RigVMFunction_MathDoubleMul(
    A=&RigVMFunction_MathDoubleClamp(Value=RigVMDispatchNode(True=@Absolute.Result, False=@Remap.Result, Condition=@n4.Value) {template="DISPATCH_RigVMDispatch_If(in Condition,in True,in False,out Result)"}.Result).Result
    B=RigVMVariableNode() {variable=premult}.Value
   )
   > RigVMFunctionReturnNode(alpha=@Multiply.Result, rawQuat=RigVMVariableNode() {variable=entry_outputQuat}.Value)
  ]}
 RigVMFunctionLibrary {nodes [GetAlphaFromQuatDelta = RigVMCollapseNode(
    ExecuteContext=()
    Rotation Order=ZYX
    child=(Type=Bone,Name="None")
    outputMinMax=(X=0,Y=1)
    parent=(Type=Bone,Name="None")
    premult=1
    rotationMinMax=(X=0,Y=0)
    twistAxis=(X=1,Y=0,Z=0)
   )]}
 RigVMFunction_Sequence_1_ContainedGraph {nodes [
   n1 = RigVMFunctionEntryNode()
   RigVMFunction_Sequence = &RigVMFunction_Sequence(ExecuteContext=@n1.ExecuteContext)
   n2 = &RigVMFunction_Sequence(ExecuteContext=@RigVMFunction_Sequence.B)
   RigVMFunctionReturnNode(A=@RigVMFunction_Sequence.A, B=@n2.A, C=@n2.B)
  ]}
 RigVMModel {nodes [
   RigUnit_BeginExecution = &RigUnit_BeginExecution()
   /Script/FortniteGame.RigUnit_FortDeformRig(
    Bool=(bUpperArm=true,bElbow=true,bWrist=true,bGroin=true,bGlute=true,bKnee=true,bPec=true,bBicep=false,bLat=false)
    Curve="(bLUpperArmUpAlpha=False,bRUpperArmUpAlpha=False,bLUpperArmFrAlpha=False,bRUpperArmFrAlpha=False,bLUpperArmBkAlpha=False,bRUpperArmBkAlpha=False,bLWristUpAlpha=False,bRWristUpAlpha=False,bLWristDnAlpha=False,bRWristDnAlpha=False,bLKneeAlpha=False,bRKneeAlp ...(truncated)"
    Debug=(bUpperArmDebug=False,bWristDebug=False,bLatDebug=False,bGroinGluteDebug=False)
    Value="(UpperArmUp=(Rotation=(X=0.0,Y=-0.0,Z=0.0,W=1.0),Translation=(X=0.0,Y=0.0,Z=8.0),Scale3D=(X=0.70,Y=1.0,Z=1.0)),UpperArmFwdFr=(Rotation=(X=0.0,Y=-0.0,Z=0.0,W=1.0),Translation=(X=0.0,Y=10.0,Z=0.0),Scale3D=(X=0.50,Y=1.0,Z=1.0)),UpperArmFwdBk=(Rotation=(X=0.0, ...(truncated)"
    ExecutePin=@RigUnit_BeginExecution.ExecutePin
   ) {is=struct RigVMUnitNode}
   > Torso_cube_constraints = RigVMCollapseNode()
   Torso_cube_rotation = RigVMCollapseNode(ExecutePin=@Torso_cube_constraints.ExecuteContext)
   Floating_fragmetns_movement = RigVMCollapseNode(ExecutePin=@Torso_cube_rotation.ExecutePin)
   Correctives_graph = RigVMCollapseNode(ExecuteContext=@Floating_fragmetns_movement.ExecutePin)
  ]}
]
hierarchy [
 root {loc=<0 0 0>, rot=<0 0 0 1>}
 pelvis {parent=root, loc=<0 -0.7071 93.1191>, rot=<0.0127 0.707 -0.0127 -0.707>}
 spine_01 {parent=pelvis, loc=<6.7494 -0.6388 0>, rot=<0 0 0.034 -0.9994>}
 spine_02 {parent=spine_01, loc=<7.2518 -0.035 0>, rot=<0 0 0.0238 -0.9997>}
 spine_03 {parent=spine_02, loc=<6.0104 0 0>, rot=<0 0 0.0149 -0.9999>}
 spine_04 {parent=spine_03, loc=<8.1024 0 0>, rot=<0 0 -0.0132 -0.9999>}
 spine_05 {parent=spine_04, loc=<10.0146 0 0>, rot=<0 0 -0.0015 -1>}
 neck_01 {parent=spine_05, loc=<11.314 0.153 0>, rot=<0 0 -0.2003 -0.9797>}
 neck_02 {parent=neck_01, loc=<4.5627 0 0>, rot=<0 0 0.0504 -0.9987>}
 head {parent=neck_02, loc=<5.3914 0 0>, rot=<0 0 0.115 -0.9934>}
 attach_cape {parent=spine_05, loc=<9.017 -11.714 0>, rot=<0.0616 -0.7044 -0.0616 -0.7044>}
 cape {parent=attach_cape, type=Null, loc=<0 0 0>, rot=<0 0 0 1>}
 attach_backpack {parent=spine_05, loc=<-0.123 -13.473 0>, rot=<-0.0093 -0.707 0.0093 -0.707>}
 Backpack_BR {parent=attach_backpack, type=Null, loc=<0 0 0>, rot=<0 0 0 1>}
 clavicle_l {parent=spine_05, loc=<4.3575 -0.1466 -2.3223>, rot=<0.0465 -0.7289 -0.009 -0.6829>}
 upperarm_l {parent=clavicle_l, loc=<15.5111 0 0>, rot=<0.0226 -0.3234 -0.0088 -0.9459>}
 lowerarm_l {parent=upperarm_l, loc=<26.6514 0 0>, rot=<0.0391 -0.0099 -0.2453 -0.9686>}
 hand_l {parent=lowerarm_l, loc=<22.3491 0 0>, rot=<-0.7081 -0.0507 0.0975 -0.6975>}
 thumb_01_l {parent=hand_l, loc=<1.3118 -0.3191 -2.459>, rot=<0.6546 -0.4042 0.0664 -0.6354>}
 thumb_02_l {parent=thumb_01_l, loc=<5.9336 -0.0448 0.0401>, rot=<-0.0037 0.0213 0.298 -0.9543>}
 thumb_03_l {parent=thumb_02_l, loc=<2.2223 0.3682 -0.0907>, rot=<0.0118 -0.013 0.0066 -0.9998>}
 index_metacarpal_l {parent=hand_l, loc=<1.6535 0.0183 -2.96>, rot=<0.0158 -0.0124 -0.0173 -0.9996>}
 index_01_l {parent=index_metacarpal_l, loc=<8.3008 0.0357 -0.0524>, rot=<-0.0044 0.0485 0.0924 -0.9945>}
 index_02_l {parent=index_01_l, loc=<3.3587 -0.3004 0.0452>, rot=<-0.002 0.0027 0.3225 -0.9466>}
 index_03_l {parent=index_02_l, loc=<2.5578 -0.123 -0.0252>, rot=<-0.002 0.0067 0.2558 -0.9667>}
 middle_metacarpal_l {parent=hand_l, loc=<2.1856 -0.1725 -0.5984>, rot=<0.006 0.0168 -0.0329 -0.9993>}
 middle_01_l {parent=middle_metacarpal_l, loc=<7.3275 -0.2447 0.0129>, rot=<-0.0208 0.0042 0.0858 -0.9961>}
 middle_02_l {parent=middle_01_l, loc=<4.0123 0.1401 0.1175>, rot=<-0.0218 0.0258 0.3512 -0.9357>}
 middle_03_l {parent=middle_02_l, loc=<2.6077 0.0534 0.1373>, rot=<0.0082 0.0188 0.1708 -0.9851>}
 ring_metacarpal_l {parent=hand_l, loc=<2.2074 -0.3625 1.3638>, rot=<-0.0284 0.0722 -0.0458 -0.9959>}
 ring_01_l {parent=ring_metacarpal_l, loc=<7.0295 -0.6197 -0.1446>, rot=<-0.0489 -0.0252 0.1892 -0.9804>}
 ring_02_l {parent=ring_01_l, loc=<3.6174 0.4647 -0.1027>, rot=<0 0 0.2806 -0.9598>}
 ring_03_l {parent=ring_02_l, loc=<2.4152 0.114 0.0199>, rot=<0 0 0.1189 -0.9929>}
 pinky_metacarpal_l {parent=hand_l, loc=<1.7616 -0.9753 3.3814>, rot=<-0.0397 0.0814 -0.069 -0.9935>}
 pinky_01_l {parent=pinky_metacarpal_l, loc=<7.1709 -0.4831 -0.1094>, rot=<-0.0654 -0.0035 0.2242 -0.9723>}
 pinky_02_l {parent=pinky_01_l, loc=<2.8274 0.0893 -0.1043>, rot=<0.0037 0.0011 0.2854 -0.9584>}
 pinky_03_l {parent=pinky_02_l, loc=<2.2046 0.0307 -0.0078>, rot=<-0.0002 0.0098 0.1445 -0.9895>}
 deform_wrist_dn_base_l {parent=hand_l, loc=<0 -2.2342 0>, rot=<0 0 0 -1>}
 deform_wrist_dn_l {parent=deform_wrist_dn_base_l, loc=<0 0 0>, rot=<0 0 0 1>}
 deform_wrist_up_base_l {parent=hand_l, loc=<0 2.2682 0>, rot=<0 0 0 -1>}
 deform_wrist_up_l {parent=deform_wrist_up_base_l, loc=<0 0 0>, rot=<0 0 0 1>}
 lowerarm_twist_02_l {parent=lowerarm_l, loc=<9.6199 0 0>, rot=<0 0 0 -1>}
 lowerarm_twist_01_l {parent=lowerarm_l, loc=<17.8549 0 0>, rot=<0 0 0 -1>}
 deform_elbow_in_base_l {parent=lowerarm_l, loc=<0.2112 5.259 0.8274>, rot=<0 0 0 -1>}
 deform_elbow_in_l {parent=deform_elbow_in_base_l, loc=<0 0 0>, rot=<0 0 0 1>}
 deform_elbow_out_base_l {parent=lowerarm_l, loc=<0 -4 0>, rot=<0 0 0 -1>}
 deform_elbow_out_l {parent=deform_elbow_out_base_l, loc=<0 0 0>, rot=<0 0 0 1>}
 base_hand_l_floatingFragmentf_04 {parent=lowerarm_l, loc=<5.4249 -6.7466 7.355>, rot=<-0.499 0.5673 0.1252 -0.643>}
 dyn_hand_l_floatingFragmentf_04 {parent=base_hand_l_floatingFragmentf_04, loc=<0 0 0>, rot=<0 0 0 1>}
 base_hand_l_floatingFragmentf_03 {parent=lowerarm_l, loc=<12.2483 -1.753 6.1837>, rot=<-0.1313 0.5556 0.1242 -0.8116>}
 dyn_hand_l_floatingFragmentf_03 {parent=base_hand_l_floatingFragmentf_03, loc=<0 0 0>, rot=<0 0 0 1>}
 base_hand_l_floatingFragmentf_02 {parent=lowerarm_l, loc=<3.4627 -4.1348 8.8174>, rot=<-0.1798 0.6058 -0.0214 -0.7747>}
 dyn_hand_l_floatingFragmentf_02 {parent=base_hand_l_floatingFragmentf_02, loc=<0 0 0>, rot=<0 0 0 1>}
 base_hand_l_floatingFragmentf_01 {parent=lowerarm_l, loc=<7.0974 2.3718 8.3407>, rot=<0.0942 0.6239 -0.028 -0.7753>}
 dyn_hand_l_floatingFragmentf_01 {parent=base_hand_l_floatingFragmentf_01, loc=<0 0 0>, rot=<0 0 0 1>}
 upperarm_twist_01_l {parent=upperarm_l, loc=<0.0041 0 0>, rot=<0 0 0 -1>}
 deform_upperarm_bk_base_l {parent=upperarm_twist_01_l, loc=<1.0813 -7.2193 -1.9453>, rot=<0 0 0 1>}
 deform_upperarm_bk_l {parent=deform_upperarm_bk_base_l, loc=<0 0 0>, rot=<0 0 0 1>}
 deform_upperarm_fr_base_l {parent=upperarm_twist_01_l, loc=<-0.3691 7.0506 -2.765>, rot=<0 0 0 1>}
 deform_upperarm_fr_l {parent=deform_upperarm_fr_base_l, loc=<0 0 0>, rot=<0 0 0 1>}
 deform_upperarm_up_base_l {parent=upperarm_twist_01_l, loc=<-3.2216 -1.1391 3.3311>, rot=<0 0 0 1>}
 deform_upperarm_up_l {parent=deform_upperarm_up_base_l, loc=<0 0 0>, rot=<0 0 0 1>}
 upperarm_twist_02_l {parent=upperarm_l, loc=<13.7941 0 0>, rot=<0 0 0 -1>}
 base_arm_l_floatingFragmentf_01 {parent=upperarm_l, loc=<2.0616 -4.1643 10.224>, rot=<-0.1909 0.0609 -0.0363 -0.979>}
 dyn_arm_l_floatingFragmentf_01 {parent=base_arm_l_floatingFragmentf_01, loc=<0 0 0>, rot=<0 0 0 1>}
 base_arm_l_floatingFragmentf_02 {parent=upperarm_l, loc=<6.7369 1.2973 7.2464>, rot=<-0.1872 -0.0644 -0.0067 -0.9802>}
 dyn_arm_l_floatingFragmentf_02 {parent=base_arm_l_floatingFragmentf_02, loc=<0 0 0>, rot=<0 0 0 1>}
 base_arm_l_floatingFragmentf_03 {parent=upperarm_l, loc=<10.0418 -5.2807 2.7818>, rot=<-0.2372 -0.381 0.0767 -0.8903>}
 dyn_arm_l_floatingFragmentf_03 {parent=base_arm_l_floatingFragmentf_03, loc=<0 0 0>, rot=<0 0 0 1>}
 base_arm_l_floatingFragmentf_04 {parent=upperarm_l, loc=<9.1942 -6.8269 -0.6315>, rot=<-0.3682 -0.4878 0.0953 -0.7857>}
 dyn_arm_l_floatingFragmentf_04 {parent=base_arm_l_floatingFragmentf_04, loc=<0 0 0>, rot=<0 0 0 1>}
 base_arm_l_floatingFragmentf_05 {parent=upperarm_l, loc=<5.1376 -7.6327 -2.566>, rot=<-0.3944 -0.6659 -0.0709 -0.6292>}
 dyn_arm_l_floatingFragmentf_05 {parent=base_arm_l_floatingFragmentf_05, loc=<0 0 0>, rot=<0 0 0 1>}
 upperarm_l_null {parent=clavicle_l, type=Null, loc=<15.5111 0 0>, rot=<0.0226 -0.3234 -0.0088 -0.9459>}
 shoulder_cube_fragments_l {parent=clavicle_l, loc=<4.6176 -5.0088 0.3105>, rot=<-0.0129 0.5667 -0.0217 -0.8235>}
 base_shoulder_l_floatingFragmentf_01 {parent=shoulder_cube_fragments_l, loc=<12.4358 2.8618 5.0158>, rot=<0 -0.5389 0 -0.8423>}
 dyn_shoulder_l_floatingFragment_01 {parent=base_shoulder_l_floatingFragmentf_01, loc=<0 0 0>, rot=<0 0 0 1>}
 base_shoulder_l_floatingFragmentf_02 {parent=shoulder_cube_fragments_l, loc=<17.1171 5.2161 3.2828>, rot=<0 -0.5389 0 -0.8423>}
 dyb_shoulder_l_floatingFragment_02 {parent=base_shoulder_l_floatingFragmentf_02, loc=<0 0 0>, rot=<0 0 0 1>}
 base_shoulder_l_floatingFragmentf_03 {parent=shoulder_cube_fragments_l, loc=<22.1671 0.2874 -0.2389>, rot=<0 -0.5389 0 -0.8423>}
 dyn_shoulder_l_floatingFragment_03 {parent=base_shoulder_l_floatingFragmentf_03, loc=<0 0 0>, rot=<0 0 0 1>}
 base_shoulder_l_floatingFragmentf_04 {parent=shoulder_cube_fragments_l, loc=<18.6302 1.4896 -2.7078>, rot=<0 -0.5389 0 -0.8423>}
 dyn_shoulder_l_floatingFragment_04 {parent=base_shoulder_l_floatingFragmentf_04, loc=<0 0 0>, rot=<0 0 0 1>}
 base_shoulder_l_floatingFragment_05 {parent=shoulder_cube_fragments_l, loc=<13.6441 -2.6957 0.0855>, rot=<0 -0.5389 0 -0.8423>}
 dyn_shoulder_l_floatingFragment_05 {parent=base_shoulder_l_floatingFragment_05, loc=<0 0 0>, rot=<0 0 0 1>}
 clavicle_r {parent=spine_05, loc=<4.3575 -0.1466 2.3223>, rot=<0.7289 0.0465 -0.6829 0.009>}
 upperarm_r {parent=clavicle_r, loc=<-15.5111 0 0>, rot=<0.0226 -0.3234 -0.0088 -0.9459>}
 lowerarm_r {parent=upperarm_r, loc=<-26.6514 0 0>, rot=<0.0391 -0.0099 -0.2453 -0.9686>}
 hand_r {parent=lowerarm_r, loc=<-22.3491 0 0>, rot=<-0.7081 -0.0507 0.0975 -0.6975>}
 thumb_01_r {parent=hand_r, loc=<-1.3118 0.3191 2.459>, rot=<0.6546 -0.4042 0.0664 -0.6354>}
 thumb_02_r {parent=thumb_01_r, loc=<-5.9336 0.0448 -0.0401>, rot=<-0.0037 0.0213 0.298 -0.9543>}
 thumb_03_r {parent=thumb_02_r, loc=<-2.2223 -0.3682 0.0907>, rot=<0.0118 -0.013 0.0066 -0.9998>}
 index_metacarpal_r {parent=hand_r, loc=<-1.6535 -0.0183 2.96>, rot=<0.0158 -0.0124 -0.0173 -0.9996>}
 index_01_r {parent=index_metacarpal_r, loc=<-8.3008 -0.0357 0.0524>, rot=<-0.0044 0.0485 0.0924 -0.9945>}
 index_02_r {parent=index_01_r, loc=<-3.3587 0.3004 -0.0452>, rot=<-0.002 0.0027 0.3225 -0.9466>}
 index_03_r {parent=index_02_r, loc=<-2.5578 0.123 0.0252>, rot=<-0.002 0.0067 0.2558 -0.9667>}
 middle_metacarpal_r {parent=hand_r, loc=<-2.1856 0.1725 0.5984>, rot=<0.006 0.0168 -0.0329 -0.9993>}
 middle_01_r {parent=middle_metacarpal_r, loc=<-7.3275 0.2447 -0.0129>, rot=<-0.0208 0.0042 0.0858 -0.9961>}
 middle_02_r {parent=middle_01_r, loc=<-4.0123 -0.1401 -0.1175>, rot=<-0.0218 0.0258 0.3512 -0.9357>}
 middle_03_r {parent=middle_02_r, loc=<-2.6077 -0.0534 -0.1373>, rot=<0.0082 0.0188 0.1708 -0.9851>}
 ring_metacarpal_r {parent=hand_r, loc=<-2.2074 0.3625 -1.3638>, rot=<-0.0284 0.0722 -0.0458 -0.9959>}
 ring_01_r {parent=ring_metacarpal_r, loc=<-7.0295 0.6197 0.1446>, rot=<-0.0489 -0.0252 0.1892 -0.9804>}
 ring_02_r {parent=ring_01_r, loc=<-3.6174 -0.4647 0.1027>, rot=<0 0 0.2806 -0.9598>}
 ring_03_r {parent=ring_02_r, loc=<-2.4152 -0.114 -0.0199>, rot=<0 0 0.1189 -0.9929>}
 pinky_metacarpal_r {parent=hand_r, loc=<-1.7616 0.9753 -3.3814>, rot=<-0.0397 0.0814 -0.069 -0.9935>}
 pinky_01_r {parent=pinky_metacarpal_r, loc=<-7.1709 0.4831 0.1094>, rot=<-0.0654 -0.0035 0.2242 -0.9723>}
 pinky_02_r {parent=pinky_01_r, loc=<-2.8274 -0.0893 0.1043>, rot=<0.0037 0.0011 0.2854 -0.9584>}
 pinky_03_r {parent=pinky_02_r, loc=<-2.2046 -0.0307 0.0078>, rot=<-0.0002 0.0098 0.1445 -0.9895>}
 deform_wrist_dn_base_r {parent=hand_r, loc=<-0.0004 2.2345 -0.0001>, rot=<0 0 0 -1>}
 deform_wrist_dn_r {parent=deform_wrist_dn_base_r, loc=<0 0 0>, rot=<0 0 0 1>}
 deform_wrist_up_base_r {parent=hand_r, loc=<0.0003 -2.2684 0.0001>, rot=<0 0 0 -1>}
 deform_wrist_up_r {parent=deform_wrist_up_base_r, loc=<0 0 0>, rot=<0 0 0 1>}
 lowerarm_twist_02_r {parent=lowerarm_r, loc=<-9.6199 0 0>, rot=<0 0 0 -1>}
 lowerarm_twist_01_r {parent=lowerarm_r, loc=<-17.8549 0 0>, rot=<0 0 0 -1>}
 deform_elbow_in_base_r {parent=lowerarm_r, loc=<-0.2111 -5.2591 -0.8275>, rot=<0 0 0 -1>}
 deform_elbow_in_r {parent=deform_elbow_in_base_r, loc=<-0.0001 0 0.0002>, rot=<0 0 0 1>}
 deform_elbow_out_base_r {parent=lowerarm_r, loc=<0 4 -0.0001>, rot=<0 0 0 -1>}
 deform_elbow_out_r {parent=deform_elbow_out_base_r, loc=<0 0 0>, rot=<0 0 0 1>}
 base_hand_r_floatingFragment_01 {parent=lowerarm_r, loc=<-6.9131 -2.6728 -8.2343>, rot=<0.7952 -0.2141 0.56 0.0907>}
 dyn_hand_r_floatingFragment_01 {parent=base_hand_r_floatingFragment_01, loc=<0 0 0>, rot=<0 0 0 1>}
 base_hand_r_floatingFragment_02 {parent=lowerarm_r, loc=<-12.3269 1.1822 -7.0091>, rot=<0.7527 -0.1638 0.6264 -0.119>}
 dyn_hand_r_floatingFragment_02 {parent=base_hand_r_floatingFragment_02, loc=<0 0 0>, rot=<0 0 0 1>}
 base_hand_r_floatingFragment_03 {parent=lowerarm_r, loc=<-14.0731 5.3981 -3.1781>, rot=<0.6722 0.1317 0.5522 -0.4752>}
 dyn_hand_r_floatingFragment_03 {parent=base_hand_r_floatingFragment_03, loc=<0 0 0>, rot=<0 0 0 1>}
 upperarm_twist_01_r {parent=upperarm_r, loc=<-0.0041 0 0>, rot=<0 0 0 -1>}
 deform_upperarm_bk_base_r {parent=upperarm_twist_01_r, loc=<-1.0814 7.2192 1.9455>, rot=<0 0 0 1>}
 deform_upperarm_bk_r {parent=deform_upperarm_bk_base_r, loc=<0 0 0>, rot=<0 0 0 1>}
 deform_upperarm_fr_base_r {parent=upperarm_twist_01_r, loc=<0.3689 -7.0506 2.7652>, rot=<0 0 0 1>}
 deform_upperarm_fr_r {parent=deform_upperarm_fr_base_r, loc=<0 0 0>, rot=<0 0 0 1>}
 deform_upperarm_up_base_r {parent=upperarm_twist_01_r, loc=<3.2214 1.1391 -3.3308>, rot=<0 0 0 1>}
 deform_upperarm_up_r {parent=deform_upperarm_up_base_r, loc=<0 0 0>, rot=<0 0 0 -1>}
 upperarm_twist_02_r {parent=upperarm_r, loc=<-13.7941 0 0>, rot=<0 0 0 -1>}
 shoulder_cube_fragments_r {parent=upperarm_r, loc=<0 0 0>, rot=<0.7111 0.2284 -0.6396 -0.1819>}
 base_arm_r_floatingFragmentf_03 {parent=upperarm_r, loc=<-5.0776 -7.1179 1.9916>, rot=<0.4551 -0.7897 0.065 0.4062>}
 dyn_arm_r_floatingFragmentf_03 {parent=base_arm_r_floatingFragmentf_03, loc=<0 0 0>, rot=<0 0 0 1>}
 base_arm_r_floatingFragmentf_02 {parent=upperarm_r, loc=<-1.1625 3.5817 -13.3721>, rot=<0.953 0.1286 0.0723 -0.2648>}
 dyn_arm_r_floatingFragmentf_02 {parent=base_arm_r_floatingFragmentf_02, loc=<0 0 0>, rot=<0 0 0 1>}
 base_arm_r_floatingFragmentf_01 {parent=upperarm_r, loc=<1.6539 9.9167 -6.1214>, rot=<0.9456 0.0614 0.2198 -0.2318>}
 dyn_arm_r_floatingFragmentf_01 {parent=base_arm_r_floatingFragmentf_01, loc=<0 0 0>, rot=<0 0 0 1>}
 upperarm_r_null {parent=clavicle_r, type=Null, loc=<-15.5111 0 0>, rot=<0.0226 -0.3234 -0.0088 -0.9459>}
 clavicle_l_null {parent=spine_05, type=Null, loc=<4.3575 -0.1466 -2.3223>, rot=<0.0465 -0.7289 -0.009 -0.6829>}
 clavicle_r_null {parent=spine_05, type=Null, loc=<4.3575 -0.1466 2.3223>, rot=<0.7289 0.0465 -0.6829 0.009>}
 deform_pec_aim_base_l {parent=spine_04, loc=<7.8732 12.6388 -2.4123>, rot=<0.0102 -0.4602 0.3979 -0.7936>}
 deform_pec_aim_l {parent=deform_pec_aim_base_l, loc=<0 0 0>, rot=<0 0 0 1>}
 deform_pec_base_l {parent=deform_pec_aim_l, loc=<10.766 0 0>, rot=<0 0 0 1>}
 deform_pec_l {parent=deform_pec_base_l, loc=<0.055 0 0>, rot=<0 0 0 1>}
 deform_pec_aim_base_r {parent=spine_04, loc=<7.8727 12.6388 2.4123>, rot=<-0.4602 -0.0102 0.7936 0.3979>}
 deform_pec_aim_r {parent=deform_pec_aim_base_r, loc=<0 0 0>, rot=<0 0 0 1>}
 deform_pec_base_r {parent=deform_pec_aim_r, loc=<-10.7662 -0.0006 -0.0007>, rot=<0 0 0 1>}
 deform_pec_r {parent=deform_pec_base_r, loc=<0 0 0>, rot=<0 0 0 1>}
 base_torso_cube {parent=spine_04, loc=<7.3501 9.952 0.0067>, rot=<0.0294 -0.7065 -0.0294 -0.7065>}
 dyn_torso_cube {parent=base_torso_cube, loc=<0 0 0>, rot=<0 0 0 1>}
 base_torso_cube_A {parent=spine_04, loc=<6.6845 9.9374 -0.0199>, rot=<0.0294 -0.7065 -0.0294 -0.7065>}
 dyn_torso_cube_A {parent=base_torso_cube_A, loc=<0.025 0 0>, rot=<0 0 0 1>}
 base_torso_cube_B {parent=spine_04, loc=<6.6845 9.9374 -0.0199>, rot=<0.0294 -0.7065 -0.0294 -0.7065>}
 dyn_torso_cube_B {parent=base_torso_cube_B, loc=<-0.025 0 0>, rot=<0 0 0 1>}
 base_torso_cube_C {parent=spine_04, loc=<6.6845 9.9374 -0.0199>, rot=<0.0294 -0.7065 -0.0294 -0.7065>}
 dyn_torso_cube_C {parent=base_torso_cube_C, loc=<0 0 0.025>, rot=<0 0 0 1>}
 base_torso_cube_D {parent=spine_04, loc=<6.6845 9.9374 -0.0199>, rot=<0.0294 -0.7065 -0.0294 -0.7065>}
 dyn_torso_cube_D {parent=base_torso_cube_D, loc=<0 0 -0.025>, rot=<0 0 0 1>}
 base_torso_cube_E {parent=spine_04, loc=<6.6845 9.9374 -0.0199>, rot=<0.0294 -0.7065 -0.0294 -0.7065>}
 dyn_torso_cube_E {parent=base_torso_cube_E, loc=<0 0 0>, rot=<0 0 0 1>}
 base_lat_r {parent=spine_04, loc=<2.8649 0 14.4593>, rot=<0 0.7071 0 -0.7071>}
 dyn_lat_r______ {parent=base_lat_r, loc=<0 0 0>, rot=<0 0 0 1>}
 dyn_lat_r {parent=base_lat_r, loc=<0 0 0>, rot=<0 0 0 1>}
 base_lat_l {parent=spine_04, loc=<2.8649 0 -15.3445>, rot=<0 -0.7071 0 -0.7071>}
 dyn_lat_l {parent=base_lat_l, loc=<0 0 0>, rot=<0 0 0 1>}
 base_body_floatingCube_01 {parent=spine_01, loc=<3.6162 -0.5636 -15.6002>, rot=<0.8677 0.148 0.256 0.3995>}
 dyn_body_floatingCube_01 {parent=base_body_floatingCube_01, loc=<0 0 0>, rot=<0 0 0 1>}
 thigh_l {parent=pelvis, loc=<-2.5693 -0.0109 -7.7501>, rot=<0.0012 0.0572 0.0155 -0.9982>}
 calf_l {parent=thigh_l, loc=<-39.0484 0 0>, rot=<0.0006 -0.0161 -0.0391 -0.9991>}
 foot_l {parent=calf_l, loc=<-42.561 0.1077 0.4292>, rot=<0.0418 -0.0455 0.0343 -0.9975>}
 ball_l {parent=foot_l, loc=<-6.1599 13.8264 0.6806>, rot=<0.0001 -0.0001 -0.7063 -0.7079>}
 base_leg_r_floatingCube_05 {parent=foot_l, loc=<-0.9685 6.4037 -5.6798>, rot=<-0.0357 -0.7045 -0.0284 -0.7082>}
 dyn_leg_r_floatingCube_05 {parent=base_leg_r_floatingCube_05, loc=<0 0 0>, rot=<0 0 0 1>}
 calf_twist_02_l {parent=calf_l, loc=<-18.8064 -0.0967 0.0397>, rot=<0.0237 0.0095 -0.0035 -0.9997>}
 calf_twist_01_l {parent=calf_l, loc=<-35.813 -0.0967 0.0397>, rot=<0.0237 0.0095 -0.0035 -0.9997>}
 deform_knee_in_base_l {parent=calf_l, loc=<1.298 -5.2957 -0.0496>, rot=<0 0 0 -1>}
 deform_knee_in_l {parent=deform_knee_in_base_l, loc=<0 0 0>, rot=<0 0 0 1>}
 deform_knee_out_base_l {parent=calf_l, loc=<0 4 0>, rot=<0 0 0 -1>}
 deform_knee_out_l {parent=deform_knee_out_base_l, loc=<0 0 0>, rot=<0 0 0 1>}
 base_leg_l_floatingCube_01 {parent=calf_l, loc=<-4.4674 -8.4021 -10.5685>, rot=<-0.0315 -0.7349 0.027 -0.6769>}
 dyn_leg_l_floatingCube_01 {parent=base_leg_l_floatingCube_01, loc=<0 0 0>, rot=<0 0 0 1>}
 base_leg_l_floatingCube_02 {parent=calf_l, loc=<-6.9917 -3.8416 -8.806>, rot=<-0.0315 -0.7349 0.027 -0.6769>}
 dyn_leg_l_floatingCube_02 {parent=base_leg_l_floatingCube_02, loc=<0 0 0>, rot=<0 0 0 1>}
 base_leg_l_floatingCube_03 {parent=calf_l, loc=<-36.2735 -4.0668 -8.1858>, rot=<-0.0315 -0.7349 0.027 -0.6769>}
 dyn_leg_l_floatingCube_03 {parent=base_leg_l_floatingCube_03, loc=<0 0 0>, rot=<0 0 0 1>}
 base_leg_l_floatingCube_04 {parent=calf_l, loc=<-34.2456 1.6853 5.0314>, rot=<-0.0315 -0.7349 0.027 -0.6769>}
 dyn_leg_l_floatingCube_04 {parent=base_leg_l_floatingCube_04, loc=<0 0 0>, rot=<0 0 0 1>}
 thigh_twist_01_l {parent=thigh_l, loc=<-21.6303 0 0>, rot=<0.0013 0 0 -1>}
 deform_groin_base_l {parent=thigh_l, loc=<-2.1908 8.1142 -1.7442>, rot=<0 0 0 -1>}
 deform_groin_l {parent=deform_groin_base_l, loc=<0 0 0>, rot=<0 0 0 1>}
 thigh_r {parent=pelvis, loc=<-2.5693 -0.0109 7.7501>, rot=<0.0572 -0.0012 0.9982 0.0155>}
 calf_r {parent=thigh_r, loc=<39.0484 0 0>, rot=<0.0006 -0.0161 -0.0391 -0.9991>}
 foot_r {parent=calf_r, loc=<42.561 -0.1077 -0.4292>, rot=<0.0418 -0.0455 0.0343 -0.9975>}
 ball_r {parent=foot_r, loc=<6.1599 -13.8264 -0.6806>, rot=<0.0001 -0.0001 -0.7063 -0.7079>}
 base_leg_r_floatingCube_04 {parent=foot_r, loc=<0.3128 -8.623 4.5978>, rot=<0.7082 0.0284 -0.7045 -0.0357>}
 dyn_leg_r_floatingCube_04 {parent=base_leg_r_floatingCube_04, loc=<0 0 0>, rot=<0 0 0 1>}
 calf_twist_02_r {parent=calf_r, loc=<18.8064 0.0967 -0.0397>, rot=<0.0237 0.0095 -0.0035 -0.9997>}
 base_leg_r_floatingCube_01 {parent=calf_twist_02_r, loc=<4.0762 -0.3126 6.286>, rot=<-0.6688 0.0123 0.7417 0.0498>}
 dyn_leg_r_floatingCube_01 {parent=base_leg_r_floatingCube_01, loc=<0 0 0>, rot=<0 0 0 1>}
 base_leg_r_floatingCube_02 {parent=calf_twist_02_r, loc=<13.7063 2.7426 6.6178>, rot=<-0.6688 0.0123 0.7417 0.0498>}
 dyn_leg_r_floatingCube_02 {parent=base_leg_r_floatingCube_02, loc=<0 0 0>, rot=<0 0 0 1>}
 calf_twist_01_r {parent=calf_r, loc=<35.813 0.0967 -0.0397>, rot=<0.0237 0.0095 -0.0035 -0.9997>}
 base_leg_r_floatingCube_03 {parent=calf_twist_01_r, loc=<1.0897 1.4958 -6.124>, rot=<-0.6688 0.0123 0.7417 0.0498>}
 dyn_leg_r_floatingCube_03 {parent=base_leg_r_floatingCube_03, loc=<0 0 0>, rot=<0 0 0 1>}
 deform_knee_in_base_r {parent=calf_r, loc=<-1.298 5.2957 0.0496>, rot=<0 0 0 -1>}
 deform_knee_in_r {parent=deform_knee_in_base_r, loc=<0 0 0>, rot=<0 0 0 1>}
 deform_knee_out_base_r {parent=calf_r, loc=<0 -4 0>, rot=<0 0 0 -1>}
 deform_knee_out_r {parent=deform_knee_out_base_r, loc=<0 0 0>, rot=<0 0 0 1>}
 thigh_twist_01_r {parent=thigh_r, loc=<21.6303 0 0>, rot=<0.0013 0 0 -1>}
 deform_groin_base_r {parent=thigh_r, loc=<2.1908 -8.1142 1.7442>, rot=<0 0 0 -1>}
 deform_groin_r {parent=deform_groin_base_r, loc=<0 0 0>, rot=<0 0 0 1>}
 shoulder_cube_fragments_l_ctrl {controlType=EULER_TRANSFORM, type=Control, loc=<0 0 0>, rot=<0 0 0 1>}
]
variables [
 Ambient Rotation {type=FRotator, default=(Pitch=0,Yaw=0,Roll=0)}
 Ambient Traslation {type=FVector, default=(X=0,Y=0,Z=0)}
 ShardSlowAmount {type=double, default=1}
]
