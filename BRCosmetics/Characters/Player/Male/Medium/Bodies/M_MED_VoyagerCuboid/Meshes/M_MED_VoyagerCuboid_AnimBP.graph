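# Animation Blueprint for the VoyagerCuboid player body cosmetic, copying the attached parent mesh's pose and passing it through the body's Control Rig.
# Each update derives a steady spin on all three axes from real time, plus a sine-based drift at differing frequencies, with downward Z motion also nudging Y.
# These fill Ambient Rotation and Translation variables; the rig node's matching inputs show zero values and no visible link to them.
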AnimBlueprint /BRCosmetics/Characters/Player/Male/Medium/Bodies/M_MED_VoyagerCuboid/Meshes/M_MED_VoyagerCuboid_AnimBP
target_skeleton=&/M_MED_VoyagerCuboid_Skeleton
graphs [
 AnimGraph {nodes [
   n1 = Control Rig!(Source=Copy Pose From Mesh() {bUseAttachedParent=true}, Alpha, Ambient Rotation:Rotator="0, 0, 0", Ambient Traslation:Vector="0, 0, 0") -> (Pose) {controlRigAssetReference {blueprintRigClass=&/M_MED_VoyagerCuboid_CtrlRig.M_MED_VoyagerCuboid_CtrlRig_C}}
   Output Pose(Result=@n1) {layerGroup=None}
  ]}
 EventGraph {nodes [
   Event BlueprintUpdateAnimation()
   n1 = float * float(A=float / float(A=GetRealTimeSeconds(), B=-720) {is=title K2Node_CallFunction}, B=93750)
   > Set Ambient Rotation(Ambient Rotation:Rotator=MakeRotator(Roll=@n1, Pitch=@n1, Yaw=@n1)) -> (Output_Get:Rotator)
   n2 = GetRealTimeSeconds()
   n3 = float * float(A="Sin (Radians)"(A=float * float(A=1.2, B=@n2)), B=0.2)
   n4 = float * float(A="Sin (Radians)"(A=float * float(A=1, B=@n2)), B=0.3)
   n5 = Remap Float(Input="Sin (Radians)"(A=float * float(A=2, B=@n2)), OldRange=(X=-1,Y=1), NewRange=(X=-1,Y=2)) -> (Output=~)
   n6 = float * float(A="Clamp (Float)"(Value=@n5, Min=-1, Max=0), B=0.5)
   > Set Ambient Translation(Ambient Translation:Vector=vector + vector(A=MakeVector(X=@n3, Y=@n4, Z=@n5), B=MakeVector(Y=@n6))) -> (Output_Get:Vector)
  ]}
]
variables [
 Ambient Rotation {type=Rotator}
 Ambient Translation {type=Vector}
]
